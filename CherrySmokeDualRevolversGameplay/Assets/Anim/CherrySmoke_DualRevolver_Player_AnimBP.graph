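# Player animation Blueprint for the CherrySmoke dual revolvers, overriding item layers for directional air-dodge and roll animations.
# A Dodge state machine runs start, in-air, fall, recovery and neutral double-jump states, choosing left/right graph blendspaces by dodge angle or direction.
# Neutral Recovery adds aim offset and alternating-hand additive firing.
# A reload flag toggles upper-body masking of the dodge, then Hand Adjustment layers hand poses based on tactical sprint.
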AnimBlueprint /CherrySmokeDualRevolversGameplay/Assets/Anim/CherrySmoke_DualRevolver_Player_AnimBP
target_skeleton=&Fortnite_M_Avg_Player_Skeleton
graphs [
 AnimGraph {nodes [Output Pose()]}
 EventGraph {links [], nodes []}
 Hand Adjustment {nodes [
   n1 = Sequence Player 'CherrySmoke_DualRevolver_Targeted_Pose'() {bLoopAnimation=false, sequence=/CherrySmokeDualRevolversGameplay/Assets/Anim/Poses/CherrySmoke_DualRevolver_Targeted_Pose}
   n2 = Use cached pose 'Hand Adjustment Input Pose'()
   Output Pose(Result=Blend Poses by bool(
     BlendPose_0=Layered blend per bone(BasePose=@n2, BlendPoses_0=@n1) {layerSetup [{branchFilters [{boneName=hand_r}, {boneName=hand_l}]}]}
     BlendPose_1=Use cached pose 'Hand Adjustment Input Pose'()
     bActiveValue=Get bIsTacticalSprint() -> (bIsTacticalSprint:bool)
    ) {blendType=Linear})
   Save cached pose 'Hand Adjustment Input Pose'(Pose=Input Pose()) {cacheName=Hand Adjustment Input Pose}
  ]}
 ItemAimOffset {nodes [
   Input Pose!() -> (Pose, AimOffsetAlpha:double, Yaw:double, Pitch:double) {name=InPose_AimOffset}
   Input Pose!() -> (Pose) {name=InPose_PreAimOffsets}
   Output Pose() {layerGroup=Item Override}
  ]}
 ItemFallAdditive {nodes [
   Input Pose() {name=InPoseItemFallAdditive}
   Output Pose() {layerGroup=Item Override}
  ]}
 ItemFallLandAdditive {nodes [
   Input Pose() {name=InPoseItemFallLandAdditive}
   Output Pose() {layerGroup=Item Override}
  ]}
 ItemFallLevel2Override {nodes [
   Input Pose() {name=InPose_FallLevel2}
   Output Pose() {layerGroup=Item Override}
  ]}
 ItemFinalPoseOverride {nodes [
   Input Pose() {name=InFinalPose}
   Output Pose() {layerGroup=Item Override}
  ]}
 ItemFullBodyAimOffsets {nodes [
   Input Pose!() -> (Pose, FullBody_AimOffsetAlpha:double, Aim_Yaw:double, Aim_Pitch:double) {name=InPoseFullBodyAimOffsets}
   Input Pose!() -> (Pose) {name=InPosePreFullBodyAimOffsets}
   Output Pose() {layerGroup=Item Override}
  ]}
 ItemFullBodyOverride {nodes [
   n1 = Use cached pose 'MaskUpperBody'()
   n2 = Use cached pose 'Dodge SM'()
   n3 = Get bReloading() -> (bReloading:bool)
   Output Pose(Result=Hand Adjustment!(InPose:Pose Link Structure=Blend Poses by bool(BlendPose_0=@n1, BlendPose_1=@n2, bActiveValue=@n3) {blendType=Linear}) -> (Pose:Pose Link Structure) {is=title AnimGraphNode_LinkedAnimLayer, layer=Hand Adjustment}) {layerGroup=Item Override}
   Save cached pose 'Incoming Pose'(Pose=Input Pose() {name=InPoseFullBody}) {cacheName=Incoming Pose}
   AnimGraphNode_StateMachine_0 = Dodge SM() {is=title AnimGraphNode_StateMachine}
   Save cached pose 'Dodge SM'(Pose=@AnimGraphNode_StateMachine_0) {cacheName=Dodge SM, cachePoseName=Dodge SM}
   Save cached pose 'MaskUpperBody'(Pose=Layered blend per bone(BasePose=Use cached pose 'Dodge SM'(), BlendPoses_0=Use cached pose 'Incoming Pose'()) {layerSetup [{branchFilters [{blendDepth=3, boneName=Spine_03}]}]}) {cacheName=MaskUpperBody}
  ]}
 Dodge SM {
  entry=Empty
  path=ItemFullBodyOverride.AnimGraphNode_StateMachine_0.Dodge SM
  ids=id
  nodes [
   To Neutral Roll() {is=title AnimStateAliasNode, also {aliases [Empty]}}
   Exit Dodge() {is=title AnimStateAliasNode, also {aliases [Fall, In Air, Recovery Roll, Start Roll]}}
   AnimStateConduitNode_1 = Air Dodge Active() {is=title AnimStateConduitNode}
   AnimStateNode_0 = Empty() {
    is=title AnimStateNode
    stateEntered {guid=#1}
    stateFullyBlended {guid=#2}
    stateLeft {guid=#3}
   }
   AnimStateNode_1 = Start Roll() {
    is=title AnimStateNode
    bAlwaysResetOnEntry=true
    stateEntered {guid=#4}
    stateFullyBlended {guid=#5}
    stateLeft {guid=#6}
   }
   AnimStateNode_2 = Neutral Roll() {is=title AnimStateNode}
   AnimStateNode_3 = Recovery Roll() {
    is=title AnimStateNode
    bAlwaysResetOnEntry=true
    stateEntered {guid=#7}
    stateFullyBlended {guid=#8}
    stateLeft {guid=#9}
   }
   AnimStateNode_4 = Neutral Recovery() {is=title AnimStateNode}
   AnimStateNode_5 = In Air() {
    is=title AnimStateNode
    bAlwaysResetOnEntry=true
    stateEntered {guid=#10}
    stateFullyBlended {guid=#11}
    stateLeft {guid=#12}
   }
   AnimStateNode_7 = Fall() {
    is=title AnimStateNode
    bAlwaysResetOnEntry=true
    stateEntered {guid=#13}
    stateFullyBlended {guid=#14}
    stateLeft {guid=#15}
   }
   AnimStateTransitionNode_0 = AnimStateTransitionNode()
   AnimStateTransitionNode_1 = AnimStateTransitionNode() {
    bAutomaticRuleBasedOnSequencePlayerInState=true
    transitionEnd {guid=#16}
    transitionInterrupt {guid=#17}
    transitionStart {guid=#18}
   }
   AnimStateTransitionNode_10 = AnimStateTransitionNode() {
    transitionEnd {guid=#19}
    transitionInterrupt {guid=#20}
    transitionStart {guid=#21}
   }
   AnimStateTransitionNode_11 = AnimStateTransitionNode() {crossfadeDuration=0}
   AnimStateTransitionNode_13 = AnimStateTransitionNode()
   AnimStateTransitionNode_14 = AnimStateTransitionNode() {bAutomaticRuleBasedOnSequencePlayerInState=true, crossfadeDuration=0.1}
   AnimStateTransitionNode_15 = AnimStateTransitionNode() {transitionStart {guid=#22, notifyName=Start Roll Transition}}
   AnimStateTransitionNode_16 = AnimStateTransitionNode() {bAutomaticRuleBasedOnSequencePlayerInState=true}
   AnimStateTransitionNode_2 = AnimStateTransitionNode() {bAutomaticRuleBasedOnSequencePlayerInState=true, crossfadeDuration=0.25, logicType=TLT_Inertialization, priorityOrder=2}
   AnimStateTransitionNode_4 = AnimStateTransitionNode()
   AnimStateTransitionNode_5 = AnimStateTransitionNode() {crossfadeDuration=0}
   AnimStateTransitionNode_6 = AnimStateTransitionNode() {bAutomaticRuleBasedOnSequencePlayerInState=true}
   AnimStateTransitionNode_7 = AnimStateTransitionNode() {
    crossfadeDuration=0.1
    transitionEnd {guid=#23}
    transitionInterrupt {guid=#24}
    transitionStart {guid=#25}
   }
   AnimStateTransitionNode_8 = AnimStateTransitionNode()
  ]
  links [
   {
    dst=Start Roll
    rule=ItemFullBodyOverride.AnimGraphNode_StateMachine_0.Dodge SM.AnimStateTransitionNode_15.Transition
    src=Air Dodge Active
    via=AnimStateTransitionNode_15
   }
   {
    dst=Air Dodge Active
    rule=ItemFullBodyOverride.AnimGraphNode_StateMachine_0.Dodge SM.AnimStateTransitionNode_10.Transition
    src=Empty
    via=AnimStateTransitionNode_10
   }
   {
    dst=Empty
    rule=ItemFullBodyOverride.AnimGraphNode_StateMachine_0.Dodge SM.AnimStateTransitionNode_0.Transition
    src=Exit Dodge
    via=AnimStateTransitionNode_0
   }
   {
    dst=Empty
    rule=ItemFullBodyOverride.AnimGraphNode_StateMachine_0.Dodge SM.AnimStateTransitionNode_8.Transition
    src=Exit Dodge
    via=AnimStateTransitionNode_8
   }
   {
    dst=Recovery Roll
    rule=ItemFullBodyOverride.AnimGraphNode_StateMachine_0.Dodge SM.AnimStateTransitionNode_11.Transition
    src=Fall
    via=AnimStateTransitionNode_11
   }
   {
    dst=Recovery Roll
    rule=ItemFullBodyOverride.AnimGraphNode_StateMachine_0.Dodge SM.AnimStateTransitionNode_16.Transition
    src=Fall
    via=AnimStateTransitionNode_16
   }
   {
    dst=Fall
    rule=ItemFullBodyOverride.AnimGraphNode_StateMachine_0.Dodge SM.AnimStateTransitionNode_7.Transition
    src=In Air
    via=AnimStateTransitionNode_7
   }
   {
    dst=Empty
    rule=ItemFullBodyOverride.AnimGraphNode_StateMachine_0.Dodge SM.AnimStateTransitionNode_5.Transition
    src=Neutral Recovery
    via=AnimStateTransitionNode_5
   }
   {
    dst=Neutral Recovery
    rule=ItemFullBodyOverride.AnimGraphNode_StateMachine_0.Dodge SM.AnimStateTransitionNode_4.Transition
    src=Neutral Roll
    via=AnimStateTransitionNode_4
   }
   {
    dst=Neutral Recovery
    rule=ItemFullBodyOverride.AnimGraphNode_StateMachine_0.Dodge SM.AnimStateTransitionNode_6.Transition
    src=Neutral Roll
    via=AnimStateTransitionNode_6
   }
   {
    dst=Air Dodge Active
    rule=ItemFullBodyOverride.AnimGraphNode_StateMachine_0.Dodge SM.AnimStateTransitionNode_1.Transition
    src=Recovery Roll
    via=AnimStateTransitionNode_1
   }
   {
    dst=Empty
    rule=ItemFullBodyOverride.AnimGraphNode_StateMachine_0.Dodge SM.AnimStateTransitionNode_2.Transition
    src=Recovery Roll
    via=AnimStateTransitionNode_2
   }
   {
    dst=In Air
    rule=ItemFullBodyOverride.AnimGraphNode_StateMachine_0.Dodge SM.AnimStateTransitionNode_14.Transition
    src=Start Roll
    via=AnimStateTransitionNode_14
   }
   {
    dst=Neutral Roll
    rule=ItemFullBodyOverride.AnimGraphNode_StateMachine_0.Dodge SM.AnimStateTransitionNode_13.Transition
    src=To Neutral Roll
    via=AnimStateTransitionNode_13
   }
  ]
 }
 Air Dodge Active {
  path=ItemFullBodyOverride.AnimGraphNode_StateMachine_0.Dodge SM.AnimStateConduitNode_1.Air Dodge Active
  nodes [Result(bCanEnterTransition=Get bAirDodgeActive() -> (bAirDodgeActive:bool))]
 }
 Empty {
  path=ItemFullBodyOverride.AnimGraphNode_StateMachine_0.Dodge SM.AnimStateNode_0.Empty
  nodes [
   n1 = Use cached pose 'Incoming Pose'() {becomeRelevantFunction {bSelfContext=true, memberGuid=#26, memberName=OnEmptyBecomeRelevant}}
   Output Animation Pose(Result=@n1)
  ]
 }
 Start Roll {
  path=ItemFullBodyOverride.AnimGraphNode_StateMachine_0.Dodge SM.AnimStateNode_1.Start Roll
  nodes [
   AnimGraphNode_StateMachine_0 = Dodge Start SM() {is=title AnimGraphNode_StateMachine, bAllowConduitEntryStates=true}
   Output Animation Pose(Result=@AnimGraphNode_StateMachine_0) {
    becomeRelevantFunction {bSelfContext=true, memberGuid=#27, memberName=OnStartRollRelevant}
    node {becomeRelevantFunction {functionName=OnBecomeRelevant}}
   }
  ]
 }
 Dodge Start SM {
  entry=Is Left
  path=ItemFullBodyOverride.AnimGraphNode_StateMachine_0.Dodge SM.AnimStateNode_1.Start Roll.AnimGraphNode_StateMachine_0.Dodge Start SM
  ids=id
  nodes [
   AnimStateConduitNode_0 = Is Left() {is=title AnimStateConduitNode}
   AnimStateNode_0 = Left() {
    is=title AnimStateNode
    bAlwaysResetOnEntry=true
    stateEntered {guid=#28}
    stateFullyBlended {guid=#29}
    stateLeft {guid=#30}
   }
   AnimStateNode_1 = Right() {
    is=title AnimStateNode
    bAlwaysResetOnEntry=true
    stateEntered {guid=#31}
    stateFullyBlended {guid=#32}
    stateLeft {guid=#33}
   }
   AnimStateTransitionNode_2 = AnimStateTransitionNode() {
    transitionEnd {guid=#34}
    transitionInterrupt {guid=#35}
    transitionStart {guid=#36}
   }
   AnimStateTransitionNode_3 = AnimStateTransitionNode() {
    transitionEnd {guid=#37}
    transitionInterrupt {guid=#38}
    transitionStart {guid=#39}
   }
  ]
  links [
   {
    dst=Left
    rule=ItemFullBodyOverride.AnimGraphNode_StateMachine_0.Dodge SM.AnimStateNode_1.Start Roll.AnimGraphNode_StateMachine_0.Dodge Start SM.AnimStateTransitionNode_2.AnimationTransitionGraph_1
    src=Is Left
    via=AnimStateTransitionNode_2
   }
   {
    dst=Right
    rule=ItemFullBodyOverride.AnimGraphNode_StateMachine_0.Dodge SM.AnimStateNode_1.Start Roll.AnimGraphNode_StateMachine_0.Dodge Start SM.AnimStateTransitionNode_3.AnimationTransitionGraph_1
    src=Is Left
    via=AnimStateTransitionNode_3
   }
  ]
 }
 Is Left {
  path=ItemFullBodyOverride.AnimGraphNode_StateMachine_0.Dodge SM.AnimStateNode_1.Start Roll.AnimGraphNode_StateMachine_0.Dodge Start SM.AnimStateConduitNode_0.Is Left
  nodes [Result(bCanEnterTransition=true)]
 }
 Left {
  path=ItemFullBodyOverride.AnimGraphNode_StateMachine_0.Dodge SM.AnimStateNode_1.Start Roll.AnimGraphNode_StateMachine_0.Dodge Start SM.AnimStateNode_0.Left
  nodes [
   AnimGraphNode_BlendSpaceGraph_0 = Blendspace 'Left Dodge Start'!(X:float=Get DodgeAngleRadians() -> (DodgeAngleRadians:double)) -> (Pose:Pose Link Structure) {is=title AnimGraphNode_BlendSpaceGraph, groupName=LeftDodgeStart}
   Output Animation Pose(Result=@AnimGraphNode_BlendSpaceGraph_0)
  ]
 }
 Left Dodge Start {
  path=ItemFullBodyOverride.AnimGraphNode_StateMachine_0.Dodge SM.AnimStateNode_1.Start Roll.AnimGraphNode_StateMachine_0.Dodge Start SM.AnimStateNode_0.Left.AnimGraphNode_BlendSpaceGraph_0.Left Dodge Start
  links []
  nodes []
 }
 BackwardDodgeStart2 {
  path=ItemFullBodyOverride.AnimGraphNode_StateMachine_0.Dodge SM.AnimStateNode_1.Start Roll.AnimGraphNode_StateMachine_0.Dodge Start SM.AnimStateNode_0.Left.AnimGraphNode_BlendSpaceGraph_0.Left Dodge Start.BackwardDodgeStart2
  nodes [
   n1 = Sequence Player 'CherrySmoke_DualRevolver_Dodge_B_Start_anim'() {
    bLoopAnimation=false
    bOverridePositionWhenJoiningSyncGroupAsLeader=true
    sequence=/CherrySmokeDualRevolversGameplay/Assets/Anim/TEMP/CherrySmoke_DualRevolver_Dodge_B_Start_anim
   }
   Output Animation Pose!(Result:Pose Link Structure=@n1) {is=title AnimGraphNode_BlendSpaceSampleResult}
  ]
 }
 BackwardsDodgeStart {
  path=ItemFullBodyOverride.AnimGraphNode_StateMachine_0.Dodge SM.AnimStateNode_1.Start Roll.AnimGraphNode_StateMachine_0.Dodge Start SM.AnimStateNode_0.Left.AnimGraphNode_BlendSpaceGraph_0.Left Dodge Start.BackwardsDodgeStart
  nodes [
   n1 = Sequence Player 'CherrySmoke_DualRevolver_Dodge_B_Start_anim'() {
    bLoopAnimation=false
    bOverridePositionWhenJoiningSyncGroupAsLeader=true
    sequence=&CherrySmoke_DualRevolver_Dodge_B_Start_anim
   }
   Output Animation Pose(Result=@n1)
  ]
 }
 ForwardDodgeStart {
  path=ItemFullBodyOverride.AnimGraphNode_StateMachine_0.Dodge SM.AnimStateNode_1.Start Roll.AnimGraphNode_StateMachine_0.Dodge Start SM.AnimStateNode_0.Left.AnimGraphNode_BlendSpaceGraph_0.Left Dodge Start.ForwardDodgeStart
  nodes [
   n1 = Sequence Player 'CherrySmoke_DualRevolver_Dodge_F_Start_anim'() {
    bLoopAnimation=false
    bOverridePositionWhenJoiningSyncGroupAsLeader=true
    sequence=/CherrySmokeDualRevolversGameplay/Assets/Anim/TEMP/CherrySmoke_DualRevolver_Dodge_F_Start_anim
   }
   Output Animation Pose(Result=@n1)
  ]
 }
 LeftDodgeStart {
  path=ItemFullBodyOverride.AnimGraphNode_StateMachine_0.Dodge SM.AnimStateNode_1.Start Roll.AnimGraphNode_StateMachine_0.Dodge Start SM.AnimStateNode_0.Left.AnimGraphNode_BlendSpaceGraph_0.Left Dodge Start.LeftDodgeStart
  nodes [Output Animation Pose(Result=Sequence Player 'CherrySmoke_DualRevolver_Dodge_L_Start_anim'() {bLoopAnimation=false, sequence=/CherrySmokeDualRevolversGameplay/Assets/Anim/TEMP/CherrySmoke_DualRevolver_Dodge_L_Start_anim})]
 }
 Left_BackLeft_Start {
  path=ItemFullBodyOverride.AnimGraphNode_StateMachine_0.Dodge SM.AnimStateNode_1.Start Roll.AnimGraphNode_StateMachine_0.Dodge Start SM.AnimStateNode_0.Left.AnimGraphNode_BlendSpaceGraph_0.Left Dodge Start.Left_BackLeft_Start
  nodes [
   n1 = Sequence Player 'CherrySmoke_DualRevolver_Dodge_LeftSide_BL_Start'() {bLoopAnimation=false, sequence=/CherrySmokeDualRevolversGameplay/Assets/Anim/CharacterAnim/CherrySmoke_DualRevolver_Dodge_LeftSide_BL_Start}
   Output Animation Pose(Result=@n1)
  ]
 }
 Left_BackRight_Start {
  path=ItemFullBodyOverride.AnimGraphNode_StateMachine_0.Dodge SM.AnimStateNode_1.Start Roll.AnimGraphNode_StateMachine_0.Dodge Start SM.AnimStateNode_0.Left.AnimGraphNode_BlendSpaceGraph_0.Left Dodge Start.Left_BackRight_Start
  nodes [
   n1 = Sequence Player 'CherrySmoke_DualRevolver_Dodge_LeftSide_BR_Start'() {bLoopAnimation=false, sequence=/CherrySmokeDualRevolversGameplay/Assets/Anim/CharacterAnim/CherrySmoke_DualRevolver_Dodge_LeftSide_BR_Start}
   Output Animation Pose(Result=@n1)
  ]
 }
 Left_FrontLeft_Start {
  path=ItemFullBodyOverride.AnimGraphNode_StateMachine_0.Dodge SM.AnimStateNode_1.Start Roll.AnimGraphNode_StateMachine_0.Dodge Start SM.AnimStateNode_0.Left.AnimGraphNode_BlendSpaceGraph_0.Left Dodge Start.Left_FrontLeft_Start
  nodes [
   n1 = Sequence Player 'CherrySmoke_DualRevolver_Dodge_LeftSide_FL_Start'() {bLoopAnimation=false, sequence=/CherrySmokeDualRevolversGameplay/Assets/Anim/CharacterAnim/CherrySmoke_DualRevolver_Dodge_LeftSide_FL_Start}
   Output Animation Pose(Result=@n1)
  ]
 }
 Left_FrontRight_Start {
  path=ItemFullBodyOverride.AnimGraphNode_StateMachine_0.Dodge SM.AnimStateNode_1.Start Roll.AnimGraphNode_StateMachine_0.Dodge Start SM.AnimStateNode_0.Left.AnimGraphNode_BlendSpaceGraph_0.Left Dodge Start.Left_FrontRight_Start
  nodes [
   n1 = Sequence Player 'CherrySmoke_DualRevolver_Dodge_LeftSide_FR_Start'() {bLoopAnimation=false, sequence=/CherrySmokeDualRevolversGameplay/Assets/Anim/CharacterAnim/CherrySmoke_DualRevolver_Dodge_LeftSide_FR_Start}
   Output Animation Pose(Result=@n1)
  ]
 }
 RightDodgeStart {
  path=ItemFullBodyOverride.AnimGraphNode_StateMachine_0.Dodge SM.AnimStateNode_1.Start Roll.AnimGraphNode_StateMachine_0.Dodge Start SM.AnimStateNode_0.Left.AnimGraphNode_BlendSpaceGraph_0.Left Dodge Start.RightDodgeStart
  nodes [
   n1 = Sequence Player 'CherrySmoke_DualRevolver_Dodge_R_Start_anim'() {
    bLoopAnimation=false
    bOverridePositionWhenJoiningSyncGroupAsLeader=true
    sequence=/CherrySmokeDualRevolversGameplay/Assets/Anim/TEMP/CherrySmoke_DualRevolver_Dodge_R_Start_anim
   }
   Output Animation Pose(Result=@n1)
  ]
 }
 Right {
  path=ItemFullBodyOverride.AnimGraphNode_StateMachine_0.Dodge SM.AnimStateNode_1.Start Roll.AnimGraphNode_StateMachine_0.Dodge Start SM.AnimStateNode_1.Right
  nodes [
   AnimGraphNode_BlendSpaceGraph_0 = Blendspace 'Right Dodge Start'!(X:float=Get DodgeAngleRadians()) -> (Pose:Pose Link Structure) {is=title AnimGraphNode_BlendSpaceGraph, groupName=RightDodgeStart}
   Output Animation Pose(Result=@AnimGraphNode_BlendSpaceGraph_0) {is=title AnimGraphNode_StateResult}
  ]
 }
 Right Dodge Start {
  path=ItemFullBodyOverride.AnimGraphNode_StateMachine_0.Dodge SM.AnimStateNode_1.Start Roll.AnimGraphNode_StateMachine_0.Dodge Start SM.AnimStateNode_1.Right.AnimGraphNode_BlendSpaceGraph_0.Right Dodge Start
  links []
  nodes []
 }
 BackwardDodgeStart2 {
  path=ItemFullBodyOverride.AnimGraphNode_StateMachine_0.Dodge SM.AnimStateNode_1.Start Roll.AnimGraphNode_StateMachine_0.Dodge Start SM.AnimStateNode_1.Right.AnimGraphNode_BlendSpaceGraph_0.Right Dodge Start.BackwardDodgeStart2
  nodes [
   n1 = Sequence Player 'CherrySmoke_DualRevolver_Dodge_B_Start_anim'() {
    bLoopAnimation=false
    bOverridePositionWhenJoiningSyncGroupAsLeader=true
    sequence=&CherrySmoke_DualRevolver_Dodge_B_Start_anim
   }
   Output Animation Pose(Result=@n1) {is=title AnimGraphNode_BlendSpaceSampleResult}
  ]
 }
 BackwardsDodgeStart {
  path=ItemFullBodyOverride.AnimGraphNode_StateMachine_0.Dodge SM.AnimStateNode_1.Start Roll.AnimGraphNode_StateMachine_0.Dodge Start SM.AnimStateNode_1.Right.AnimGraphNode_BlendSpaceGraph_0.Right Dodge Start.BackwardsDodgeStart
  nodes [
   n1 = Sequence Player 'CherrySmoke_DualRevolver_Dodge_B_Start_anim'() {
    bLoopAnimation=false
    bOverridePositionWhenJoiningSyncGroupAsLeader=true
    sequence=&CherrySmoke_DualRevolver_Dodge_B_Start_anim
   }
   Output Animation Pose(Result=@n1)
  ]
 }
 ForwardDodgeStart {
  path=ItemFullBodyOverride.AnimGraphNode_StateMachine_0.Dodge SM.AnimStateNode_1.Start Roll.AnimGraphNode_StateMachine_0.Dodge Start SM.AnimStateNode_1.Right.AnimGraphNode_BlendSpaceGraph_0.Right Dodge Start.ForwardDodgeStart
  nodes [
   n1 = Sequence Player 'CherrySmoke_DualRevolver_Dodge_F_Start_anim'() {
    bLoopAnimation=false
    bOverridePositionWhenJoiningSyncGroupAsLeader=true
    sequence=&CherrySmoke_DualRevolver_Dodge_F_Start_anim
   }
   Output Animation Pose(Result=@n1)
  ]
 }
 LeftDodgeStart {
  path=ItemFullBodyOverride.AnimGraphNode_StateMachine_0.Dodge SM.AnimStateNode_1.Start Roll.AnimGraphNode_StateMachine_0.Dodge Start SM.AnimStateNode_1.Right.AnimGraphNode_BlendSpaceGraph_0.Right Dodge Start.LeftDodgeStart
  nodes [Output Animation Pose(Result=Sequence Player 'CherrySmoke_DualRevolver_Dodge_L_Start_anim'() {bLoopAnimation=false, sequence=&CherrySmoke_DualRevolver_Dodge_L_Start_anim})]
 }
 Left_BackLeft_Start {
  path=ItemFullBodyOverride.AnimGraphNode_StateMachine_0.Dodge SM.AnimStateNode_1.Start Roll.AnimGraphNode_StateMachine_0.Dodge Start SM.AnimStateNode_1.Right.AnimGraphNode_BlendSpaceGraph_0.Right Dodge Start.Left_BackLeft_Start
  nodes [
   n1 = Sequence Player 'CherrySmoke_DualRevolver_Dodge_LeftSide_BL_Start'() {bLoopAnimation=false, sequence=&CherrySmoke_DualRevolver_Dodge_LeftSide_BL_Start}
   Output Animation Pose(Result=@n1)
  ]
 }
 RightDodgeStart {
  path=ItemFullBodyOverride.AnimGraphNode_StateMachine_0.Dodge SM.AnimStateNode_1.Start Roll.AnimGraphNode_StateMachine_0.Dodge Start SM.AnimStateNode_1.Right.AnimGraphNode_BlendSpaceGraph_0.Right Dodge Start.RightDodgeStart
  nodes [
   n1 = Sequence Player 'CherrySmoke_DualRevolver_Dodge_R_Start_anim'() {
    bLoopAnimation=false
    bOverridePositionWhenJoiningSyncGroupAsLeader=true
    sequence=&CherrySmoke_DualRevolver_Dodge_R_Start_anim
   }
   Output Animation Pose(Result=@n1)
  ]
 }
 Right_BackRight_Start {
  path=ItemFullBodyOverride.AnimGraphNode_StateMachine_0.Dodge SM.AnimStateNode_1.Start Roll.AnimGraphNode_StateMachine_0.Dodge Start SM.AnimStateNode_1.Right.AnimGraphNode_BlendSpaceGraph_0.Right Dodge Start.Right_BackRight_Start
  nodes [
   n1 = Sequence Player 'CherrySmoke_DualRevolver_Dodge_RightSide_BR_Start'() {bLoopAnimation=false, sequence=/CherrySmokeDualRevolversGameplay/Assets/Anim/CharacterAnim/CherrySmoke_DualRevolver_Dodge_RightSide_BR_Start}
   Output Animation Pose(Result=@n1)
  ]
 }
 Right_FrontLeft_Start {
  path=ItemFullBodyOverride.AnimGraphNode_StateMachine_0.Dodge SM.AnimStateNode_1.Start Roll.AnimGraphNode_StateMachine_0.Dodge Start SM.AnimStateNode_1.Right.AnimGraphNode_BlendSpaceGraph_0.Right Dodge Start.Right_FrontLeft_Start
  nodes [
   n1 = Sequence Player 'CherrySmoke_DualRevolver_Dodge_LeftSide_FL_Start'() {bLoopAnimation=false, sequence=&CherrySmoke_DualRevolver_Dodge_LeftSide_FL_Start}
   Output Animation Pose(Result=@n1)
  ]
 }
 Right_FrontRight_Start {
  path=ItemFullBodyOverride.AnimGraphNode_StateMachine_0.Dodge SM.AnimStateNode_1.Start Roll.AnimGraphNode_StateMachine_0.Dodge Start SM.AnimStateNode_1.Right.AnimGraphNode_BlendSpaceGraph_0.Right Dodge Start.Right_FrontRight_Start
  nodes [
   n1 = Sequence Player 'CherrySmoke_DualRevolver_Dodge_RightSide_FR_Start'() {bLoopAnimation=false, sequence=/CherrySmokeDualRevolversGameplay/Assets/Anim/CharacterAnim/CherrySmoke_DualRevolver_Dodge_RightSide_FR_Start}
   Output Animation Pose(Result=@n1)
  ]
 }
 AnimationTransitionGraph_1 {
  path=ItemFullBodyOverride.AnimGraphNode_StateMachine_0.Dodge SM.AnimStateNode_1.Start Roll.AnimGraphNode_StateMachine_0.Dodge Start SM.AnimStateTransitionNode_2.AnimationTransitionGraph_1
  nodes [Result(bCanEnterTransition=Get bIsLeftDodgeSpace() -> (bIsLeftDodgeSpace:bool))]
 }
 AnimationTransitionGraph_1 {
  path=ItemFullBodyOverride.AnimGraphNode_StateMachine_0.Dodge SM.AnimStateNode_1.Start Roll.AnimGraphNode_StateMachine_0.Dodge Start SM.AnimStateTransitionNode_3.AnimationTransitionGraph_1
  nodes [Result(bCanEnterTransition=NOT Boolean(A=Get bIsLeftDodgeSpace()))]
 }
 Neutral Roll {
  path=ItemFullBodyOverride.AnimGraphNode_StateMachine_0.Dodge SM.AnimStateNode_2.Neutral Roll
  nodes [Output Animation Pose(Result=Sequence Player 'CherrySmoke_DualRevolver_DoubleJump_Start'() {bLoopAnimation=false, sequence=/CherrySmokeDualRevolversGameplay/Assets/Anim/TEMP/Dodge/CherrySmoke_DualRevolver_DoubleJump_Start}) {is=title AnimGraphNode_StateResult}]
 }
 Recovery Roll {
  path=ItemFullBodyOverride.AnimGraphNode_StateMachine_0.Dodge SM.AnimStateNode_3.Recovery Roll
  nodes [
   AnimGraphNode_BlendSpaceGraph_0 = Blendspace 'Roll'!(X:float=Get LastDodgeDirection() -> (LastDodgeDirection:float)) -> (Pose:Pose Link Structure) {is=title AnimGraphNode_BlendSpaceGraph}
   Output Animation Pose(Result=@AnimGraphNode_BlendSpaceGraph_0)
  ]
 }
 Roll {
  path=ItemFullBodyOverride.AnimGraphNode_StateMachine_0.Dodge SM.AnimStateNode_3.Recovery Roll.AnimGraphNode_BlendSpaceGraph_0.Roll
  links []
  nodes []
 }
 BackwardsRoll_L {
  path=ItemFullBodyOverride.AnimGraphNode_StateMachine_0.Dodge SM.AnimStateNode_3.Recovery Roll.AnimGraphNode_BlendSpaceGraph_0.Roll.BackwardsRoll_L
  nodes [
   n1 = Sequence Player 'CherrySmoke_DualRevolver_Dodge_B_Exit_anim'() {bLoopAnimation=false, playRate=0.8, sequence=/CherrySmokeDualRevolversGameplay/Assets/Anim/TEMP/CherrySmoke_DualRevolver_Dodge_B_Exit_anim}
   Output Animation Pose(Result=@n1) {is=title AnimGraphNode_BlendSpaceSampleResult}
  ]
 }
 BackwardsRoll_R {
  path=ItemFullBodyOverride.AnimGraphNode_StateMachine_0.Dodge SM.AnimStateNode_3.Recovery Roll.AnimGraphNode_BlendSpaceGraph_0.Roll.BackwardsRoll_R
  nodes [
   n1 = Sequence Player 'CherrySmoke_DualRevolver_Dodge_B_Exit_anim'() {bLoopAnimation=false, playRate=0.8, sequence=&CherrySmoke_DualRevolver_Dodge_B_Exit_anim}
   Output Animation Pose(Result=@n1)
  ]
 }
 ForwardRoll {
  path=ItemFullBodyOverride.AnimGraphNode_StateMachine_0.Dodge SM.AnimStateNode_3.Recovery Roll.AnimGraphNode_BlendSpaceGraph_0.Roll.ForwardRoll
  nodes [
   n1 = Sequence Player 'CherrySmoke_DualRevolver_Dodge_F_Exit_anim'() {bLoopAnimation=false, playRate=0.8, sequence=/CherrySmokeDualRevolversGameplay/Assets/Anim/TEMP/CherrySmoke_DualRevolver_Dodge_F_Exit_anim}
   Output Animation Pose(Result=@n1)
  ]
 }
 LeftRoll {
  path=ItemFullBodyOverride.AnimGraphNode_StateMachine_0.Dodge SM.AnimStateNode_3.Recovery Roll.AnimGraphNode_BlendSpaceGraph_0.Roll.LeftRoll
  nodes [
   n1 = Sequence Player 'CherrySmoke_DualRevolver_Dodge_L_Exit_anim'() {bLoopAnimation=false, playRate=0.8, sequence=/CherrySmokeDualRevolversGameplay/Assets/Anim/TEMP/CherrySmoke_DualRevolver_Dodge_L_Exit_anim}
   Output Animation Pose(Result=@n1)
  ]
 }
 RightRoll {
  path=ItemFullBodyOverride.AnimGraphNode_StateMachine_0.Dodge SM.AnimStateNode_3.Recovery Roll.AnimGraphNode_BlendSpaceGraph_0.Roll.RightRoll
  nodes [
   n1 = Sequence Player 'CherrySmoke_DualRevolver_Dodge_R_Exit_anim'() {bLoopAnimation=false, playRate=0.8, sequence=/CherrySmokeDualRevolversGameplay/Assets/Anim/TEMP/CherrySmoke_DualRevolver_Dodge_R_Exit_anim}
   Output Animation Pose(Result=@n1)
  ]
 }
 Neutral Recovery {
  path=ItemFullBodyOverride.AnimGraphNode_StateMachine_0.Dodge SM.AnimStateNode_4.Neutral Recovery
  nodes [
   n1 = Sequence Player 'CherrySmoke_DualRevolver_DoubleJump_Targeted_Exit'() {bLoopAnimation=false, sequence=/CherrySmokeDualRevolversGameplay/Assets/Anim/TEMP/Dodge/CherrySmoke_DualRevolver_DoubleJump_Targeted_Exit}
   n2 = Sequence Player 'CherrySmoke_DualRevolver_DoubleJump_NonTargeted_Exit'() {bLoopAnimation=false, sequence=/CherrySmokeDualRevolversGameplay/Assets/Anim/TEMP/Dodge/CherrySmoke_DualRevolver_DoubleJump_NonTargeted_Exit}
   n3 = Blend Poses by bool(BlendPose_0=@n1, BlendPose_1=@n2, bActiveValue=Get bIsHangtimeActive() -> (bIsHangtimeActive:bool)) {blendType=Linear}
   n4 = Get AimYaw() -> (AimYaw:float)
   n5 = Get AimPitch()
   AnimGraphNode_StateMachine_0 = Forward Firing() {is=title AnimGraphNode_StateMachine}
   Output Animation Pose(Result=Apply Additive(
     Base=AimOffset Player 'AO_DualRevolver_Targeted'!(X:float=@n4, Y:float=@n5, BasePose:Pose Link Structure=@n3, Alpha:float=1) -> (Pose:Pose Link Structure) {blendSpace=/CherrySmokeDualRevolversGameplay/Assets/Anim/Poses/AO_DualRevolver_Targeted, alphaBoolBlend {blendOption=Linear}}
     Additive=@AnimGraphNode_StateMachine_0
    ) {alphaBoolBlend {blendOption=Linear}}) {is=title AnimGraphNode_StateResult}
  ]
 }
 Forward Firing {
  entry=Empty
  path=ItemFullBodyOverride.AnimGraphNode_StateMachine_0.Dodge SM.AnimStateNode_4.Neutral Recovery.AnimGraphNode_StateMachine_0.Forward Firing
  ids=id
  nodes [
   AnimStateConduitNode_0 = Is Firing() {is=title AnimStateConduitNode}
   AnimStateNode_0 = Left Fire() {is=title AnimStateNode, bAlwaysResetOnEntry=true}
   AnimStateNode_1 = Empty()
   AnimStateNode_2 = Right Fire() {is=title AnimStateNode, bAlwaysResetOnEntry=true}
   AnimStateTransitionNode_0 = AnimStateTransitionNode()
   AnimStateTransitionNode_1 = AnimStateTransitionNode() {crossfadeDuration=0.05}
   AnimStateTransitionNode_2 = AnimStateTransitionNode() {crossfadeDuration=0.05}
   AnimStateTransitionNode_3 = AnimStateTransitionNode() {bAutomaticRuleBasedOnSequencePlayerInState=true}
   AnimStateTransitionNode_4 = AnimStateTransitionNode() {bAutomaticRuleBasedOnSequencePlayerInState=true}
  ]
  links [
   {
    dst=Is Firing
    rule=ItemFullBodyOverride.AnimGraphNode_StateMachine_0.Dodge SM.AnimStateNode_4.Neutral Recovery.AnimGraphNode_StateMachine_0.Forward Firing.AnimStateTransitionNode_0.Transition
    src=Empty
    via=AnimStateTransitionNode_0
   }
   {
    dst=Left Fire
    rule=ItemFullBodyOverride.AnimGraphNode_StateMachine_0.Dodge SM.AnimStateNode_4.Neutral Recovery.AnimGraphNode_StateMachine_0.Forward Firing.AnimStateTransitionNode_1.Transition
    src=Is Firing
    via=AnimStateTransitionNode_1
   }
   {
    dst=Right Fire
    rule=ItemFullBodyOverride.AnimGraphNode_StateMachine_0.Dodge SM.AnimStateNode_4.Neutral Recovery.AnimGraphNode_StateMachine_0.Forward Firing.AnimStateTransitionNode_2.Transition
    src=Is Firing
    via=AnimStateTransitionNode_2
   }
   {
    dst=Is Firing
    rule=ItemFullBodyOverride.AnimGraphNode_StateMachine_0.Dodge SM.AnimStateNode_4.Neutral Recovery.AnimGraphNode_StateMachine_0.Forward Firing.AnimStateTransitionNode_3.Transition
    src=Left Fire
    via=AnimStateTransitionNode_3
   }
   {
    dst=Is Firing
    rule=ItemFullBodyOverride.AnimGraphNode_StateMachine_0.Dodge SM.AnimStateNode_4.Neutral Recovery.AnimGraphNode_StateMachine_0.Forward Firing.AnimStateTransitionNode_4.Transition
    src=Right Fire
    via=AnimStateTransitionNode_4
   }
  ]
 }
 Is Firing {
  path=ItemFullBodyOverride.AnimGraphNode_StateMachine_0.Dodge SM.AnimStateNode_4.Neutral Recovery.AnimGraphNode_StateMachine_0.Forward Firing.AnimStateConduitNode_0.Is Firing
  nodes [Result(bCanEnterTransition=Property Access!() -> (Value:bool=~) {is=title K2Node_PropertyAccess})]
 }
 Left Fire {
  path=ItemFullBodyOverride.AnimGraphNode_StateMachine_0.Dodge SM.AnimStateNode_4.Neutral Recovery.AnimGraphNode_StateMachine_0.Forward Firing.AnimStateNode_0.Left Fire
  nodes [
   n1 = "Sequence Player (additive) 'CherrySmoke_DualRevolver_DoubleJump_Fire_Targeted_L'"!(PlayRate:float=Get AdditiveLoopPlayRate() -> (AdditiveLoopPlayRate:double)) -> (Pose) {bLoopAnimation=false, sequence=/CherrySmokeDualRevolversGameplay/Assets/Anim/TEMP/CherrySmoke_DualRevolver_DoubleJump_Fire_Targeted_L}
   Output Animation Pose(Result=@n1)
  ]
 }
 Empty {
  path=ItemFullBodyOverride.AnimGraphNode_StateMachine_0.Dodge SM.AnimStateNode_4.Neutral Recovery.AnimGraphNode_StateMachine_0.Forward Firing.AnimStateNode_1.Empty
  nodes [Output Animation Pose(Result=Additive Identity Pose())]
 }
 Right Fire {
  path=ItemFullBodyOverride.AnimGraphNode_StateMachine_0.Dodge SM.AnimStateNode_4.Neutral Recovery.AnimGraphNode_StateMachine_0.Forward Firing.AnimStateNode_2.Right Fire
  nodes [
   n1 = "Sequence Player (additive) 'CherrySmoke_DualRevolver_DoubleJump_Fire_Targeted_R'"!(PlayRate:float=Get AdditiveLoopPlayRate()) -> (Pose) {bLoopAnimation=false, sequence=/CherrySmokeDualRevolversGameplay/Assets/Anim/TEMP/CherrySmoke_DualRevolver_DoubleJump_Fire_Targeted_R}
   Output Animation Pose(Result=@n1)
  ]
 }
 Transition {
  path=ItemFullBodyOverride.AnimGraphNode_StateMachine_0.Dodge SM.AnimStateNode_4.Neutral Recovery.AnimGraphNode_StateMachine_0.Forward Firing.AnimStateTransitionNode_0.Transition
  nodes [Result(bCanEnterTransition=true)]
 }
 Transition {
  path=ItemFullBodyOverride.AnimGraphNode_StateMachine_0.Dodge SM.AnimStateNode_4.Neutral Recovery.AnimGraphNode_StateMachine_0.Forward Firing.AnimStateTransitionNode_1.Transition
  nodes [Result(bCanEnterTransition=NOT Boolean(A=Get bIsLastHandFireLeft() -> (bIsLastHandFireLeft:bool)))]
 }
 Transition {
  path=ItemFullBodyOverride.AnimGraphNode_StateMachine_0.Dodge SM.AnimStateNode_4.Neutral Recovery.AnimGraphNode_StateMachine_0.Forward Firing.AnimStateTransitionNode_2.Transition
  nodes [Result(bCanEnterTransition=Get bIsLastHandFireLeft())]
 }
 Transition {
  path=ItemFullBodyOverride.AnimGraphNode_StateMachine_0.Dodge SM.AnimStateNode_4.Neutral Recovery.AnimGraphNode_StateMachine_0.Forward Firing.AnimStateTransitionNode_3.Transition
  nodes [Result()]
 }
 Transition {
  path=ItemFullBodyOverride.AnimGraphNode_StateMachine_0.Dodge SM.AnimStateNode_4.Neutral Recovery.AnimGraphNode_StateMachine_0.Forward Firing.AnimStateTransitionNode_4.Transition
  nodes [Result()]
 }
 In Air {
  path=ItemFullBodyOverride.AnimGraphNode_StateMachine_0.Dodge SM.AnimStateNode_5.In Air
  nodes [
   AnimGraphNode_StateMachine_0 = Dodge Directional Loop() {is=title AnimGraphNode_StateMachine, bAllowConduitEntryStates=true}
   AnimGraphNode_StateMachine_1 = Dodge Directional() {is=title AnimGraphNode_StateMachine, bAllowConduitEntryStates=true}
   n1 = Sequence Player 'CherrySmoke_DualRevolver_Dodge_F_Diving_anim'() {
    node {
     bLoopAnimation=false
     bOverridePositionWhenJoiningSyncGroupAsLeader=true
     sequence=/CherrySmokeDualRevolversGameplay/Assets/Anim/TEMP/CherrySmoke_DualRevolver_Dodge_F_Diving_anim
     updateFunction {functionName=OnUpdate_0}
    }
    updateFunction {bSelfContext=true, memberGuid=#40, memberName=OnInAirRefSequenceUpdate}
   }
   Output Animation Pose(Result=Two Way Blend(
     A=Apply Additive(Base=@AnimGraphNode_StateMachine_1, Additive=@AnimGraphNode_StateMachine_0) {alphaBoolBlend {blendOption=Linear}}
     B=@n1
     Alpha=0
    ) {blendNode {bAlwaysUpdateChildren=true, alphaBoolBlend {blendOption=Linear}}}) {
    node {stateExitFunction {functionName=OnStateExit}, updateFunction {functionName=OnUpdate_0}}
    stateExitFunction {bSelfContext=true, memberGuid=#41, memberName=OnInAirStateExit}
    updateFunction {bSelfContext=true, memberGuid=#42, memberName=OnInAirUpdate}
   }
  ]
 }
 Dodge Directional Loop {
  entry=Is Left
  path=ItemFullBodyOverride.AnimGraphNode_StateMachine_0.Dodge SM.AnimStateNode_5.In Air.AnimGraphNode_StateMachine_0.Dodge Directional Loop
  ids=id
  nodes [
   AnimStateConduitNode_0 = Is Left()
   AnimStateNode_0 = Left() {
    stateEntered {guid=#28}
    stateFullyBlended {guid=#29}
    stateLeft {guid=#30}
   }
   AnimStateNode_1 = Right() {
    stateEntered {guid=#31}
    stateFullyBlended {guid=#32}
    stateLeft {guid=#33}
   }
   AnimStateTransitionNode_2 = AnimStateTransitionNode() {
    transitionEnd {guid=#34}
    transitionInterrupt {guid=#35}
    transitionStart {guid=#36}
   }
   AnimStateTransitionNode_3 = AnimStateTransitionNode() {
    transitionEnd {guid=#37}
    transitionInterrupt {guid=#38}
    transitionStart {guid=#39}
   }
  ]
  links [
   {
    dst=Left
    rule=ItemFullBodyOverride.AnimGraphNode_StateMachine_0.Dodge SM.AnimStateNode_5.In Air.AnimGraphNode_StateMachine_0.Dodge Directional Loop.AnimStateTransitionNode_2.Transition
    src=Is Left
    via=AnimStateTransitionNode_2
   }
   {
    dst=Right
    rule=ItemFullBodyOverride.AnimGraphNode_StateMachine_0.Dodge SM.AnimStateNode_5.In Air.AnimGraphNode_StateMachine_0.Dodge Directional Loop.AnimStateTransitionNode_3.Transition
    src=Is Left
    via=AnimStateTransitionNode_3
   }
  ]
 }
 Is Left {
  path=ItemFullBodyOverride.AnimGraphNode_StateMachine_0.Dodge SM.AnimStateNode_5.In Air.AnimGraphNode_StateMachine_0.Dodge Directional Loop.AnimStateConduitNode_0.Is Left
  nodes [Result(bCanEnterTransition=true)]
 }
 Left {
  path=ItemFullBodyOverride.AnimGraphNode_StateMachine_0.Dodge SM.AnimStateNode_5.In Air.AnimGraphNode_StateMachine_0.Dodge Directional Loop.AnimStateNode_0.Left
  nodes [
   AnimGraphNode_BlendSpaceGraph_1 = "Blendspace 'Left  Dodge Loop Additive'"!(X:float=Get DodgeAngleRadians()) -> (Pose:Pose Link Structure) {is=title AnimGraphNode_BlendSpaceGraph, groupName=DodgeLoop}
   Output Animation Pose(Result=@AnimGraphNode_BlendSpaceGraph_1)
  ]
 }
 "Left  Dodge Loop Additive" {
  path="ItemFullBodyOverride.AnimGraphNode_StateMachine_0.Dodge SM.AnimStateNode_5.In Air.AnimGraphNode_StateMachine_0.Dodge Directional Loop.AnimStateNode_0.Left.AnimGraphNode_BlendSpaceGraph_1.Left  Dodge Loop Additive"
  links []
  nodes []
 }
 BackwardDodgeLoop_L {
  path="ItemFullBodyOverride.AnimGraphNode_StateMachine_0.Dodge SM.AnimStateNode_5.In Air.AnimGraphNode_StateMachine_0.Dodge Directional Loop.AnimStateNode_0.Left.AnimGraphNode_BlendSpaceGraph_1.Left  Dodge Loop Additive.BackwardDodgeLoop_L"
  nodes [Output Animation Pose(Result="Sequence Player (additive) 'CherrySmoke_DualRevolver_Dodge_LeftSide_B_Loop'"!(PlayRate:float=1) -> (Pose) {sequence=/CherrySmokeDualRevolversGameplay/Assets/Anim/CharacterAnim/CherrySmoke_DualRevolver_Dodge_LeftSide_B_Loop}) {is=title AnimGraphNode_BlendSpaceSampleResult}]
 }
 BackwardsDodgeLoop_R {
  path="ItemFullBodyOverride.AnimGraphNode_StateMachine_0.Dodge SM.AnimStateNode_5.In Air.AnimGraphNode_StateMachine_0.Dodge Directional Loop.AnimStateNode_0.Left.AnimGraphNode_BlendSpaceGraph_1.Left  Dodge Loop Additive.BackwardsDodgeLoop_R"
  nodes [Output Animation Pose(Result="Sequence Player (additive) 'CherrySmoke_DualRevolver_Dodge_LeftSide_B_Loop'"(PlayRate=1) {sequence=&CherrySmoke_DualRevolver_Dodge_LeftSide_B_Loop})]
 }
 ForwardDodgeLoop {
  path="ItemFullBodyOverride.AnimGraphNode_StateMachine_0.Dodge SM.AnimStateNode_5.In Air.AnimGraphNode_StateMachine_0.Dodge Directional Loop.AnimStateNode_0.Left.AnimGraphNode_BlendSpaceGraph_1.Left  Dodge Loop Additive.ForwardDodgeLoop"
  nodes [Output Animation Pose(Result="Sequence Player (additive) 'CherrySmoke_DualRevolver_Dodge_LeftSide_F_Loop'"!(PlayRate:float=1) -> (Pose) {sequence=/CherrySmokeDualRevolversGameplay/Assets/Anim/CharacterAnim/CherrySmoke_DualRevolver_Dodge_LeftSide_F_Loop})]
 }
 LeftDodgeLoop {
  path="ItemFullBodyOverride.AnimGraphNode_StateMachine_0.Dodge SM.AnimStateNode_5.In Air.AnimGraphNode_StateMachine_0.Dodge Directional Loop.AnimStateNode_0.Left.AnimGraphNode_BlendSpaceGraph_1.Left  Dodge Loop Additive.LeftDodgeLoop"
  nodes [Output Animation Pose(Result="Sequence Player (additive) 'CherrySmoke_DualRevolver_Dodge_LeftSide_L_Loop'"!(PlayRate:float=1) -> (Pose) {sequence=/CherrySmokeDualRevolversGameplay/Assets/Anim/CharacterAnim/CherrySmoke_DualRevolver_Dodge_LeftSide_L_Loop})]
 }
 Left_BackLeft_Loop {
  path="ItemFullBodyOverride.AnimGraphNode_StateMachine_0.Dodge SM.AnimStateNode_5.In Air.AnimGraphNode_StateMachine_0.Dodge Directional Loop.AnimStateNode_0.Left.AnimGraphNode_BlendSpaceGraph_1.Left  Dodge Loop Additive.Left_BackLeft_Loop"
  nodes [Output Animation Pose(Result="Sequence Player (additive) 'CherrySmoke_DualRevolver_Dodge_LeftSide_BL_Loop'"!(PlayRate:float=1) -> (Pose) {sequence=/CherrySmokeDualRevolversGameplay/Assets/Anim/CharacterAnim/CherrySmoke_DualRevolver_Dodge_LeftSide_BL_Loop})]
 }
 Left_BackRight_Loop {
  path="ItemFullBodyOverride.AnimGraphNode_StateMachine_0.Dodge SM.AnimStateNode_5.In Air.AnimGraphNode_StateMachine_0.Dodge Directional Loop.AnimStateNode_0.Left.AnimGraphNode_BlendSpaceGraph_1.Left  Dodge Loop Additive.Left_BackRight_Loop"
  nodes [Output Animation Pose(Result="Sequence Player (additive) 'CherrySmoke_DualRevolver_Dodge_LeftSide_BR_Loop'"!(PlayRate:float=1) -> (Pose) {sequence=/CherrySmokeDualRevolversGameplay/Assets/Anim/CharacterAnim/CherrySmoke_DualRevolver_Dodge_LeftSide_BR_Loop})]
 }
 Left_FrontLeft_Loop {
  path="ItemFullBodyOverride.AnimGraphNode_StateMachine_0.Dodge SM.AnimStateNode_5.In Air.AnimGraphNode_StateMachine_0.Dodge Directional Loop.AnimStateNode_0.Left.AnimGraphNode_BlendSpaceGraph_1.Left  Dodge Loop Additive.Left_FrontLeft_Loop"
  nodes [Output Animation Pose(Result="Sequence Player (additive) 'CherrySmoke_DualRevolver_Dodge_LeftSide_FL_Loop'"!(PlayRate:float=1) -> (Pose) {sequence=/CherrySmokeDualRevolversGameplay/Assets/Anim/CharacterAnim/CherrySmoke_DualRevolver_Dodge_LeftSide_FL_Loop})]
 }
 Left_FrontRight_Loop {
  path="ItemFullBodyOverride.AnimGraphNode_StateMachine_0.Dodge SM.AnimStateNode_5.In Air.AnimGraphNode_StateMachine_0.Dodge Directional Loop.AnimStateNode_0.Left.AnimGraphNode_BlendSpaceGraph_1.Left  Dodge Loop Additive.Left_FrontRight_Loop"
  nodes [Output Animation Pose(Result="Sequence Player (additive) 'CherrySmoke_DualRevolver_Dodge_LeftSide_FR_Loop'"!(PlayRate:float=1) -> (Pose) {sequence=/CherrySmokeDualRevolversGameplay/Assets/Anim/CharacterAnim/CherrySmoke_DualRevolver_Dodge_LeftSide_FR_Loop})]
 }
 RightDodgeLoop {
  path="ItemFullBodyOverride.AnimGraphNode_StateMachine_0.Dodge SM.AnimStateNode_5.In Air.AnimGraphNode_StateMachine_0.Dodge Directional Loop.AnimStateNode_0.Left.AnimGraphNode_BlendSpaceGraph_1.Left  Dodge Loop Additive.RightDodgeLoop"
  nodes [Output Animation Pose(Result="Sequence Player (additive) 'CherrySmoke_DualRevolver_Dodge_LeftSide_R_Loop'"!(PlayRate:float=1) -> (Pose) {sequence=/CherrySmokeDualRevolversGameplay/Assets/Anim/CharacterAnim/CherrySmoke_DualRevolver_Dodge_LeftSide_R_Loop})]
 }
 Right {
  path=ItemFullBodyOverride.AnimGraphNode_StateMachine_0.Dodge SM.AnimStateNode_5.In Air.AnimGraphNode_StateMachine_0.Dodge Directional Loop.AnimStateNode_1.Right
  nodes [
   AnimGraphNode_BlendSpaceGraph_1 = Blendspace 'Right Dodge Loop Additive'!(X:float=Get DodgeAngleRadians()) -> (Pose:Pose Link Structure) {is=title AnimGraphNode_BlendSpaceGraph}
   Output Animation Pose(Result=@AnimGraphNode_BlendSpaceGraph_1) {is=title AnimGraphNode_StateResult}
  ]
 }
 Right Dodge Loop Additive {
  path=ItemFullBodyOverride.AnimGraphNode_StateMachine_0.Dodge SM.AnimStateNode_5.In Air.AnimGraphNode_StateMachine_0.Dodge Directional Loop.AnimStateNode_1.Right.AnimGraphNode_BlendSpaceGraph_1.Right Dodge Loop Additive
  links []
  nodes []
 }
 Right_BackLeft_Loop {
  path=ItemFullBodyOverride.AnimGraphNode_StateMachine_0.Dodge SM.AnimStateNode_5.In Air.AnimGraphNode_StateMachine_0.Dodge Directional Loop.AnimStateNode_1.Right.AnimGraphNode_BlendSpaceGraph_1.Right Dodge Loop Additive.Right_BackLeft_Loop
  nodes [Output Animation Pose(Result="Sequence Player (additive) 'CherrySmoke_DualRevolver_Dodge_RightSide_BL_Loop'"!(PlayRate:float=1) -> (Pose) {sequence=/CherrySmokeDualRevolversGameplay/Assets/Anim/CharacterAnim/CherrySmoke_DualRevolver_Dodge_RightSide_BL_Loop}) {is=title AnimGraphNode_BlendSpaceSampleResult}]
 }
 Right_BackRight_Loop {
  path=ItemFullBodyOverride.AnimGraphNode_StateMachine_0.Dodge SM.AnimStateNode_5.In Air.AnimGraphNode_StateMachine_0.Dodge Directional Loop.AnimStateNode_1.Right.AnimGraphNode_BlendSpaceGraph_1.Right Dodge Loop Additive.Right_BackRight_Loop
  nodes [Output Animation Pose(Result="Sequence Player (additive) 'CherrySmoke_DualRevolver_Dodge_RightSide_BR_Loop'"!(PlayRate:float=1) -> (Pose) {sequence=/CherrySmokeDualRevolversGameplay/Assets/Anim/CharacterAnim/CherrySmoke_DualRevolver_Dodge_RightSide_BR_Loop})]
 }
 Right_Backward_Left_Loop {
  path=ItemFullBodyOverride.AnimGraphNode_StateMachine_0.Dodge SM.AnimStateNode_5.In Air.AnimGraphNode_StateMachine_0.Dodge Directional Loop.AnimStateNode_1.Right.AnimGraphNode_BlendSpaceGraph_1.Right Dodge Loop Additive.Right_Backward_Left_Loop
  nodes [Output Animation Pose(Result="Sequence Player (additive) 'CherrySmoke_DualRevolver_Dodge_RightSide_B_Loop'"!(PlayRate:float=1) -> (Pose) {sequence=/CherrySmokeDualRevolversGameplay/Assets/Anim/CharacterAnim/CherrySmoke_DualRevolver_Dodge_RightSide_B_Loop})]
 }
 Right_Backwards_Right_Loop {
  path=ItemFullBodyOverride.AnimGraphNode_StateMachine_0.Dodge SM.AnimStateNode_5.In Air.AnimGraphNode_StateMachine_0.Dodge Directional Loop.AnimStateNode_1.Right.AnimGraphNode_BlendSpaceGraph_1.Right Dodge Loop Additive.Right_Backwards_Right_Loop
  nodes [Output Animation Pose(Result="Sequence Player (additive) 'CherrySmoke_DualRevolver_Dodge_RightSide_B_Loop'"(PlayRate=1) {sequence=&CherrySmoke_DualRevolver_Dodge_RightSide_B_Loop})]
 }
 Right_Forward_Loop {
  path=ItemFullBodyOverride.AnimGraphNode_StateMachine_0.Dodge SM.AnimStateNode_5.In Air.AnimGraphNode_StateMachine_0.Dodge Directional Loop.AnimStateNode_1.Right.AnimGraphNode_BlendSpaceGraph_1.Right Dodge Loop Additive.Right_Forward_Loop
  nodes [Output Animation Pose(Result="Sequence Player (additive) 'CherrySmoke_DualRevolver_Dodge_RightSide_F_Loop'"!(PlayRate:float=1) -> (Pose) {sequence=/CherrySmokeDualRevolversGameplay/Assets/Anim/CharacterAnim/CherrySmoke_DualRevolver_Dodge_RightSide_F_Loop})]
 }
 Right_FrontLeft_Loop {
  path=ItemFullBodyOverride.AnimGraphNode_StateMachine_0.Dodge SM.AnimStateNode_5.In Air.AnimGraphNode_StateMachine_0.Dodge Directional Loop.AnimStateNode_1.Right.AnimGraphNode_BlendSpaceGraph_1.Right Dodge Loop Additive.Right_FrontLeft_Loop
  nodes [Output Animation Pose(Result="Sequence Player (additive) 'CherrySmoke_DualRevolver_Dodge_RightSide_FL_Loop'"!(PlayRate:float=1) -> (Pose) {sequence=/CherrySmokeDualRevolversGameplay/Assets/Anim/CharacterAnim/CherrySmoke_DualRevolver_Dodge_RightSide_FL_Loop})]
 }
 Right_FrontRight_Loop {
  path=ItemFullBodyOverride.AnimGraphNode_StateMachine_0.Dodge SM.AnimStateNode_5.In Air.AnimGraphNode_StateMachine_0.Dodge Directional Loop.AnimStateNode_1.Right.AnimGraphNode_BlendSpaceGraph_1.Right Dodge Loop Additive.Right_FrontRight_Loop
  nodes [Output Animation Pose(Result="Sequence Player (additive) 'CherrySmoke_DualRevolver_Dodge_RightSide_FR_Loop'"!(PlayRate:float=1) -> (Pose) {sequence=/CherrySmokeDualRevolversGameplay/Assets/Anim/CharacterAnim/CherrySmoke_DualRevolver_Dodge_RightSide_FR_Loop})]
 }
 Right_Left_Loop {
  path=ItemFullBodyOverride.AnimGraphNode_StateMachine_0.Dodge SM.AnimStateNode_5.In Air.AnimGraphNode_StateMachine_0.Dodge Directional Loop.AnimStateNode_1.Right.AnimGraphNode_BlendSpaceGraph_1.Right Dodge Loop Additive.Right_Left_Loop
  nodes [Output Animation Pose(Result="Sequence Player (additive) 'CherrySmoke_DualRevolver_Dodge_RightSide_L_Loop'"!(PlayRate:float=1) -> (Pose) {sequence=/CherrySmokeDualRevolversGameplay/Assets/Anim/CharacterAnim/CherrySmoke_DualRevolver_Dodge_RightSide_L_Loop})]
 }
 Right_Right_Loop {
  path=ItemFullBodyOverride.AnimGraphNode_StateMachine_0.Dodge SM.AnimStateNode_5.In Air.AnimGraphNode_StateMachine_0.Dodge Directional Loop.AnimStateNode_1.Right.AnimGraphNode_BlendSpaceGraph_1.Right Dodge Loop Additive.Right_Right_Loop
  nodes [Output Animation Pose(Result="Sequence Player (additive) 'CherrySmoke_DualRevolver_Dodge_RightSide_R_Loop'"!(PlayRate:float=1) -> (Pose) {sequence=/CherrySmokeDualRevolversGameplay/Assets/Anim/CharacterAnim/CherrySmoke_DualRevolver_Dodge_RightSide_R_Loop})]
 }
 Transition {
  path=ItemFullBodyOverride.AnimGraphNode_StateMachine_0.Dodge SM.AnimStateNode_5.In Air.AnimGraphNode_StateMachine_0.Dodge Directional Loop.AnimStateTransitionNode_2.Transition
  nodes [Result(bCanEnterTransition=Get bIsLeftDodgeSpace())]
 }
 Transition {
  path=ItemFullBodyOverride.AnimGraphNode_StateMachine_0.Dodge SM.AnimStateNode_5.In Air.AnimGraphNode_StateMachine_0.Dodge Directional Loop.AnimStateTransitionNode_3.Transition
  nodes [Result(bCanEnterTransition=NOT Boolean(A=Get bIsLeftDodgeSpace()))]
 }
 Dodge Directional {
  entry=Is Left
  path=ItemFullBodyOverride.AnimGraphNode_StateMachine_0.Dodge SM.AnimStateNode_5.In Air.AnimGraphNode_StateMachine_1.Dodge Directional
  ids=id
  nodes [
   AnimStateConduitNode_0 = Is Left()
   AnimStateNode_0 = Left() {
    stateEntered {guid=#28}
    stateFullyBlended {guid=#29}
    stateLeft {guid=#30}
   }
   AnimStateNode_1 = Right() {
    stateEntered {guid=#31}
    stateFullyBlended {guid=#32}
    stateLeft {guid=#33}
   }
   AnimStateTransitionNode_2 = AnimStateTransitionNode() {
    transitionEnd {guid=#34}
    transitionInterrupt {guid=#35}
    transitionStart {guid=#36}
   }
   AnimStateTransitionNode_3 = AnimStateTransitionNode() {
    transitionEnd {guid=#37}
    transitionInterrupt {guid=#38}
    transitionStart {guid=#39}
   }
  ]
  links [
   {
    dst=Left
    rule=ItemFullBodyOverride.AnimGraphNode_StateMachine_0.Dodge SM.AnimStateNode_5.In Air.AnimGraphNode_StateMachine_1.Dodge Directional.AnimStateTransitionNode_2.Transition
    src=Is Left
    via=AnimStateTransitionNode_2
   }
   {
    dst=Right
    rule=ItemFullBodyOverride.AnimGraphNode_StateMachine_0.Dodge SM.AnimStateNode_5.In Air.AnimGraphNode_StateMachine_1.Dodge Directional.AnimStateTransitionNode_3.Transition
    src=Is Left
    via=AnimStateTransitionNode_3
   }
  ]
 }
 Is Left {
  path=ItemFullBodyOverride.AnimGraphNode_StateMachine_0.Dodge SM.AnimStateNode_5.In Air.AnimGraphNode_StateMachine_1.Dodge Directional.AnimStateConduitNode_0.Is Left
  nodes [Result(bCanEnterTransition=true)]
 }
 Left {
  path=ItemFullBodyOverride.AnimGraphNode_StateMachine_0.Dodge SM.AnimStateNode_5.In Air.AnimGraphNode_StateMachine_1.Dodge Directional.AnimStateNode_0.Left
  nodes [
   AnimGraphNode_BlendSpaceGraph_3 = Blendspace 'Dodge Left'!(X:float=Get DodgeAngleRadians()) -> (Pose:Pose Link Structure) {is=title AnimGraphNode_BlendSpaceGraph, tag=DodgeLoop, groupName=DodgeLoop}
   Output Animation Pose(Result=@AnimGraphNode_BlendSpaceGraph_3) {is=title AnimGraphNode_StateResult}
  ]
 }
 Dodge Left {
  path=ItemFullBodyOverride.AnimGraphNode_StateMachine_0.Dodge SM.AnimStateNode_5.In Air.AnimGraphNode_StateMachine_1.Dodge Directional.AnimStateNode_0.Left.AnimGraphNode_BlendSpaceGraph_3.Dodge Left
  links []
  nodes []
 }
 Left_BackLeft_InAir {
  path=ItemFullBodyOverride.AnimGraphNode_StateMachine_0.Dodge SM.AnimStateNode_5.In Air.AnimGraphNode_StateMachine_1.Dodge Directional.AnimStateNode_0.Left.AnimGraphNode_BlendSpaceGraph_3.Dodge Left.Left_BackLeft_InAir
  nodes [
   n1 = Sequence Player 'CherrySmoke_DualRevolver_Dodge_LeftSide_BL_Diving'() {bLoopAnimation=false, sequence=/CherrySmokeDualRevolversGameplay/Assets/Anim/CharacterAnim/CherrySmoke_DualRevolver_Dodge_LeftSide_BL_Diving}
   AnimGraphNode_RotationOffsetBlendSpaceGraph_0 = AimOffset 'LeftSide_BackLeft_AimOffset'!(X:float=Get AimPitch(), BasePose:Pose Link Structure=@n1, Alpha:float=1) -> (Pose:Pose Link Structure) {is=title AnimGraphNode_RotationOffsetBlendSpaceGraph}
   AnimGraphNode_StateMachine_0 = Left Firing() {is=title AnimGraphNode_StateMachine}
   Output Animation Pose(Result=Apply Additive(Base=@AnimGraphNode_RotationOffsetBlendSpaceGraph_0, Additive=@AnimGraphNode_StateMachine_0) {alphaBoolBlend {blendOption=Linear}}) {is=title AnimGraphNode_BlendSpaceSampleResult}
  ]
 }
 LeftSide_BackLeft_AimOffset {
  path=ItemFullBodyOverride.AnimGraphNode_StateMachine_0.Dodge SM.AnimStateNode_5.In Air.AnimGraphNode_StateMachine_1.Dodge Directional.AnimStateNode_0.Left.AnimGraphNode_BlendSpaceGraph_3.Dodge Left.Left_BackLeft_InAir.AnimGraphNode_RotationOffsetBlendSpaceGraph_0.LeftSide_BackLeft_AimOffset
  links []
  nodes []
 }
 L_BL_Center {
  path=ItemFullBodyOverride.AnimGraphNode_StateMachine_0.Dodge SM.AnimStateNode_5.In Air.AnimGraphNode_StateMachine_1.Dodge Directional.AnimStateNode_0.Left.AnimGraphNode_BlendSpaceGraph_3.Dodge Left.Left_BackLeft_InAir.AnimGraphNode_RotationOffsetBlendSpaceGraph_0.LeftSide_BackLeft_AimOffset.L_BL_Center
  nodes [Output Animation Pose(Result=Additive Identity Pose())]
 }
 L_BL_Down {
  path=ItemFullBodyOverride.AnimGraphNode_StateMachine_0.Dodge SM.AnimStateNode_5.In Air.AnimGraphNode_StateMachine_1.Dodge Directional.AnimStateNode_0.Left.AnimGraphNode_BlendSpaceGraph_3.Dodge Left.Left_BackLeft_InAir.AnimGraphNode_RotationOffsetBlendSpaceGraph_0.LeftSide_BackLeft_AimOffset.L_BL_Down
  nodes [
   n1 = "Sequence Player (additive) 'CherrySmoke_DualRevolvers_Dodge_LeftSide_BL_AO_D50'"() {
    bLoopAnimation=false
    sequence=/CherrySmokeDualRevolversGameplay/Assets/Anim/Poses/Dodge_AO/CherrySmoke_DualRevolvers_Dodge_LeftSide_BL_AO_D50
   }
   Output Animation Pose(Result=@n1)
  ]
 }
 L_BL_UP {
  path=ItemFullBodyOverride.AnimGraphNode_StateMachine_0.Dodge SM.AnimStateNode_5.In Air.AnimGraphNode_StateMachine_1.Dodge Directional.AnimStateNode_0.Left.AnimGraphNode_BlendSpaceGraph_3.Dodge Left.Left_BackLeft_InAir.AnimGraphNode_RotationOffsetBlendSpaceGraph_0.LeftSide_BackLeft_AimOffset.L_BL_UP
  nodes [
   n1 = "Sequence Player (additive) 'CherrySmoke_DualRevolvers_Dodge_LeftSide_BL_AO_U50'"() {
    bLoopAnimation=false
    sequence=/CherrySmokeDualRevolversGameplay/Assets/Anim/Poses/Dodge_AO/CherrySmoke_DualRevolvers_Dodge_LeftSide_BL_AO_U50
   }
   Output Animation Pose(Result=@n1)
  ]
 }
 Left Firing {
  entry=Empty
  path=ItemFullBodyOverride.AnimGraphNode_StateMachine_0.Dodge SM.AnimStateNode_5.In Air.AnimGraphNode_StateMachine_1.Dodge Directional.AnimStateNode_0.Left.AnimGraphNode_BlendSpaceGraph_3.Dodge Left.Left_BackLeft_InAir.AnimGraphNode_StateMachine_0.Left Firing
  ids=id
  nodes [
   AnimStateConduitNode_0 = Is Firing()
   AnimStateNode_0 = Left Fire() {bAlwaysResetOnEntry=true}
   AnimStateNode_1 = Empty()
   AnimStateNode_2 = Right Fire() {bAlwaysResetOnEntry=true}
   AnimStateTransitionNode_0 = AnimStateTransitionNode()
   AnimStateTransitionNode_1 = AnimStateTransitionNode() {crossfadeDuration=0.05}
   AnimStateTransitionNode_14 = AnimStateTransitionNode() {bAutomaticRuleBasedOnSequencePlayerInState=true}
   AnimStateTransitionNode_15 = AnimStateTransitionNode() {bAutomaticRuleBasedOnSequencePlayerInState=true}
   AnimStateTransitionNode_2 = AnimStateTransitionNode() {crossfadeDuration=0.05}
  ]
  links [
   {
    dst=Is Firing
    rule=ItemFullBodyOverride.AnimGraphNode_StateMachine_0.Dodge SM.AnimStateNode_5.In Air.AnimGraphNode_StateMachine_1.Dodge Directional.AnimStateNode_0.Left.AnimGraphNode_BlendSpaceGraph_3.Dodge Left.Left_BackLeft_InAir.AnimGraphNode_StateMachine_0.Left Firing.AnimStateTransitionNode_0.Transition
    src=Empty
    via=AnimStateTransitionNode_0
   }
   {
    dst=Left Fire
    rule=ItemFullBodyOverride.AnimGraphNode_StateMachine_0.Dodge SM.AnimStateNode_5.In Air.AnimGraphNode_StateMachine_1.Dodge Directional.AnimStateNode_0.Left.AnimGraphNode_BlendSpaceGraph_3.Dodge Left.Left_BackLeft_InAir.AnimGraphNode_StateMachine_0.Left Firing.AnimStateTransitionNode_1.Transition
    src=Is Firing
    via=AnimStateTransitionNode_1
   }
   {
    dst=Right Fire
    rule=ItemFullBodyOverride.AnimGraphNode_StateMachine_0.Dodge SM.AnimStateNode_5.In Air.AnimGraphNode_StateMachine_1.Dodge Directional.AnimStateNode_0.Left.AnimGraphNode_BlendSpaceGraph_3.Dodge Left.Left_BackLeft_InAir.AnimGraphNode_StateMachine_0.Left Firing.AnimStateTransitionNode_2.Transition
    src=Is Firing
    via=AnimStateTransitionNode_2
   }
   {
    dst=Is Firing
    rule=ItemFullBodyOverride.AnimGraphNode_StateMachine_0.Dodge SM.AnimStateNode_5.In Air.AnimGraphNode_StateMachine_1.Dodge Directional.AnimStateNode_0.Left.AnimGraphNode_BlendSpaceGraph_3.Dodge Left.Left_BackLeft_InAir.AnimGraphNode_StateMachine_0.Left Firing.AnimStateTransitionNode_14.Transition
    src=Left Fire
    via=AnimStateTransitionNode_14
   }
   {
    dst=Is Firing
    rule=ItemFullBodyOverride.AnimGraphNode_StateMachine_0.Dodge SM.AnimStateNode_5.In Air.AnimGraphNode_StateMachine_1.Dodge Directional.AnimStateNode_0.Left.AnimGraphNode_BlendSpaceGraph_3.Dodge Left.Left_BackLeft_InAir.AnimGraphNode_StateMachine_0.Left Firing.AnimStateTransitionNode_15.Transition
    src=Right Fire
    via=AnimStateTransitionNode_15
   }
  ]
 }
 Is Firing {
  path=ItemFullBodyOverride.AnimGraphNode_StateMachine_0.Dodge SM.AnimStateNode_5.In Air.AnimGraphNode_StateMachine_1.Dodge Directional.AnimStateNode_0.Left.AnimGraphNode_BlendSpaceGraph_3.Dodge Left.Left_BackLeft_InAir.AnimGraphNode_StateMachine_0.Left Firing.AnimStateConduitNode_0.Is Firing
  nodes [Result(bCanEnterTransition=Property Access() -> (Value=~))]
 }
 Left Fire {
  path=ItemFullBodyOverride.AnimGraphNode_StateMachine_0.Dodge SM.AnimStateNode_5.In Air.AnimGraphNode_StateMachine_1.Dodge Directional.AnimStateNode_0.Left.AnimGraphNode_BlendSpaceGraph_3.Dodge Left.Left_BackLeft_InAir.AnimGraphNode_StateMachine_0.Left Firing.AnimStateNode_0.Left Fire
  nodes [
   n1 = "Sequence Player (additive) 'CherrySmoke_DualRevolvers_left_dodge_fire_L'"!(PlayRate:float=Get AdditiveLoopPlayRate()) -> (Pose) {bLoopAnimation=false, sequence=/CherrySmokeDualRevolversGameplay/Assets/Anim/CharacterAnim/CherrySmoke_DualRevolvers_left_dodge_fire_L}
   Output Animation Pose(Result=@n1) {is=title AnimGraphNode_StateResult}
  ]
 }
 Empty {
  path=ItemFullBodyOverride.AnimGraphNode_StateMachine_0.Dodge SM.AnimStateNode_5.In Air.AnimGraphNode_StateMachine_1.Dodge Directional.AnimStateNode_0.Left.AnimGraphNode_BlendSpaceGraph_3.Dodge Left.Left_BackLeft_InAir.AnimGraphNode_StateMachine_0.Left Firing.AnimStateNode_1.Empty
  nodes [Output Animation Pose(Result=Additive Identity Pose())]
 }
 Right Fire {
  path=ItemFullBodyOverride.AnimGraphNode_StateMachine_0.Dodge SM.AnimStateNode_5.In Air.AnimGraphNode_StateMachine_1.Dodge Directional.AnimStateNode_0.Left.AnimGraphNode_BlendSpaceGraph_3.Dodge Left.Left_BackLeft_InAir.AnimGraphNode_StateMachine_0.Left Firing.AnimStateNode_2.Right Fire
  nodes [
   n1 = "Sequence Player (additive) 'CherrySmoke_DualRevolvers_left_dodge_fire_R'"!(PlayRate:float=Get AdditiveLoopPlayRate()) -> (Pose) {bLoopAnimation=false, sequence=/CherrySmokeDualRevolversGameplay/Assets/Anim/CharacterAnim/CherrySmoke_DualRevolvers_left_dodge_fire_R}
   Output Animation Pose(Result=@n1)
  ]
 }
 Transition {
  path=ItemFullBodyOverride.AnimGraphNode_StateMachine_0.Dodge SM.AnimStateNode_5.In Air.AnimGraphNode_StateMachine_1.Dodge Directional.AnimStateNode_0.Left.AnimGraphNode_BlendSpaceGraph_3.Dodge Left.Left_BackLeft_InAir.AnimGraphNode_StateMachine_0.Left Firing.AnimStateTransitionNode_0.Transition
  nodes [Result(bCanEnterTransition=true)]
 }
 Transition {
  path=ItemFullBodyOverride.AnimGraphNode_StateMachine_0.Dodge SM.AnimStateNode_5.In Air.AnimGraphNode_StateMachine_1.Dodge Directional.AnimStateNode_0.Left.AnimGraphNode_BlendSpaceGraph_3.Dodge Left.Left_BackLeft_InAir.AnimGraphNode_StateMachine_0.Left Firing.AnimStateTransitionNode_1.Transition
  nodes [Result(bCanEnterTransition=NOT Boolean(A=Get bIsLastHandFireLeft()))]
 }
 Transition {
  path=ItemFullBodyOverride.AnimGraphNode_StateMachine_0.Dodge SM.AnimStateNode_5.In Air.AnimGraphNode_StateMachine_1.Dodge Directional.AnimStateNode_0.Left.AnimGraphNode_BlendSpaceGraph_3.Dodge Left.Left_BackLeft_InAir.AnimGraphNode_StateMachine_0.Left Firing.AnimStateTransitionNode_14.Transition
  nodes [Result()]
 }
 Transition {
  path=ItemFullBodyOverride.AnimGraphNode_StateMachine_0.Dodge SM.AnimStateNode_5.In Air.AnimGraphNode_StateMachine_1.Dodge Directional.AnimStateNode_0.Left.AnimGraphNode_BlendSpaceGraph_3.Dodge Left.Left_BackLeft_InAir.AnimGraphNode_StateMachine_0.Left Firing.AnimStateTransitionNode_15.Transition
  nodes [Result()]
 }
 Transition {
  path=ItemFullBodyOverride.AnimGraphNode_StateMachine_0.Dodge SM.AnimStateNode_5.In Air.AnimGraphNode_StateMachine_1.Dodge Directional.AnimStateNode_0.Left.AnimGraphNode_BlendSpaceGraph_3.Dodge Left.Left_BackLeft_InAir.AnimGraphNode_StateMachine_0.Left Firing.AnimStateTransitionNode_2.Transition
  nodes [Result(bCanEnterTransition=Get bIsLastHandFireLeft())]
 }
 Left_BackRight_InAir {
  path=ItemFullBodyOverride.AnimGraphNode_StateMachine_0.Dodge SM.AnimStateNode_5.In Air.AnimGraphNode_StateMachine_1.Dodge Directional.AnimStateNode_0.Left.AnimGraphNode_BlendSpaceGraph_3.Dodge Left.Left_BackRight_InAir
  nodes [
   n1 = Sequence Player 'CherrySmoke_DualRevolver_Dodge_LeftSide_BR_Diving'() {bLoopAnimation=false, sequence=/CherrySmokeDualRevolversGameplay/Assets/Anim/CharacterAnim/CherrySmoke_DualRevolver_Dodge_LeftSide_BR_Diving}
   AnimGraphNode_RotationOffsetBlendSpaceGraph_0 = AimOffset 'LeftSide_BackRight_AimOffset'!(X:float=Get AimPitch(), BasePose:Pose Link Structure=@n1, Alpha:float=1) -> (Pose:Pose Link Structure) {is=title AnimGraphNode_RotationOffsetBlendSpaceGraph}
   AnimGraphNode_StateMachine_0 = Right Firing() {is=title AnimGraphNode_StateMachine}
   Output Animation Pose(Result=Apply Additive(Base=@AnimGraphNode_RotationOffsetBlendSpaceGraph_0, Additive=@AnimGraphNode_StateMachine_0) {alphaBoolBlend {blendOption=Linear}}) {is=title AnimGraphNode_BlendSpaceSampleResult}
  ]
 }
 LeftSide_BackRight_AimOffset {
  path=ItemFullBodyOverride.AnimGraphNode_StateMachine_0.Dodge SM.AnimStateNode_5.In Air.AnimGraphNode_StateMachine_1.Dodge Directional.AnimStateNode_0.Left.AnimGraphNode_BlendSpaceGraph_3.Dodge Left.Left_BackRight_InAir.AnimGraphNode_RotationOffsetBlendSpaceGraph_0.LeftSide_BackRight_AimOffset
  links []
  nodes []
 }
 L_BR_Center {
  path=ItemFullBodyOverride.AnimGraphNode_StateMachine_0.Dodge SM.AnimStateNode_5.In Air.AnimGraphNode_StateMachine_1.Dodge Directional.AnimStateNode_0.Left.AnimGraphNode_BlendSpaceGraph_3.Dodge Left.Left_BackRight_InAir.AnimGraphNode_RotationOffsetBlendSpaceGraph_0.LeftSide_BackRight_AimOffset.L_BR_Center
  nodes [Output Animation Pose(Result=Additive Identity Pose())]
 }
 L_BR_Down {
  path=ItemFullBodyOverride.AnimGraphNode_StateMachine_0.Dodge SM.AnimStateNode_5.In Air.AnimGraphNode_StateMachine_1.Dodge Directional.AnimStateNode_0.Left.AnimGraphNode_BlendSpaceGraph_3.Dodge Left.Left_BackRight_InAir.AnimGraphNode_RotationOffsetBlendSpaceGraph_0.LeftSide_BackRight_AimOffset.L_BR_Down
  nodes [
   n1 = "Sequence Player (additive) 'CherrySmoke_DualRevolvers_Dodge_LeftSide_BR_AO_D50'"() {
    bLoopAnimation=false
    sequence=/CherrySmokeDualRevolversGameplay/Assets/Anim/Poses/Dodge_AO/CherrySmoke_DualRevolvers_Dodge_LeftSide_BR_AO_D50
   }
   Output Animation Pose(Result=@n1)
  ]
 }
 L_BR_UP {
  path=ItemFullBodyOverride.AnimGraphNode_StateMachine_0.Dodge SM.AnimStateNode_5.In Air.AnimGraphNode_StateMachine_1.Dodge Directional.AnimStateNode_0.Left.AnimGraphNode_BlendSpaceGraph_3.Dodge Left.Left_BackRight_InAir.AnimGraphNode_RotationOffsetBlendSpaceGraph_0.LeftSide_BackRight_AimOffset.L_BR_UP
  nodes [
   n1 = "Sequence Player (additive) 'CherrySmoke_DualRevolvers_Dodge_LeftSide_BR_AO_U50'"() {
    bLoopAnimation=false
    sequence=/CherrySmokeDualRevolversGameplay/Assets/Anim/Poses/Dodge_AO/CherrySmoke_DualRevolvers_Dodge_LeftSide_BR_AO_U50
   }
   Output Animation Pose(Result=@n1)
  ]
 }
 Right Firing {
  entry=Empty
  path=ItemFullBodyOverride.AnimGraphNode_StateMachine_0.Dodge SM.AnimStateNode_5.In Air.AnimGraphNode_StateMachine_1.Dodge Directional.AnimStateNode_0.Left.AnimGraphNode_BlendSpaceGraph_3.Dodge Left.Left_BackRight_InAir.AnimGraphNode_StateMachine_0.Right Firing
  ids=id
  nodes [
   AnimStateConduitNode_0 = Is Firing()
   AnimStateNode_0 = Left Fire() {bAlwaysResetOnEntry=true}
   AnimStateNode_1 = Empty()
   AnimStateNode_2 = Right Fire() {bAlwaysResetOnEntry=true}
   AnimStateTransitionNode_3 = AnimStateTransitionNode()
   AnimStateTransitionNode_4 = AnimStateTransitionNode()
   AnimStateTransitionNode_5 = AnimStateTransitionNode()
   AnimStateTransitionNode_6 = AnimStateTransitionNode() {bAutomaticRuleBasedOnSequencePlayerInState=true}
   AnimStateTransitionNode_7 = AnimStateTransitionNode() {bAutomaticRuleBasedOnSequencePlayerInState=true}
  ]
  links [
   {
    dst=Is Firing
    rule=ItemFullBodyOverride.AnimGraphNode_StateMachine_0.Dodge SM.AnimStateNode_5.In Air.AnimGraphNode_StateMachine_1.Dodge Directional.AnimStateNode_0.Left.AnimGraphNode_BlendSpaceGraph_3.Dodge Left.Left_BackRight_InAir.AnimGraphNode_StateMachine_0.Right Firing.AnimStateTransitionNode_3.Transition
    src=Empty
    via=AnimStateTransitionNode_3
   }
   {
    dst=Left Fire
    rule=ItemFullBodyOverride.AnimGraphNode_StateMachine_0.Dodge SM.AnimStateNode_5.In Air.AnimGraphNode_StateMachine_1.Dodge Directional.AnimStateNode_0.Left.AnimGraphNode_BlendSpaceGraph_3.Dodge Left.Left_BackRight_InAir.AnimGraphNode_StateMachine_0.Right Firing.AnimStateTransitionNode_4.Transition
    src=Is Firing
    via=AnimStateTransitionNode_4
   }
   {
    dst=Right Fire
    rule=ItemFullBodyOverride.AnimGraphNode_StateMachine_0.Dodge SM.AnimStateNode_5.In Air.AnimGraphNode_StateMachine_1.Dodge Directional.AnimStateNode_0.Left.AnimGraphNode_BlendSpaceGraph_3.Dodge Left.Left_BackRight_InAir.AnimGraphNode_StateMachine_0.Right Firing.AnimStateTransitionNode_5.Transition
    src=Is Firing
    via=AnimStateTransitionNode_5
   }
   {
    dst=Is Firing
    rule=ItemFullBodyOverride.AnimGraphNode_StateMachine_0.Dodge SM.AnimStateNode_5.In Air.AnimGraphNode_StateMachine_1.Dodge Directional.AnimStateNode_0.Left.AnimGraphNode_BlendSpaceGraph_3.Dodge Left.Left_BackRight_InAir.AnimGraphNode_StateMachine_0.Right Firing.AnimStateTransitionNode_6.Transition
    src=Left Fire
    via=AnimStateTransitionNode_6
   }
   {
    dst=Is Firing
    rule=ItemFullBodyOverride.AnimGraphNode_StateMachine_0.Dodge SM.AnimStateNode_5.In Air.AnimGraphNode_StateMachine_1.Dodge Directional.AnimStateNode_0.Left.AnimGraphNode_BlendSpaceGraph_3.Dodge Left.Left_BackRight_InAir.AnimGraphNode_StateMachine_0.Right Firing.AnimStateTransitionNode_7.Transition
    src=Right Fire
    via=AnimStateTransitionNode_7
   }
  ]
 }
 Is Firing {
  path=ItemFullBodyOverride.AnimGraphNode_StateMachine_0.Dodge SM.AnimStateNode_5.In Air.AnimGraphNode_StateMachine_1.Dodge Directional.AnimStateNode_0.Left.AnimGraphNode_BlendSpaceGraph_3.Dodge Left.Left_BackRight_InAir.AnimGraphNode_StateMachine_0.Right Firing.AnimStateConduitNode_0.Is Firing
  nodes [Result(bCanEnterTransition=Property Access() -> (Value=~))]
 }
 Left Fire {
  path=ItemFullBodyOverride.AnimGraphNode_StateMachine_0.Dodge SM.AnimStateNode_5.In Air.AnimGraphNode_StateMachine_1.Dodge Directional.AnimStateNode_0.Left.AnimGraphNode_BlendSpaceGraph_3.Dodge Left.Left_BackRight_InAir.AnimGraphNode_StateMachine_0.Right Firing.AnimStateNode_0.Left Fire
  nodes [
   n1 = "Sequence Player (additive) 'CherrySmoke_DualRevolvers_right_dodge_fire_L'"!(PlayRate:float=Get AdditiveLoopPlayRate()) -> (Pose) {bLoopAnimation=false, sequence=/CherrySmokeDualRevolversGameplay/Assets/Anim/CharacterAnim/CherrySmoke_DualRevolvers_right_dodge_fire_L}
   Output Animation Pose(Result=@n1) {is=title AnimGraphNode_StateResult}
  ]
 }
 Empty {
  path=ItemFullBodyOverride.AnimGraphNode_StateMachine_0.Dodge SM.AnimStateNode_5.In Air.AnimGraphNode_StateMachine_1.Dodge Directional.AnimStateNode_0.Left.AnimGraphNode_BlendSpaceGraph_3.Dodge Left.Left_BackRight_InAir.AnimGraphNode_StateMachine_0.Right Firing.AnimStateNode_1.Empty
  nodes [Output Animation Pose(Result=Additive Identity Pose())]
 }
 Right Fire {
  path=ItemFullBodyOverride.AnimGraphNode_StateMachine_0.Dodge SM.AnimStateNode_5.In Air.AnimGraphNode_StateMachine_1.Dodge Directional.AnimStateNode_0.Left.AnimGraphNode_BlendSpaceGraph_3.Dodge Left.Left_BackRight_InAir.AnimGraphNode_StateMachine_0.Right Firing.AnimStateNode_2.Right Fire
  nodes [
   n1 = "Sequence Player (additive) 'CherrySmoke_DualRevolvers_right_dodge_fire_R'"!(PlayRate:float=Get AdditiveLoopPlayRate()) -> (Pose) {bLoopAnimation=false, sequence=/CherrySmokeDualRevolversGameplay/Assets/Anim/CharacterAnim/CherrySmoke_DualRevolvers_right_dodge_fire_R}
   Output Animation Pose(Result=@n1)
  ]
 }
 Transition {
  path=ItemFullBodyOverride.AnimGraphNode_StateMachine_0.Dodge SM.AnimStateNode_5.In Air.AnimGraphNode_StateMachine_1.Dodge Directional.AnimStateNode_0.Left.AnimGraphNode_BlendSpaceGraph_3.Dodge Left.Left_BackRight_InAir.AnimGraphNode_StateMachine_0.Right Firing.AnimStateTransitionNode_3.Transition
  nodes [Result(bCanEnterTransition=true)]
 }
 Transition {
  path=ItemFullBodyOverride.AnimGraphNode_StateMachine_0.Dodge SM.AnimStateNode_5.In Air.AnimGraphNode_StateMachine_1.Dodge Directional.AnimStateNode_0.Left.AnimGraphNode_BlendSpaceGraph_3.Dodge Left.Left_BackRight_InAir.AnimGraphNode_StateMachine_0.Right Firing.AnimStateTransitionNode_4.Transition
  nodes [Result(bCanEnterTransition=NOT Boolean(A=Get bIsLastHandFireLeft()))]
 }
 Transition {
  path=ItemFullBodyOverride.AnimGraphNode_StateMachine_0.Dodge SM.AnimStateNode_5.In Air.AnimGraphNode_StateMachine_1.Dodge Directional.AnimStateNode_0.Left.AnimGraphNode_BlendSpaceGraph_3.Dodge Left.Left_BackRight_InAir.AnimGraphNode_StateMachine_0.Right Firing.AnimStateTransitionNode_5.Transition
  nodes [Result(bCanEnterTransition=Get bIsLastHandFireLeft())]
 }
 Transition {
  path=ItemFullBodyOverride.AnimGraphNode_StateMachine_0.Dodge SM.AnimStateNode_5.In Air.AnimGraphNode_StateMachine_1.Dodge Directional.AnimStateNode_0.Left.AnimGraphNode_BlendSpaceGraph_3.Dodge Left.Left_BackRight_InAir.AnimGraphNode_StateMachine_0.Right Firing.AnimStateTransitionNode_6.Transition
  nodes [Result()]
 }
 Transition {
  path=ItemFullBodyOverride.AnimGraphNode_StateMachine_0.Dodge SM.AnimStateNode_5.In Air.AnimGraphNode_StateMachine_1.Dodge Directional.AnimStateNode_0.Left.AnimGraphNode_BlendSpaceGraph_3.Dodge Left.Left_BackRight_InAir.AnimGraphNode_StateMachine_0.Right Firing.AnimStateTransitionNode_7.Transition
  nodes [Result()]
 }
 Left_Back_LeftSide_InAir {
  path=ItemFullBodyOverride.AnimGraphNode_StateMachine_0.Dodge SM.AnimStateNode_5.In Air.AnimGraphNode_StateMachine_1.Dodge Directional.AnimStateNode_0.Left.AnimGraphNode_BlendSpaceGraph_3.Dodge Left.Left_Back_LeftSide_InAir
  nodes [
   AnimGraphNode_RotationOffsetBlendSpaceGraph_0 = AimOffset 'LeftSide_Back_Left_AimOffset'!(
    X:float=Get AimPitch()
    BasePose:Pose Link Structure="Sequence Evaluator (by time) 'CherrySmoke_DualRevolver_Dodge_LeftSide_B_Diving'"!(ExplicitTime:float=Get InAirAccumulatedTime() -> (InAirAccumulatedTime:double)) -> (Pose:Pose Link Structure) {
     is=title AnimGraphNode_SequenceEvaluator
     sequence=/CherrySmokeDualRevolversGameplay/Assets/Anim/CharacterAnim/CherrySmoke_DualRevolver_Dodge_LeftSide_B_Diving
    }
    Alpha:float=1
   ) -> (Pose:Pose Link Structure) {is=title AnimGraphNode_RotationOffsetBlendSpaceGraph}
   AnimGraphNode_StateMachine_0 = Backwards Firing() {is=title AnimGraphNode_StateMachine}
   Output Animation Pose(Result=Apply Additive(Base=@AnimGraphNode_RotationOffsetBlendSpaceGraph_0, Additive=@AnimGraphNode_StateMachine_0) {alphaBoolBlend {blendOption=Linear}}) {is=title AnimGraphNode_BlendSpaceSampleResult}
  ]
 }
 LeftSide_Back_Left_AimOffset {
  path=ItemFullBodyOverride.AnimGraphNode_StateMachine_0.Dodge SM.AnimStateNode_5.In Air.AnimGraphNode_StateMachine_1.Dodge Directional.AnimStateNode_0.Left.AnimGraphNode_BlendSpaceGraph_3.Dodge Left.Left_Back_LeftSide_InAir.AnimGraphNode_RotationOffsetBlendSpaceGraph_0.LeftSide_Back_Left_AimOffset
  links []
  nodes []
 }
 L_B_L_Center {
  path=ItemFullBodyOverride.AnimGraphNode_StateMachine_0.Dodge SM.AnimStateNode_5.In Air.AnimGraphNode_StateMachine_1.Dodge Directional.AnimStateNode_0.Left.AnimGraphNode_BlendSpaceGraph_3.Dodge Left.Left_Back_LeftSide_InAir.AnimGraphNode_RotationOffsetBlendSpaceGraph_0.LeftSide_Back_Left_AimOffset.L_B_L_Center
  nodes [Output Animation Pose(Result=Additive Identity Pose())]
 }
 L_B_L_Down {
  path=ItemFullBodyOverride.AnimGraphNode_StateMachine_0.Dodge SM.AnimStateNode_5.In Air.AnimGraphNode_StateMachine_1.Dodge Directional.AnimStateNode_0.Left.AnimGraphNode_BlendSpaceGraph_3.Dodge Left.Left_Back_LeftSide_InAir.AnimGraphNode_RotationOffsetBlendSpaceGraph_0.LeftSide_Back_Left_AimOffset.L_B_L_Down
  nodes [
   n1 = "Sequence Player (additive) 'CherrySmoke_DualRevolvers_Dodge_LeftSide_B_AO_D10'"() {
    bLoopAnimation=false
    sequence=/CherrySmokeDualRevolversGameplay/Assets/Anim/Poses/Dodge_AO/CherrySmoke_DualRevolvers_Dodge_LeftSide_B_AO_D10
   }
   Output Animation Pose(Result=@n1)
  ]
 }
 L_B_L_Up {
  path=ItemFullBodyOverride.AnimGraphNode_StateMachine_0.Dodge SM.AnimStateNode_5.In Air.AnimGraphNode_StateMachine_1.Dodge Directional.AnimStateNode_0.Left.AnimGraphNode_BlendSpaceGraph_3.Dodge Left.Left_Back_LeftSide_InAir.AnimGraphNode_RotationOffsetBlendSpaceGraph_0.LeftSide_Back_Left_AimOffset.L_B_L_Up
  nodes [
   n1 = "Sequence Player (additive) 'CherrySmoke_DualRevolvers_Dodge_LeftSide_B_AO_U50'"() {
    bLoopAnimation=false
    sequence=/CherrySmokeDualRevolversGameplay/Assets/Anim/Poses/Dodge_AO/CherrySmoke_DualRevolvers_Dodge_LeftSide_B_AO_U50
   }
   Output Animation Pose(Result=@n1)
  ]
 }
 Backwards Firing {
  entry=Empty
  path=ItemFullBodyOverride.AnimGraphNode_StateMachine_0.Dodge SM.AnimStateNode_5.In Air.AnimGraphNode_StateMachine_1.Dodge Directional.AnimStateNode_0.Left.AnimGraphNode_BlendSpaceGraph_3.Dodge Left.Left_Back_LeftSide_InAir.AnimGraphNode_StateMachine_0.Backwards Firing
  ids=id
  nodes [
   AnimStateConduitNode_0 = Is Firing()
   AnimStateNode_0 = Left Fire() {bAlwaysResetOnEntry=true}
   AnimStateNode_1 = Empty()
   AnimStateNode_2 = Right Fire() {bAlwaysResetOnEntry=true}
   AnimStateTransitionNode_0 = AnimStateTransitionNode() {bAutomaticRuleBasedOnSequencePlayerInState=true}
   AnimStateTransitionNode_1 = AnimStateTransitionNode() {bAutomaticRuleBasedOnSequencePlayerInState=true}
   AnimStateTransitionNode_3 = AnimStateTransitionNode()
   AnimStateTransitionNode_4 = AnimStateTransitionNode() {crossfadeDuration=0.05}
   AnimStateTransitionNode_5 = AnimStateTransitionNode() {crossfadeDuration=0.05}
  ]
  links [
   {
    dst=Is Firing
    rule=ItemFullBodyOverride.AnimGraphNode_StateMachine_0.Dodge SM.AnimStateNode_5.In Air.AnimGraphNode_StateMachine_1.Dodge Directional.AnimStateNode_0.Left.AnimGraphNode_BlendSpaceGraph_3.Dodge Left.Left_Back_LeftSide_InAir.AnimGraphNode_StateMachine_0.Backwards Firing.AnimStateTransitionNode_3.Transition
    src=Empty
    via=AnimStateTransitionNode_3
   }
   {
    dst=Left Fire
    rule=ItemFullBodyOverride.AnimGraphNode_StateMachine_0.Dodge SM.AnimStateNode_5.In Air.AnimGraphNode_StateMachine_1.Dodge Directional.AnimStateNode_0.Left.AnimGraphNode_BlendSpaceGraph_3.Dodge Left.Left_Back_LeftSide_InAir.AnimGraphNode_StateMachine_0.Backwards Firing.AnimStateTransitionNode_4.Transition
    src=Is Firing
    via=AnimStateTransitionNode_4
   }
   {
    dst=Right Fire
    rule=ItemFullBodyOverride.AnimGraphNode_StateMachine_0.Dodge SM.AnimStateNode_5.In Air.AnimGraphNode_StateMachine_1.Dodge Directional.AnimStateNode_0.Left.AnimGraphNode_BlendSpaceGraph_3.Dodge Left.Left_Back_LeftSide_InAir.AnimGraphNode_StateMachine_0.Backwards Firing.AnimStateTransitionNode_5.Transition
    src=Is Firing
    via=AnimStateTransitionNode_5
   }
   {
    dst=Is Firing
    rule=ItemFullBodyOverride.AnimGraphNode_StateMachine_0.Dodge SM.AnimStateNode_5.In Air.AnimGraphNode_StateMachine_1.Dodge Directional.AnimStateNode_0.Left.AnimGraphNode_BlendSpaceGraph_3.Dodge Left.Left_Back_LeftSide_InAir.AnimGraphNode_StateMachine_0.Backwards Firing.AnimStateTransitionNode_0.Transition
    src=Left Fire
    via=AnimStateTransitionNode_0
   }
   {
    dst=Is Firing
    rule=ItemFullBodyOverride.AnimGraphNode_StateMachine_0.Dodge SM.AnimStateNode_5.In Air.AnimGraphNode_StateMachine_1.Dodge Directional.AnimStateNode_0.Left.AnimGraphNode_BlendSpaceGraph_3.Dodge Left.Left_Back_LeftSide_InAir.AnimGraphNode_StateMachine_0.Backwards Firing.AnimStateTransitionNode_1.Transition
    src=Right Fire
    via=AnimStateTransitionNode_1
   }
  ]
 }
 Is Firing {
  path=ItemFullBodyOverride.AnimGraphNode_StateMachine_0.Dodge SM.AnimStateNode_5.In Air.AnimGraphNode_StateMachine_1.Dodge Directional.AnimStateNode_0.Left.AnimGraphNode_BlendSpaceGraph_3.Dodge Left.Left_Back_LeftSide_InAir.AnimGraphNode_StateMachine_0.Backwards Firing.AnimStateConduitNode_0.Is Firing
  nodes [Result(bCanEnterTransition=Property Access() -> (Value=~))]
 }
 Left Fire {
  path=ItemFullBodyOverride.AnimGraphNode_StateMachine_0.Dodge SM.AnimStateNode_5.In Air.AnimGraphNode_StateMachine_1.Dodge Directional.AnimStateNode_0.Left.AnimGraphNode_BlendSpaceGraph_3.Dodge Left.Left_Back_LeftSide_InAir.AnimGraphNode_StateMachine_0.Backwards Firing.AnimStateNode_0.Left Fire
  nodes [
   n1 = "Sequence Player (additive) 'CherrySmoke_DualRevolvers_back_dodge_fire_L'"!(PlayRate:float=Get AdditiveLoopPlayRate()) -> (Pose) {bLoopAnimation=false, sequence=/CherrySmokeDualRevolversGameplay/Assets/Anim/CharacterAnim/CherrySmoke_DualRevolvers_back_dodge_fire_L}
   Output Animation Pose(Result=@n1) {is=title AnimGraphNode_StateResult}
  ]
 }
 Empty {
  path=ItemFullBodyOverride.AnimGraphNode_StateMachine_0.Dodge SM.AnimStateNode_5.In Air.AnimGraphNode_StateMachine_1.Dodge Directional.AnimStateNode_0.Left.AnimGraphNode_BlendSpaceGraph_3.Dodge Left.Left_Back_LeftSide_InAir.AnimGraphNode_StateMachine_0.Backwards Firing.AnimStateNode_1.Empty
  nodes [Output Animation Pose(Result=Additive Identity Pose())]
 }
 Right Fire {
  path=ItemFullBodyOverride.AnimGraphNode_StateMachine_0.Dodge SM.AnimStateNode_5.In Air.AnimGraphNode_StateMachine_1.Dodge Directional.AnimStateNode_0.Left.AnimGraphNode_BlendSpaceGraph_3.Dodge Left.Left_Back_LeftSide_InAir.AnimGraphNode_StateMachine_0.Backwards Firing.AnimStateNode_2.Right Fire
  nodes [
   n1 = "Sequence Player (additive) 'CherrySmoke_DualRevolvers_back_dodge_fire_R'"!(PlayRate:float=Get AdditiveLoopPlayRate()) -> (Pose) {bLoopAnimation=false, sequence=/CherrySmokeDualRevolversGameplay/Assets/Anim/CharacterAnim/CherrySmoke_DualRevolvers_back_dodge_fire_R}
   Output Animation Pose(Result=@n1)
  ]
 }
 Transition {
  path=ItemFullBodyOverride.AnimGraphNode_StateMachine_0.Dodge SM.AnimStateNode_5.In Air.AnimGraphNode_StateMachine_1.Dodge Directional.AnimStateNode_0.Left.AnimGraphNode_BlendSpaceGraph_3.Dodge Left.Left_Back_LeftSide_InAir.AnimGraphNode_StateMachine_0.Backwards Firing.AnimStateTransitionNode_0.Transition
  nodes [Result()]
 }
 Transition {
  path=ItemFullBodyOverride.AnimGraphNode_StateMachine_0.Dodge SM.AnimStateNode_5.In Air.AnimGraphNode_StateMachine_1.Dodge Directional.AnimStateNode_0.Left.AnimGraphNode_BlendSpaceGraph_3.Dodge Left.Left_Back_LeftSide_InAir.AnimGraphNode_StateMachine_0.Backwards Firing.AnimStateTransitionNode_1.Transition
  nodes [Result()]
 }
 Transition {
  path=ItemFullBodyOverride.AnimGraphNode_StateMachine_0.Dodge SM.AnimStateNode_5.In Air.AnimGraphNode_StateMachine_1.Dodge Directional.AnimStateNode_0.Left.AnimGraphNode_BlendSpaceGraph_3.Dodge Left.Left_Back_LeftSide_InAir.AnimGraphNode_StateMachine_0.Backwards Firing.AnimStateTransitionNode_3.Transition
  nodes [Result(bCanEnterTransition=true)]
 }
 Transition {
  path=ItemFullBodyOverride.AnimGraphNode_StateMachine_0.Dodge SM.AnimStateNode_5.In Air.AnimGraphNode_StateMachine_1.Dodge Directional.AnimStateNode_0.Left.AnimGraphNode_BlendSpaceGraph_3.Dodge Left.Left_Back_LeftSide_InAir.AnimGraphNode_StateMachine_0.Backwards Firing.AnimStateTransitionNode_4.Transition
  nodes [Result(bCanEnterTransition=NOT Boolean(A=Get bIsLastHandFireLeft()))]
 }
 Transition {
  path=ItemFullBodyOverride.AnimGraphNode_StateMachine_0.Dodge SM.AnimStateNode_5.In Air.AnimGraphNode_StateMachine_1.Dodge Directional.AnimStateNode_0.Left.AnimGraphNode_BlendSpaceGraph_3.Dodge Left.Left_Back_LeftSide_InAir.AnimGraphNode_StateMachine_0.Backwards Firing.AnimStateTransitionNode_5.Transition
  nodes [Result(bCanEnterTransition=Get bIsLastHandFireLeft())]
 }
 Left_Back_RightSide_InAir {
  path=ItemFullBodyOverride.AnimGraphNode_StateMachine_0.Dodge SM.AnimStateNode_5.In Air.AnimGraphNode_StateMachine_1.Dodge Directional.AnimStateNode_0.Left.AnimGraphNode_BlendSpaceGraph_3.Dodge Left.Left_Back_RightSide_InAir
  nodes [
   AnimGraphNode_RotationOffsetBlendSpaceGraph_0 = AimOffset 'LeftSide_Back_Right_AimOffset'!(
    X:float=Get AimPitch()
    BasePose:Pose Link Structure="Sequence Evaluator (by time) 'CherrySmoke_DualRevolver_Dodge_LeftSide_B_Diving'"(ExplicitTime=Get InAirAccumulatedTime()) {sequence=&CherrySmoke_DualRevolver_Dodge_LeftSide_B_Diving}
    Alpha:float=1
   ) -> (Pose:Pose Link Structure) {is=title AnimGraphNode_RotationOffsetBlendSpaceGraph}
   AnimGraphNode_StateMachine_0 = Backwards Firing()
   Output Animation Pose(Result=Apply Additive(Base=@AnimGraphNode_RotationOffsetBlendSpaceGraph_0, Additive=@AnimGraphNode_StateMachine_0) {alphaBoolBlend {blendOption=Linear}}) {is=title AnimGraphNode_BlendSpaceSampleResult}
  ]
 }
 LeftSide_Back_Right_AimOffset {
  path=ItemFullBodyOverride.AnimGraphNode_StateMachine_0.Dodge SM.AnimStateNode_5.In Air.AnimGraphNode_StateMachine_1.Dodge Directional.AnimStateNode_0.Left.AnimGraphNode_BlendSpaceGraph_3.Dodge Left.Left_Back_RightSide_InAir.AnimGraphNode_RotationOffsetBlendSpaceGraph_0.LeftSide_Back_Right_AimOffset
  links []
  nodes []
 }
 L_B_R_Center {
  path=ItemFullBodyOverride.AnimGraphNode_StateMachine_0.Dodge SM.AnimStateNode_5.In Air.AnimGraphNode_StateMachine_1.Dodge Directional.AnimStateNode_0.Left.AnimGraphNode_BlendSpaceGraph_3.Dodge Left.Left_Back_RightSide_InAir.AnimGraphNode_RotationOffsetBlendSpaceGraph_0.LeftSide_Back_Right_AimOffset.L_B_R_Center
  nodes [Output Animation Pose(Result=Additive Identity Pose())]
 }
 L_B_R_Down {
  path=ItemFullBodyOverride.AnimGraphNode_StateMachine_0.Dodge SM.AnimStateNode_5.In Air.AnimGraphNode_StateMachine_1.Dodge Directional.AnimStateNode_0.Left.AnimGraphNode_BlendSpaceGraph_3.Dodge Left.Left_Back_RightSide_InAir.AnimGraphNode_RotationOffsetBlendSpaceGraph_0.LeftSide_Back_Right_AimOffset.L_B_R_Down
  nodes [
   n1 = "Sequence Player (additive) 'CherrySmoke_DualRevolvers_Dodge_LeftSide_B_AO_D10'"() {
    bLoopAnimation=false
    sequence=&CherrySmoke_DualRevolvers_Dodge_LeftSide_B_AO_D10
   }
   Output Animation Pose(Result=@n1)
  ]
 }
 L_B_R_Up {
  path=ItemFullBodyOverride.AnimGraphNode_StateMachine_0.Dodge SM.AnimStateNode_5.In Air.AnimGraphNode_StateMachine_1.Dodge Directional.AnimStateNode_0.Left.AnimGraphNode_BlendSpaceGraph_3.Dodge Left.Left_Back_RightSide_InAir.AnimGraphNode_RotationOffsetBlendSpaceGraph_0.LeftSide_Back_Right_AimOffset.L_B_R_Up
  nodes [
   n1 = "Sequence Player (additive) 'CherrySmoke_DualRevolvers_Dodge_LeftSide_B_AO_U50'"() {
    bLoopAnimation=false
    sequence=&CherrySmoke_DualRevolvers_Dodge_LeftSide_B_AO_U50
   }
   Output Animation Pose(Result=@n1)
  ]
 }
 Backwards Firing {
  entry=Empty
  path=ItemFullBodyOverride.AnimGraphNode_StateMachine_0.Dodge SM.AnimStateNode_5.In Air.AnimGraphNode_StateMachine_1.Dodge Directional.AnimStateNode_0.Left.AnimGraphNode_BlendSpaceGraph_3.Dodge Left.Left_Back_RightSide_InAir.AnimGraphNode_StateMachine_0.Backwards Firing
  ids=id
  nodes [
   AnimStateConduitNode_0 = Is Firing()
   AnimStateNode_0 = Left Fire() {bAlwaysResetOnEntry=true}
   AnimStateNode_1 = Empty()
   AnimStateNode_2 = Right Fire() {bAlwaysResetOnEntry=true}
   AnimStateTransitionNode_12 = AnimStateTransitionNode() {bAutomaticRuleBasedOnSequencePlayerInState=true}
   AnimStateTransitionNode_13 = AnimStateTransitionNode() {bAutomaticRuleBasedOnSequencePlayerInState=true}
   AnimStateTransitionNode_3 = AnimStateTransitionNode()
   AnimStateTransitionNode_4 = AnimStateTransitionNode() {crossfadeDuration=0.05}
   AnimStateTransitionNode_5 = AnimStateTransitionNode() {crossfadeDuration=0.05}
  ]
  links [
   {
    dst=Is Firing
    rule=ItemFullBodyOverride.AnimGraphNode_StateMachine_0.Dodge SM.AnimStateNode_5.In Air.AnimGraphNode_StateMachine_1.Dodge Directional.AnimStateNode_0.Left.AnimGraphNode_BlendSpaceGraph_3.Dodge Left.Left_Back_RightSide_InAir.AnimGraphNode_StateMachine_0.Backwards Firing.AnimStateTransitionNode_3.Transition
    src=Empty
    via=AnimStateTransitionNode_3
   }
   {
    dst=Left Fire
    rule=ItemFullBodyOverride.AnimGraphNode_StateMachine_0.Dodge SM.AnimStateNode_5.In Air.AnimGraphNode_StateMachine_1.Dodge Directional.AnimStateNode_0.Left.AnimGraphNode_BlendSpaceGraph_3.Dodge Left.Left_Back_RightSide_InAir.AnimGraphNode_StateMachine_0.Backwards Firing.AnimStateTransitionNode_4.Transition
    src=Is Firing
    via=AnimStateTransitionNode_4
   }
   {
    dst=Right Fire
    rule=ItemFullBodyOverride.AnimGraphNode_StateMachine_0.Dodge SM.AnimStateNode_5.In Air.AnimGraphNode_StateMachine_1.Dodge Directional.AnimStateNode_0.Left.AnimGraphNode_BlendSpaceGraph_3.Dodge Left.Left_Back_RightSide_InAir.AnimGraphNode_StateMachine_0.Backwards Firing.AnimStateTransitionNode_5.Transition
    src=Is Firing
    via=AnimStateTransitionNode_5
   }
   {
    dst=Is Firing
    rule=ItemFullBodyOverride.AnimGraphNode_StateMachine_0.Dodge SM.AnimStateNode_5.In Air.AnimGraphNode_StateMachine_1.Dodge Directional.AnimStateNode_0.Left.AnimGraphNode_BlendSpaceGraph_3.Dodge Left.Left_Back_RightSide_InAir.AnimGraphNode_StateMachine_0.Backwards Firing.AnimStateTransitionNode_12.Transition
    src=Left Fire
    via=AnimStateTransitionNode_12
   }
   {
    dst=Is Firing
    rule=ItemFullBodyOverride.AnimGraphNode_StateMachine_0.Dodge SM.AnimStateNode_5.In Air.AnimGraphNode_StateMachine_1.Dodge Directional.AnimStateNode_0.Left.AnimGraphNode_BlendSpaceGraph_3.Dodge Left.Left_Back_RightSide_InAir.AnimGraphNode_StateMachine_0.Backwards Firing.AnimStateTransitionNode_13.Transition
    src=Right Fire
    via=AnimStateTransitionNode_13
   }
  ]
 }
 Is Firing {
  path=ItemFullBodyOverride.AnimGraphNode_StateMachine_0.Dodge SM.AnimStateNode_5.In Air.AnimGraphNode_StateMachine_1.Dodge Directional.AnimStateNode_0.Left.AnimGraphNode_BlendSpaceGraph_3.Dodge Left.Left_Back_RightSide_InAir.AnimGraphNode_StateMachine_0.Backwards Firing.AnimStateConduitNode_0.Is Firing
  nodes [Result(bCanEnterTransition=Property Access() -> (Value=~))]
 }
 Left Fire {
  path=ItemFullBodyOverride.AnimGraphNode_StateMachine_0.Dodge SM.AnimStateNode_5.In Air.AnimGraphNode_StateMachine_1.Dodge Directional.AnimStateNode_0.Left.AnimGraphNode_BlendSpaceGraph_3.Dodge Left.Left_Back_RightSide_InAir.AnimGraphNode_StateMachine_0.Backwards Firing.AnimStateNode_0.Left Fire
  nodes [
   n1 = "Sequence Player (additive) 'CherrySmoke_DualRevolvers_back_dodge_fire_L'"(PlayRate=Get AdditiveLoopPlayRate()) {bLoopAnimation=false, sequence=&CherrySmoke_DualRevolvers_back_dodge_fire_L}
   Output Animation Pose(Result=@n1) {is=title AnimGraphNode_StateResult}
  ]
 }
 Empty {
  path=ItemFullBodyOverride.AnimGraphNode_StateMachine_0.Dodge SM.AnimStateNode_5.In Air.AnimGraphNode_StateMachine_1.Dodge Directional.AnimStateNode_0.Left.AnimGraphNode_BlendSpaceGraph_3.Dodge Left.Left_Back_RightSide_InAir.AnimGraphNode_StateMachine_0.Backwards Firing.AnimStateNode_1.Empty
  nodes [Output Animation Pose(Result=Additive Identity Pose())]
 }
 Right Fire {
  path=ItemFullBodyOverride.AnimGraphNode_StateMachine_0.Dodge SM.AnimStateNode_5.In Air.AnimGraphNode_StateMachine_1.Dodge Directional.AnimStateNode_0.Left.AnimGraphNode_BlendSpaceGraph_3.Dodge Left.Left_Back_RightSide_InAir.AnimGraphNode_StateMachine_0.Backwards Firing.AnimStateNode_2.Right Fire
  nodes [
   n1 = "Sequence Player (additive) 'CherrySmoke_DualRevolvers_back_dodge_fire_R'"(PlayRate=Get AdditiveLoopPlayRate()) {bLoopAnimation=false, sequence=&CherrySmoke_DualRevolvers_back_dodge_fire_R}
   Output Animation Pose(Result=@n1)
  ]
 }
 Transition {
  path=ItemFullBodyOverride.AnimGraphNode_StateMachine_0.Dodge SM.AnimStateNode_5.In Air.AnimGraphNode_StateMachine_1.Dodge Directional.AnimStateNode_0.Left.AnimGraphNode_BlendSpaceGraph_3.Dodge Left.Left_Back_RightSide_InAir.AnimGraphNode_StateMachine_0.Backwards Firing.AnimStateTransitionNode_12.Transition
  nodes [Result()]
 }
 Transition {
  path=ItemFullBodyOverride.AnimGraphNode_StateMachine_0.Dodge SM.AnimStateNode_5.In Air.AnimGraphNode_StateMachine_1.Dodge Directional.AnimStateNode_0.Left.AnimGraphNode_BlendSpaceGraph_3.Dodge Left.Left_Back_RightSide_InAir.AnimGraphNode_StateMachine_0.Backwards Firing.AnimStateTransitionNode_13.Transition
  nodes [Result()]
 }
 Transition {
  path=ItemFullBodyOverride.AnimGraphNode_StateMachine_0.Dodge SM.AnimStateNode_5.In Air.AnimGraphNode_StateMachine_1.Dodge Directional.AnimStateNode_0.Left.AnimGraphNode_BlendSpaceGraph_3.Dodge Left.Left_Back_RightSide_InAir.AnimGraphNode_StateMachine_0.Backwards Firing.AnimStateTransitionNode_3.Transition
  nodes [Result(bCanEnterTransition=true)]
 }
 Transition {
  path=ItemFullBodyOverride.AnimGraphNode_StateMachine_0.Dodge SM.AnimStateNode_5.In Air.AnimGraphNode_StateMachine_1.Dodge Directional.AnimStateNode_0.Left.AnimGraphNode_BlendSpaceGraph_3.Dodge Left.Left_Back_RightSide_InAir.AnimGraphNode_StateMachine_0.Backwards Firing.AnimStateTransitionNode_4.Transition
  nodes [Result(bCanEnterTransition=NOT Boolean(A=Get bIsLastHandFireLeft()))]
 }
 Transition {
  path=ItemFullBodyOverride.AnimGraphNode_StateMachine_0.Dodge SM.AnimStateNode_5.In Air.AnimGraphNode_StateMachine_1.Dodge Directional.AnimStateNode_0.Left.AnimGraphNode_BlendSpaceGraph_3.Dodge Left.Left_Back_RightSide_InAir.AnimGraphNode_StateMachine_0.Backwards Firing.AnimStateTransitionNode_5.Transition
  nodes [Result(bCanEnterTransition=Get bIsLastHandFireLeft())]
 }
 Left_ForwardLeft_InAir {
  path=ItemFullBodyOverride.AnimGraphNode_StateMachine_0.Dodge SM.AnimStateNode_5.In Air.AnimGraphNode_StateMachine_1.Dodge Directional.AnimStateNode_0.Left.AnimGraphNode_BlendSpaceGraph_3.Dodge Left.Left_ForwardLeft_InAir
  nodes [
   n1 = Sequence Player 'CherrySmoke_DualRevolver_Dodge_LeftSide_FL_Diving'() {bLoopAnimation=false, sequence=/CherrySmokeDualRevolversGameplay/Assets/Anim/CharacterAnim/CherrySmoke_DualRevolver_Dodge_LeftSide_FL_Diving}
   AnimGraphNode_RotationOffsetBlendSpaceGraph_0 = AimOffset 'LeftSide_ForwardLeft_AimOffset'!(X:float=Get AimPitch(), BasePose:Pose Link Structure=@n1, Alpha:float=1) -> (Pose:Pose Link Structure) {is=title AnimGraphNode_RotationOffsetBlendSpaceGraph}
   AnimGraphNode_StateMachine_0 = Forward Firing()
   Output Animation Pose(Result=Apply Additive(Base=@AnimGraphNode_RotationOffsetBlendSpaceGraph_0, Additive=@AnimGraphNode_StateMachine_0) {alphaBoolBlend {blendOption=Linear}}) {is=title AnimGraphNode_BlendSpaceSampleResult}
  ]
 }
 LeftSide_ForwardLeft_AimOffset {
  path=ItemFullBodyOverride.AnimGraphNode_StateMachine_0.Dodge SM.AnimStateNode_5.In Air.AnimGraphNode_StateMachine_1.Dodge Directional.AnimStateNode_0.Left.AnimGraphNode_BlendSpaceGraph_3.Dodge Left.Left_ForwardLeft_InAir.AnimGraphNode_RotationOffsetBlendSpaceGraph_0.LeftSide_ForwardLeft_AimOffset
  links []
  nodes []
 }
 L_FL_Center {
  path=ItemFullBodyOverride.AnimGraphNode_StateMachine_0.Dodge SM.AnimStateNode_5.In Air.AnimGraphNode_StateMachine_1.Dodge Directional.AnimStateNode_0.Left.AnimGraphNode_BlendSpaceGraph_3.Dodge Left.Left_ForwardLeft_InAir.AnimGraphNode_RotationOffsetBlendSpaceGraph_0.LeftSide_ForwardLeft_AimOffset.L_FL_Center
  nodes [Output Animation Pose(Result=Additive Identity Pose())]
 }
 L_FL_Down {
  path=ItemFullBodyOverride.AnimGraphNode_StateMachine_0.Dodge SM.AnimStateNode_5.In Air.AnimGraphNode_StateMachine_1.Dodge Directional.AnimStateNode_0.Left.AnimGraphNode_BlendSpaceGraph_3.Dodge Left.Left_ForwardLeft_InAir.AnimGraphNode_RotationOffsetBlendSpaceGraph_0.LeftSide_ForwardLeft_AimOffset.L_FL_Down
  nodes [
   n1 = "Sequence Player (additive) 'CherrySmoke_DualRevolvers_Dodge_LeftSide_F_AO_D50'"() {
    bLoopAnimation=false
    sequence=/CherrySmokeDualRevolversGameplay/Assets/Anim/Poses/Dodge_AO/CherrySmoke_DualRevolvers_Dodge_LeftSide_F_AO_D50
   }
   Output Animation Pose(Result=@n1)
  ]
 }
 L_FL_UP {
  path=ItemFullBodyOverride.AnimGraphNode_StateMachine_0.Dodge SM.AnimStateNode_5.In Air.AnimGraphNode_StateMachine_1.Dodge Directional.AnimStateNode_0.Left.AnimGraphNode_BlendSpaceGraph_3.Dodge Left.Left_ForwardLeft_InAir.AnimGraphNode_RotationOffsetBlendSpaceGraph_0.LeftSide_ForwardLeft_AimOffset.L_FL_UP
  nodes [
   n1 = "Sequence Player (additive) 'CherrySmoke_DualRevolvers_Dodge_LeftSide_Fl_AO_U50'"() {
    bLoopAnimation=false
    sequence=/CherrySmokeDualRevolversGameplay/Assets/Anim/Poses/Dodge_AO/CherrySmoke_DualRevolvers_Dodge_LeftSide_Fl_AO_U50
   }
   Output Animation Pose(Result=@n1)
  ]
 }
 Forward Firing {
  entry=Empty
  path=ItemFullBodyOverride.AnimGraphNode_StateMachine_0.Dodge SM.AnimStateNode_5.In Air.AnimGraphNode_StateMachine_1.Dodge Directional.AnimStateNode_0.Left.AnimGraphNode_BlendSpaceGraph_3.Dodge Left.Left_ForwardLeft_InAir.AnimGraphNode_StateMachine_0.Forward Firing
  ids=id
  nodes [
   AnimStateConduitNode_0 = Is Firing()
   AnimStateNode_0 = Left Fire() {bAlwaysResetOnEntry=true}
   AnimStateNode_1 = Empty()
   AnimStateNode_2 = Right Fire() {bAlwaysResetOnEntry=true}
   AnimStateTransitionNode_0 = AnimStateTransitionNode()
   AnimStateTransitionNode_1 = AnimStateTransitionNode() {crossfadeDuration=0.05}
   AnimStateTransitionNode_2 = AnimStateTransitionNode() {crossfadeDuration=0.05}
   AnimStateTransitionNode_3 = AnimStateTransitionNode() {bAutomaticRuleBasedOnSequencePlayerInState=true}
   AnimStateTransitionNode_4 = AnimStateTransitionNode() {bAutomaticRuleBasedOnSequencePlayerInState=true}
  ]
  links [
   {
    dst=Is Firing
    rule=ItemFullBodyOverride.AnimGraphNode_StateMachine_0.Dodge SM.AnimStateNode_5.In Air.AnimGraphNode_StateMachine_1.Dodge Directional.AnimStateNode_0.Left.AnimGraphNode_BlendSpaceGraph_3.Dodge Left.Left_ForwardLeft_InAir.AnimGraphNode_StateMachine_0.Forward Firing.AnimStateTransitionNode_0.Transition
    src=Empty
    via=AnimStateTransitionNode_0
   }
   {
    dst=Left Fire
    rule=ItemFullBodyOverride.AnimGraphNode_StateMachine_0.Dodge SM.AnimStateNode_5.In Air.AnimGraphNode_StateMachine_1.Dodge Directional.AnimStateNode_0.Left.AnimGraphNode_BlendSpaceGraph_3.Dodge Left.Left_ForwardLeft_InAir.AnimGraphNode_StateMachine_0.Forward Firing.AnimStateTransitionNode_1.Transition
    src=Is Firing
    via=AnimStateTransitionNode_1
   }
   {
    dst=Right Fire
    rule=ItemFullBodyOverride.AnimGraphNode_StateMachine_0.Dodge SM.AnimStateNode_5.In Air.AnimGraphNode_StateMachine_1.Dodge Directional.AnimStateNode_0.Left.AnimGraphNode_BlendSpaceGraph_3.Dodge Left.Left_ForwardLeft_InAir.AnimGraphNode_StateMachine_0.Forward Firing.AnimStateTransitionNode_2.Transition
    src=Is Firing
    via=AnimStateTransitionNode_2
   }
   {
    dst=Is Firing
    rule=ItemFullBodyOverride.AnimGraphNode_StateMachine_0.Dodge SM.AnimStateNode_5.In Air.AnimGraphNode_StateMachine_1.Dodge Directional.AnimStateNode_0.Left.AnimGraphNode_BlendSpaceGraph_3.Dodge Left.Left_ForwardLeft_InAir.AnimGraphNode_StateMachine_0.Forward Firing.AnimStateTransitionNode_3.Transition
    src=Left Fire
    via=AnimStateTransitionNode_3
   }
   {
    dst=Is Firing
    rule=ItemFullBodyOverride.AnimGraphNode_StateMachine_0.Dodge SM.AnimStateNode_5.In Air.AnimGraphNode_StateMachine_1.Dodge Directional.AnimStateNode_0.Left.AnimGraphNode_BlendSpaceGraph_3.Dodge Left.Left_ForwardLeft_InAir.AnimGraphNode_StateMachine_0.Forward Firing.AnimStateTransitionNode_4.Transition
    src=Right Fire
    via=AnimStateTransitionNode_4
   }
  ]
 }
 Is Firing {
  path=ItemFullBodyOverride.AnimGraphNode_StateMachine_0.Dodge SM.AnimStateNode_5.In Air.AnimGraphNode_StateMachine_1.Dodge Directional.AnimStateNode_0.Left.AnimGraphNode_BlendSpaceGraph_3.Dodge Left.Left_ForwardLeft_InAir.AnimGraphNode_StateMachine_0.Forward Firing.AnimStateConduitNode_0.Is Firing
  nodes [Result(bCanEnterTransition=Property Access() -> (Value=~))]
 }
 Left Fire {
  path=ItemFullBodyOverride.AnimGraphNode_StateMachine_0.Dodge SM.AnimStateNode_5.In Air.AnimGraphNode_StateMachine_1.Dodge Directional.AnimStateNode_0.Left.AnimGraphNode_BlendSpaceGraph_3.Dodge Left.Left_ForwardLeft_InAir.AnimGraphNode_StateMachine_0.Forward Firing.AnimStateNode_0.Left Fire
  nodes [
   n1 = "Sequence Player (additive) 'CherrySmoke_DualRevolvers_front_dodge_fire_L'"!(PlayRate:float=Get AdditiveLoopPlayRate()) -> (Pose) {bLoopAnimation=false, sequence=/CherrySmokeDualRevolversGameplay/Assets/Anim/CharacterAnim/CherrySmoke_DualRevolvers_front_dodge_fire_L}
   Output Animation Pose(Result=@n1) {is=title AnimGraphNode_StateResult}
  ]
 }
 Empty {
  path=ItemFullBodyOverride.AnimGraphNode_StateMachine_0.Dodge SM.AnimStateNode_5.In Air.AnimGraphNode_StateMachine_1.Dodge Directional.AnimStateNode_0.Left.AnimGraphNode_BlendSpaceGraph_3.Dodge Left.Left_ForwardLeft_InAir.AnimGraphNode_StateMachine_0.Forward Firing.AnimStateNode_1.Empty
  nodes [Output Animation Pose(Result=Additive Identity Pose())]
 }
 Right Fire {
  path=ItemFullBodyOverride.AnimGraphNode_StateMachine_0.Dodge SM.AnimStateNode_5.In Air.AnimGraphNode_StateMachine_1.Dodge Directional.AnimStateNode_0.Left.AnimGraphNode_BlendSpaceGraph_3.Dodge Left.Left_ForwardLeft_InAir.AnimGraphNode_StateMachine_0.Forward Firing.AnimStateNode_2.Right Fire
  nodes [
   n1 = "Sequence Player (additive) 'CherrySmoke_DualRevolvers_front_dodge_fire_R'"!(PlayRate:float=Get AdditiveLoopPlayRate()) -> (Pose) {bLoopAnimation=false, sequence=/CherrySmokeDualRevolversGameplay/Assets/Anim/CharacterAnim/CherrySmoke_DualRevolvers_front_dodge_fire_R}
   Output Animation Pose(Result=@n1)
  ]
 }
 Transition {
  path=ItemFullBodyOverride.AnimGraphNode_StateMachine_0.Dodge SM.AnimStateNode_5.In Air.AnimGraphNode_StateMachine_1.Dodge Directional.AnimStateNode_0.Left.AnimGraphNode_BlendSpaceGraph_3.Dodge Left.Left_ForwardLeft_InAir.AnimGraphNode_StateMachine_0.Forward Firing.AnimStateTransitionNode_0.Transition
  nodes [Result(bCanEnterTransition=true)]
 }
 Transition {
  path=ItemFullBodyOverride.AnimGraphNode_StateMachine_0.Dodge SM.AnimStateNode_5.In Air.AnimGraphNode_StateMachine_1.Dodge Directional.AnimStateNode_0.Left.AnimGraphNode_BlendSpaceGraph_3.Dodge Left.Left_ForwardLeft_InAir.AnimGraphNode_StateMachine_0.Forward Firing.AnimStateTransitionNode_1.Transition
  nodes [Result(bCanEnterTransition=NOT Boolean(A=Get bIsLastHandFireLeft()))]
 }
 Transition {
  path=ItemFullBodyOverride.AnimGraphNode_StateMachine_0.Dodge SM.AnimStateNode_5.In Air.AnimGraphNode_StateMachine_1.Dodge Directional.AnimStateNode_0.Left.AnimGraphNode_BlendSpaceGraph_3.Dodge Left.Left_ForwardLeft_InAir.AnimGraphNode_StateMachine_0.Forward Firing.AnimStateTransitionNode_2.Transition
  nodes [Result(bCanEnterTransition=Get bIsLastHandFireLeft())]
 }
 Transition {
  path=ItemFullBodyOverride.AnimGraphNode_StateMachine_0.Dodge SM.AnimStateNode_5.In Air.AnimGraphNode_StateMachine_1.Dodge Directional.AnimStateNode_0.Left.AnimGraphNode_BlendSpaceGraph_3.Dodge Left.Left_ForwardLeft_InAir.AnimGraphNode_StateMachine_0.Forward Firing.AnimStateTransitionNode_3.Transition
  nodes [Result()]
 }
 Transition {
  path=ItemFullBodyOverride.AnimGraphNode_StateMachine_0.Dodge SM.AnimStateNode_5.In Air.AnimGraphNode_StateMachine_1.Dodge Directional.AnimStateNode_0.Left.AnimGraphNode_BlendSpaceGraph_3.Dodge Left.Left_ForwardLeft_InAir.AnimGraphNode_StateMachine_0.Forward Firing.AnimStateTransitionNode_4.Transition
  nodes [Result()]
 }
 Left_ForwardRight_AimOffset {
  path=ItemFullBodyOverride.AnimGraphNode_StateMachine_0.Dodge SM.AnimStateNode_5.In Air.AnimGraphNode_StateMachine_1.Dodge Directional.AnimStateNode_0.Left.AnimGraphNode_BlendSpaceGraph_3.Dodge Left.Left_ForwardRight_AimOffset
  nodes [
   n1 = Sequence Player 'CherrySmoke_DualRevolver_Dodge_LeftSide_FR_Diving'() {bLoopAnimation=false, sequence=/CherrySmokeDualRevolversGameplay/Assets/Anim/CharacterAnim/CherrySmoke_DualRevolver_Dodge_LeftSide_FR_Diving}
   AnimGraphNode_RotationOffsetBlendSpaceGraph_0 = AimOffset 'LeftSide_ForwardRight_AimOffset'!(X:float=Get AimPitch(), BasePose:Pose Link Structure=@n1, Alpha:float=1) -> (Pose:Pose Link Structure) {is=title AnimGraphNode_RotationOffsetBlendSpaceGraph}
   AnimGraphNode_StateMachine_0 = Right Firing()
   Output Animation Pose(Result=Apply Additive(Base=@AnimGraphNode_RotationOffsetBlendSpaceGraph_0, Additive=@AnimGraphNode_StateMachine_0) {alphaBoolBlend {blendOption=Linear}}) {is=title AnimGraphNode_BlendSpaceSampleResult}
  ]
 }
 LeftSide_ForwardRight_AimOffset {
  path=ItemFullBodyOverride.AnimGraphNode_StateMachine_0.Dodge SM.AnimStateNode_5.In Air.AnimGraphNode_StateMachine_1.Dodge Directional.AnimStateNode_0.Left.AnimGraphNode_BlendSpaceGraph_3.Dodge Left.Left_ForwardRight_AimOffset.AnimGraphNode_RotationOffsetBlendSpaceGraph_0.LeftSide_ForwardRight_AimOffset
  links []
  nodes []
 }
 L_FR_Center {
  path=ItemFullBodyOverride.AnimGraphNode_StateMachine_0.Dodge SM.AnimStateNode_5.In Air.AnimGraphNode_StateMachine_1.Dodge Directional.AnimStateNode_0.Left.AnimGraphNode_BlendSpaceGraph_3.Dodge Left.Left_ForwardRight_AimOffset.AnimGraphNode_RotationOffsetBlendSpaceGraph_0.LeftSide_ForwardRight_AimOffset.L_FR_Center
  nodes [Output Animation Pose(Result=Additive Identity Pose())]
 }
 L_FR_Down {
  path=ItemFullBodyOverride.AnimGraphNode_StateMachine_0.Dodge SM.AnimStateNode_5.In Air.AnimGraphNode_StateMachine_1.Dodge Directional.AnimStateNode_0.Left.AnimGraphNode_BlendSpaceGraph_3.Dodge Left.Left_ForwardRight_AimOffset.AnimGraphNode_RotationOffsetBlendSpaceGraph_0.LeftSide_ForwardRight_AimOffset.L_FR_Down
  nodes [
   n1 = "Sequence Player (additive) 'CherrySmoke_DualRevolvers_Dodge_LeftSide_FR_AO_D50'"() {
    bLoopAnimation=false
    sequence=/CherrySmokeDualRevolversGameplay/Assets/Anim/Poses/Dodge_AO/CherrySmoke_DualRevolvers_Dodge_LeftSide_FR_AO_D50
   }
   Output Animation Pose(Result=@n1)
  ]
 }
 L_FR_UP {
  path=ItemFullBodyOverride.AnimGraphNode_StateMachine_0.Dodge SM.AnimStateNode_5.In Air.AnimGraphNode_StateMachine_1.Dodge Directional.AnimStateNode_0.Left.AnimGraphNode_BlendSpaceGraph_3.Dodge Left.Left_ForwardRight_AimOffset.AnimGraphNode_RotationOffsetBlendSpaceGraph_0.LeftSide_ForwardRight_AimOffset.L_FR_UP
  nodes [
   n1 = "Sequence Player (additive) 'CherrySmoke_DualRevolvers_Dodge_LeftSide_FR_AO_U40'"() {
    bLoopAnimation=false
    sequence=/CherrySmokeDualRevolversGameplay/Assets/Anim/Poses/Dodge_AO/CherrySmoke_DualRevolvers_Dodge_LeftSide_FR_AO_U40
   }
   Output Animation Pose(Result=@n1)
  ]
 }
 Right Firing {
  entry=Empty
  path=ItemFullBodyOverride.AnimGraphNode_StateMachine_0.Dodge SM.AnimStateNode_5.In Air.AnimGraphNode_StateMachine_1.Dodge Directional.AnimStateNode_0.Left.AnimGraphNode_BlendSpaceGraph_3.Dodge Left.Left_ForwardRight_AimOffset.AnimGraphNode_StateMachine_0.Right Firing
  ids=id
  nodes [
   AnimStateConduitNode_0 = Is Firing()
   AnimStateNode_0 = Left Fire() {bAlwaysResetOnEntry=true}
   AnimStateNode_1 = Empty()
   AnimStateNode_2 = Right Fire() {bAlwaysResetOnEntry=true}
   AnimStateTransitionNode_3 = AnimStateTransitionNode()
   AnimStateTransitionNode_36 = AnimStateTransitionNode() {bAutomaticRuleBasedOnSequencePlayerInState=true}
   AnimStateTransitionNode_37 = AnimStateTransitionNode() {bAutomaticRuleBasedOnSequencePlayerInState=true}
   AnimStateTransitionNode_4 = AnimStateTransitionNode()
   AnimStateTransitionNode_5 = AnimStateTransitionNode()
  ]
  links [
   {
    dst=Is Firing
    rule=ItemFullBodyOverride.AnimGraphNode_StateMachine_0.Dodge SM.AnimStateNode_5.In Air.AnimGraphNode_StateMachine_1.Dodge Directional.AnimStateNode_0.Left.AnimGraphNode_BlendSpaceGraph_3.Dodge Left.Left_ForwardRight_AimOffset.AnimGraphNode_StateMachine_0.Right Firing.AnimStateTransitionNode_3.Transition
    src=Empty
    via=AnimStateTransitionNode_3
   }
   {
    dst=Left Fire
    rule=ItemFullBodyOverride.AnimGraphNode_StateMachine_0.Dodge SM.AnimStateNode_5.In Air.AnimGraphNode_StateMachine_1.Dodge Directional.AnimStateNode_0.Left.AnimGraphNode_BlendSpaceGraph_3.Dodge Left.Left_ForwardRight_AimOffset.AnimGraphNode_StateMachine_0.Right Firing.AnimStateTransitionNode_4.Transition
    src=Is Firing
    via=AnimStateTransitionNode_4
   }
   {
    dst=Right Fire
    rule=ItemFullBodyOverride.AnimGraphNode_StateMachine_0.Dodge SM.AnimStateNode_5.In Air.AnimGraphNode_StateMachine_1.Dodge Directional.AnimStateNode_0.Left.AnimGraphNode_BlendSpaceGraph_3.Dodge Left.Left_ForwardRight_AimOffset.AnimGraphNode_StateMachine_0.Right Firing.AnimStateTransitionNode_5.Transition
    src=Is Firing
    via=AnimStateTransitionNode_5
   }
   {
    dst=Is Firing
    rule=ItemFullBodyOverride.AnimGraphNode_StateMachine_0.Dodge SM.AnimStateNode_5.In Air.AnimGraphNode_StateMachine_1.Dodge Directional.AnimStateNode_0.Left.AnimGraphNode_BlendSpaceGraph_3.Dodge Left.Left_ForwardRight_AimOffset.AnimGraphNode_StateMachine_0.Right Firing.AnimStateTransitionNode_36.Transition
    src=Left Fire
    via=AnimStateTransitionNode_36
   }
   {
    dst=Is Firing
    rule=ItemFullBodyOverride.AnimGraphNode_StateMachine_0.Dodge SM.AnimStateNode_5.In Air.AnimGraphNode_StateMachine_1.Dodge Directional.AnimStateNode_0.Left.AnimGraphNode_BlendSpaceGraph_3.Dodge Left.Left_ForwardRight_AimOffset.AnimGraphNode_StateMachine_0.Right Firing.AnimStateTransitionNode_37.Transition
    src=Right Fire
    via=AnimStateTransitionNode_37
   }
  ]
 }
 Is Firing {
  path=ItemFullBodyOverride.AnimGraphNode_StateMachine_0.Dodge SM.AnimStateNode_5.In Air.AnimGraphNode_StateMachine_1.Dodge Directional.AnimStateNode_0.Left.AnimGraphNode_BlendSpaceGraph_3.Dodge Left.Left_ForwardRight_AimOffset.AnimGraphNode_StateMachine_0.Right Firing.AnimStateConduitNode_0.Is Firing
  nodes [Result(bCanEnterTransition=Property Access() -> (Value=~))]
 }
 Left Fire {
  path=ItemFullBodyOverride.AnimGraphNode_StateMachine_0.Dodge SM.AnimStateNode_5.In Air.AnimGraphNode_StateMachine_1.Dodge Directional.AnimStateNode_0.Left.AnimGraphNode_BlendSpaceGraph_3.Dodge Left.Left_ForwardRight_AimOffset.AnimGraphNode_StateMachine_0.Right Firing.AnimStateNode_0.Left Fire
  nodes [
   n1 = "Sequence Player (additive) 'CherrySmoke_DualRevolvers_right_dodge_fire_L'"(PlayRate=Get AdditiveLoopPlayRate()) {bLoopAnimation=false, sequence=&CherrySmoke_DualRevolvers_right_dodge_fire_L}
   Output Animation Pose(Result=@n1) {is=title AnimGraphNode_StateResult}
  ]
 }
 Empty {
  path=ItemFullBodyOverride.AnimGraphNode_StateMachine_0.Dodge SM.AnimStateNode_5.In Air.AnimGraphNode_StateMachine_1.Dodge Directional.AnimStateNode_0.Left.AnimGraphNode_BlendSpaceGraph_3.Dodge Left.Left_ForwardRight_AimOffset.AnimGraphNode_StateMachine_0.Right Firing.AnimStateNode_1.Empty
  nodes [Output Animation Pose(Result=Additive Identity Pose())]
 }
 Right Fire {
  path=ItemFullBodyOverride.AnimGraphNode_StateMachine_0.Dodge SM.AnimStateNode_5.In Air.AnimGraphNode_StateMachine_1.Dodge Directional.AnimStateNode_0.Left.AnimGraphNode_BlendSpaceGraph_3.Dodge Left.Left_ForwardRight_AimOffset.AnimGraphNode_StateMachine_0.Right Firing.AnimStateNode_2.Right Fire
  nodes [
   n1 = "Sequence Player (additive) 'CherrySmoke_DualRevolvers_right_dodge_fire_R'"(PlayRate=Get AdditiveLoopPlayRate()) {bLoopAnimation=false, sequence=&CherrySmoke_DualRevolvers_right_dodge_fire_R}
   Output Animation Pose(Result=@n1)
  ]
 }
 Transition {
  path=ItemFullBodyOverride.AnimGraphNode_StateMachine_0.Dodge SM.AnimStateNode_5.In Air.AnimGraphNode_StateMachine_1.Dodge Directional.AnimStateNode_0.Left.AnimGraphNode_BlendSpaceGraph_3.Dodge Left.Left_ForwardRight_AimOffset.AnimGraphNode_StateMachine_0.Right Firing.AnimStateTransitionNode_3.Transition
  nodes [Result(bCanEnterTransition=true)]
 }
 Transition {
  path=ItemFullBodyOverride.AnimGraphNode_StateMachine_0.Dodge SM.AnimStateNode_5.In Air.AnimGraphNode_StateMachine_1.Dodge Directional.AnimStateNode_0.Left.AnimGraphNode_BlendSpaceGraph_3.Dodge Left.Left_ForwardRight_AimOffset.AnimGraphNode_StateMachine_0.Right Firing.AnimStateTransitionNode_36.Transition
  nodes [Result()]
 }
 Transition {
  path=ItemFullBodyOverride.AnimGraphNode_StateMachine_0.Dodge SM.AnimStateNode_5.In Air.AnimGraphNode_StateMachine_1.Dodge Directional.AnimStateNode_0.Left.AnimGraphNode_BlendSpaceGraph_3.Dodge Left.Left_ForwardRight_AimOffset.AnimGraphNode_StateMachine_0.Right Firing.AnimStateTransitionNode_37.Transition
  nodes [Result()]
 }
 Transition {
  path=ItemFullBodyOverride.AnimGraphNode_StateMachine_0.Dodge SM.AnimStateNode_5.In Air.AnimGraphNode_StateMachine_1.Dodge Directional.AnimStateNode_0.Left.AnimGraphNode_BlendSpaceGraph_3.Dodge Left.Left_ForwardRight_AimOffset.AnimGraphNode_StateMachine_0.Right Firing.AnimStateTransitionNode_4.Transition
  nodes [Result(bCanEnterTransition=NOT Boolean(A=Get bIsLastHandFireLeft()))]
 }
 Transition {
  path=ItemFullBodyOverride.AnimGraphNode_StateMachine_0.Dodge SM.AnimStateNode_5.In Air.AnimGraphNode_StateMachine_1.Dodge Directional.AnimStateNode_0.Left.AnimGraphNode_BlendSpaceGraph_3.Dodge Left.Left_ForwardRight_AimOffset.AnimGraphNode_StateMachine_0.Right Firing.AnimStateTransitionNode_5.Transition
  nodes [Result(bCanEnterTransition=Get bIsLastHandFireLeft())]
 }
 Left_Forward_InAir {
  path=ItemFullBodyOverride.AnimGraphNode_StateMachine_0.Dodge SM.AnimStateNode_5.In Air.AnimGraphNode_StateMachine_1.Dodge Directional.AnimStateNode_0.Left.AnimGraphNode_BlendSpaceGraph_3.Dodge Left.Left_Forward_InAir
  nodes [
   n1 = Sequence Player 'CherrySmoke_DualRevolver_Dodge_LeftSide_F_Diving'() {bLoopAnimation=false, sequence=/CherrySmokeDualRevolversGameplay/Assets/Anim/CharacterAnim/CherrySmoke_DualRevolver_Dodge_LeftSide_F_Diving}
   AnimGraphNode_RotationOffsetBlendSpaceGraph_0 = AimOffset 'LeftSide_Forward_AimOffset'!(X:float=Get AimPitch(), BasePose:Pose Link Structure=@n1, Alpha:float=1) -> (Pose:Pose Link Structure) {is=title AnimGraphNode_RotationOffsetBlendSpaceGraph}
   AnimGraphNode_StateMachine_0 = Forward Firing()
   Output Animation Pose(Result=Apply Additive(Base=@AnimGraphNode_RotationOffsetBlendSpaceGraph_0, Additive=@AnimGraphNode_StateMachine_0) {alphaBoolBlend {blendOption=Linear}}) {is=title AnimGraphNode_BlendSpaceSampleResult}
  ]
 }
 LeftSide_Forward_AimOffset {
  path=ItemFullBodyOverride.AnimGraphNode_StateMachine_0.Dodge SM.AnimStateNode_5.In Air.AnimGraphNode_StateMachine_1.Dodge Directional.AnimStateNode_0.Left.AnimGraphNode_BlendSpaceGraph_3.Dodge Left.Left_Forward_InAir.AnimGraphNode_RotationOffsetBlendSpaceGraph_0.LeftSide_Forward_AimOffset
  links []
  nodes []
 }
 L_F_Center {
  path=ItemFullBodyOverride.AnimGraphNode_StateMachine_0.Dodge SM.AnimStateNode_5.In Air.AnimGraphNode_StateMachine_1.Dodge Directional.AnimStateNode_0.Left.AnimGraphNode_BlendSpaceGraph_3.Dodge Left.Left_Forward_InAir.AnimGraphNode_RotationOffsetBlendSpaceGraph_0.LeftSide_Forward_AimOffset.L_F_Center
  nodes [Output Animation Pose(Result=Additive Identity Pose())]
 }
 L_F_Down {
  path=ItemFullBodyOverride.AnimGraphNode_StateMachine_0.Dodge SM.AnimStateNode_5.In Air.AnimGraphNode_StateMachine_1.Dodge Directional.AnimStateNode_0.Left.AnimGraphNode_BlendSpaceGraph_3.Dodge Left.Left_Forward_InAir.AnimGraphNode_RotationOffsetBlendSpaceGraph_0.LeftSide_Forward_AimOffset.L_F_Down
  nodes [
   n1 = "Sequence Player (additive) 'CherrySmoke_DualRevolvers_Dodge_LeftSide_F_AO_D50'"() {
    bLoopAnimation=false
    sequence=&CherrySmoke_DualRevolvers_Dodge_LeftSide_F_AO_D50
   }
   Output Animation Pose(Result=@n1)
  ]
 }
 L_F_UP {
  path=ItemFullBodyOverride.AnimGraphNode_StateMachine_0.Dodge SM.AnimStateNode_5.In Air.AnimGraphNode_StateMachine_1.Dodge Directional.AnimStateNode_0.Left.AnimGraphNode_BlendSpaceGraph_3.Dodge Left.Left_Forward_InAir.AnimGraphNode_RotationOffsetBlendSpaceGraph_0.LeftSide_Forward_AimOffset.L_F_UP
  nodes [
   n1 = "Sequence Player (additive) 'CherrySmoke_DualRevolvers_Dodge_LeftSide_F_AO_U20'"() {
    bLoopAnimation=false
    sequence=/CherrySmokeDualRevolversGameplay/Assets/Anim/Poses/Dodge_AO/CherrySmoke_DualRevolvers_Dodge_LeftSide_F_AO_U20
   }
   Output Animation Pose(Result=@n1)
  ]
 }
 Forward Firing {
  entry=Empty
  path=ItemFullBodyOverride.AnimGraphNode_StateMachine_0.Dodge SM.AnimStateNode_5.In Air.AnimGraphNode_StateMachine_1.Dodge Directional.AnimStateNode_0.Left.AnimGraphNode_BlendSpaceGraph_3.Dodge Left.Left_Forward_InAir.AnimGraphNode_StateMachine_0.Forward Firing
  ids=id
  nodes [
   AnimStateConduitNode_0 = Is Firing()
   AnimStateNode_0 = Left Fire() {bAlwaysResetOnEntry=true}
   AnimStateNode_1 = Empty()
   AnimStateNode_2 = Right Fire() {bAlwaysResetOnEntry=true}
   AnimStateTransitionNode_0 = AnimStateTransitionNode()
   AnimStateTransitionNode_1 = AnimStateTransitionNode() {crossfadeDuration=0}
   AnimStateTransitionNode_2 = AnimStateTransitionNode() {crossfadeDuration=0}
   AnimStateTransitionNode_3 = AnimStateTransitionNode() {bAutomaticRuleBasedOnSequencePlayerInState=true}
   AnimStateTransitionNode_4 = AnimStateTransitionNode() {bAutomaticRuleBasedOnSequencePlayerInState=true}
  ]
  links [
   {
    dst=Is Firing
    rule=ItemFullBodyOverride.AnimGraphNode_StateMachine_0.Dodge SM.AnimStateNode_5.In Air.AnimGraphNode_StateMachine_1.Dodge Directional.AnimStateNode_0.Left.AnimGraphNode_BlendSpaceGraph_3.Dodge Left.Left_Forward_InAir.AnimGraphNode_StateMachine_0.Forward Firing.AnimStateTransitionNode_0.Transition
    src=Empty
    via=AnimStateTransitionNode_0
   }
   {
    dst=Left Fire
    rule=ItemFullBodyOverride.AnimGraphNode_StateMachine_0.Dodge SM.AnimStateNode_5.In Air.AnimGraphNode_StateMachine_1.Dodge Directional.AnimStateNode_0.Left.AnimGraphNode_BlendSpaceGraph_3.Dodge Left.Left_Forward_InAir.AnimGraphNode_StateMachine_0.Forward Firing.AnimStateTransitionNode_1.Transition
    src=Is Firing
    via=AnimStateTransitionNode_1
   }
   {
    dst=Right Fire
    rule=ItemFullBodyOverride.AnimGraphNode_StateMachine_0.Dodge SM.AnimStateNode_5.In Air.AnimGraphNode_StateMachine_1.Dodge Directional.AnimStateNode_0.Left.AnimGraphNode_BlendSpaceGraph_3.Dodge Left.Left_Forward_InAir.AnimGraphNode_StateMachine_0.Forward Firing.AnimStateTransitionNode_2.Transition
    src=Is Firing
    via=AnimStateTransitionNode_2
   }
   {
    dst=Is Firing
    rule=ItemFullBodyOverride.AnimGraphNode_StateMachine_0.Dodge SM.AnimStateNode_5.In Air.AnimGraphNode_StateMachine_1.Dodge Directional.AnimStateNode_0.Left.AnimGraphNode_BlendSpaceGraph_3.Dodge Left.Left_Forward_InAir.AnimGraphNode_StateMachine_0.Forward Firing.AnimStateTransitionNode_3.Transition
    src=Left Fire
    via=AnimStateTransitionNode_3
   }
   {
    dst=Is Firing
    rule=ItemFullBodyOverride.AnimGraphNode_StateMachine_0.Dodge SM.AnimStateNode_5.In Air.AnimGraphNode_StateMachine_1.Dodge Directional.AnimStateNode_0.Left.AnimGraphNode_BlendSpaceGraph_3.Dodge Left.Left_Forward_InAir.AnimGraphNode_StateMachine_0.Forward Firing.AnimStateTransitionNode_4.Transition
    src=Right Fire
    via=AnimStateTransitionNode_4
   }
  ]
 }
 Is Firing {
  path=ItemFullBodyOverride.AnimGraphNode_StateMachine_0.Dodge SM.AnimStateNode_5.In Air.AnimGraphNode_StateMachine_1.Dodge Directional.AnimStateNode_0.Left.AnimGraphNode_BlendSpaceGraph_3.Dodge Left.Left_Forward_InAir.AnimGraphNode_StateMachine_0.Forward Firing.AnimStateConduitNode_0.Is Firing
  nodes [Result(bCanEnterTransition=Property Access() -> (Value=~))]
 }
 Left Fire {
  path=ItemFullBodyOverride.AnimGraphNode_StateMachine_0.Dodge SM.AnimStateNode_5.In Air.AnimGraphNode_StateMachine_1.Dodge Directional.AnimStateNode_0.Left.AnimGraphNode_BlendSpaceGraph_3.Dodge Left.Left_Forward_InAir.AnimGraphNode_StateMachine_0.Forward Firing.AnimStateNode_0.Left Fire
  nodes [
   n1 = "Sequence Player (additive) 'CherrySmoke_DualRevolvers_front_dodge_fire_L'"(PlayRate=Get AdditiveLoopPlayRate()) {bLoopAnimation=false, sequence=&CherrySmoke_DualRevolvers_front_dodge_fire_L}
   Output Animation Pose(Result=@n1) {is=title AnimGraphNode_StateResult}
  ]
 }
 Empty {
  path=ItemFullBodyOverride.AnimGraphNode_StateMachine_0.Dodge SM.AnimStateNode_5.In Air.AnimGraphNode_StateMachine_1.Dodge Directional.AnimStateNode_0.Left.AnimGraphNode_BlendSpaceGraph_3.Dodge Left.Left_Forward_InAir.AnimGraphNode_StateMachine_0.Forward Firing.AnimStateNode_1.Empty
  nodes [Output Animation Pose(Result=Additive Identity Pose())]
 }
 Right Fire {
  path=ItemFullBodyOverride.AnimGraphNode_StateMachine_0.Dodge SM.AnimStateNode_5.In Air.AnimGraphNode_StateMachine_1.Dodge Directional.AnimStateNode_0.Left.AnimGraphNode_BlendSpaceGraph_3.Dodge Left.Left_Forward_InAir.AnimGraphNode_StateMachine_0.Forward Firing.AnimStateNode_2.Right Fire
  nodes [
   n1 = "Sequence Player (additive) 'CherrySmoke_DualRevolvers_front_dodge_fire_R'"(PlayRate=Get AdditiveLoopPlayRate()) {bLoopAnimation=false, sequence=&CherrySmoke_DualRevolvers_front_dodge_fire_R}
   Output Animation Pose(Result=@n1)
  ]
 }
 Transition {
  path=ItemFullBodyOverride.AnimGraphNode_StateMachine_0.Dodge SM.AnimStateNode_5.In Air.AnimGraphNode_StateMachine_1.Dodge Directional.AnimStateNode_0.Left.AnimGraphNode_BlendSpaceGraph_3.Dodge Left.Left_Forward_InAir.AnimGraphNode_StateMachine_0.Forward Firing.AnimStateTransitionNode_0.Transition
  nodes [Result(bCanEnterTransition=true)]
 }
 Transition {
  path=ItemFullBodyOverride.AnimGraphNode_StateMachine_0.Dodge SM.AnimStateNode_5.In Air.AnimGraphNode_StateMachine_1.Dodge Directional.AnimStateNode_0.Left.AnimGraphNode_BlendSpaceGraph_3.Dodge Left.Left_Forward_InAir.AnimGraphNode_StateMachine_0.Forward Firing.AnimStateTransitionNode_1.Transition
  nodes [Result(bCanEnterTransition=NOT Boolean(A=Get bIsLastHandFireLeft()))]
 }
 Transition {
  path=ItemFullBodyOverride.AnimGraphNode_StateMachine_0.Dodge SM.AnimStateNode_5.In Air.AnimGraphNode_StateMachine_1.Dodge Directional.AnimStateNode_0.Left.AnimGraphNode_BlendSpaceGraph_3.Dodge Left.Left_Forward_InAir.AnimGraphNode_StateMachine_0.Forward Firing.AnimStateTransitionNode_2.Transition
  nodes [Result(bCanEnterTransition=Get bIsLastHandFireLeft())]
 }
 Transition {
  path=ItemFullBodyOverride.AnimGraphNode_StateMachine_0.Dodge SM.AnimStateNode_5.In Air.AnimGraphNode_StateMachine_1.Dodge Directional.AnimStateNode_0.Left.AnimGraphNode_BlendSpaceGraph_3.Dodge Left.Left_Forward_InAir.AnimGraphNode_StateMachine_0.Forward Firing.AnimStateTransitionNode_3.Transition
  nodes [Result()]
 }
 Transition {
  path=ItemFullBodyOverride.AnimGraphNode_StateMachine_0.Dodge SM.AnimStateNode_5.In Air.AnimGraphNode_StateMachine_1.Dodge Directional.AnimStateNode_0.Left.AnimGraphNode_BlendSpaceGraph_3.Dodge Left.Left_Forward_InAir.AnimGraphNode_StateMachine_0.Forward Firing.AnimStateTransitionNode_4.Transition
  nodes [Result()]
 }
 Left_Left_InAir {
  path=ItemFullBodyOverride.AnimGraphNode_StateMachine_0.Dodge SM.AnimStateNode_5.In Air.AnimGraphNode_StateMachine_1.Dodge Directional.AnimStateNode_0.Left.AnimGraphNode_BlendSpaceGraph_3.Dodge Left.Left_Left_InAir
  nodes [
   n1 = Sequence Player 'CherrySmoke_DualRevolver_Dodge_LeftSide_L_Diving'() {bLoopAnimation=false, sequence=/CherrySmokeDualRevolversGameplay/Assets/Anim/CharacterAnim/CherrySmoke_DualRevolver_Dodge_LeftSide_L_Diving}
   AnimGraphNode_RotationOffsetBlendSpaceGraph_0 = AimOffset 'LeftSide_Left_AimOffset'!(X:float=Get AimPitch(), BasePose:Pose Link Structure=@n1, Alpha:float=1) -> (Pose:Pose Link Structure) {is=title AnimGraphNode_RotationOffsetBlendSpaceGraph}
   AnimGraphNode_StateMachine_0 = Left Firing()
   Output Animation Pose(Result=Apply Additive(Base=@AnimGraphNode_RotationOffsetBlendSpaceGraph_0, Additive=@AnimGraphNode_StateMachine_0) {alphaBoolBlend {blendOption=Linear}}) {is=title AnimGraphNode_BlendSpaceSampleResult}
  ]
 }
 LeftSide_Left_AimOffset {
  path=ItemFullBodyOverride.AnimGraphNode_StateMachine_0.Dodge SM.AnimStateNode_5.In Air.AnimGraphNode_StateMachine_1.Dodge Directional.AnimStateNode_0.Left.AnimGraphNode_BlendSpaceGraph_3.Dodge Left.Left_Left_InAir.AnimGraphNode_RotationOffsetBlendSpaceGraph_0.LeftSide_Left_AimOffset
  links []
  nodes []
 }
 L_L_Center {
  path=ItemFullBodyOverride.AnimGraphNode_StateMachine_0.Dodge SM.AnimStateNode_5.In Air.AnimGraphNode_StateMachine_1.Dodge Directional.AnimStateNode_0.Left.AnimGraphNode_BlendSpaceGraph_3.Dodge Left.Left_Left_InAir.AnimGraphNode_RotationOffsetBlendSpaceGraph_0.LeftSide_Left_AimOffset.L_L_Center
  nodes [Output Animation Pose(Result=Additive Identity Pose())]
 }
 L_L_Down {
  path=ItemFullBodyOverride.AnimGraphNode_StateMachine_0.Dodge SM.AnimStateNode_5.In Air.AnimGraphNode_StateMachine_1.Dodge Directional.AnimStateNode_0.Left.AnimGraphNode_BlendSpaceGraph_3.Dodge Left.Left_Left_InAir.AnimGraphNode_RotationOffsetBlendSpaceGraph_0.LeftSide_Left_AimOffset.L_L_Down
  nodes [
   n1 = "Sequence Player (additive) 'CherrySmoke_DualRevolvers_Dodge_LeftSide_L_AO_D50'"() {
    bLoopAnimation=false
    sequence=/CherrySmokeDualRevolversGameplay/Assets/Anim/Poses/Dodge_AO/CherrySmoke_DualRevolvers_Dodge_LeftSide_L_AO_D50
   }
   Output Animation Pose(Result=@n1)
  ]
 }
 L_L_UP {
  path=ItemFullBodyOverride.AnimGraphNode_StateMachine_0.Dodge SM.AnimStateNode_5.In Air.AnimGraphNode_StateMachine_1.Dodge Directional.AnimStateNode_0.Left.AnimGraphNode_BlendSpaceGraph_3.Dodge Left.Left_Left_InAir.AnimGraphNode_RotationOffsetBlendSpaceGraph_0.LeftSide_Left_AimOffset.L_L_UP
  nodes [
   n1 = "Sequence Player (additive) 'CherrySmoke_DualRevolvers_Dodge_LeftSide_L_AO_U50'"() {
    bLoopAnimation=false
    sequence=/CherrySmokeDualRevolversGameplay/Assets/Anim/Poses/Dodge_AO/CherrySmoke_DualRevolvers_Dodge_LeftSide_L_AO_U50
   }
   Output Animation Pose(Result=@n1)
  ]
 }
 Left Firing {
  entry=Empty
  path=ItemFullBodyOverride.AnimGraphNode_StateMachine_0.Dodge SM.AnimStateNode_5.In Air.AnimGraphNode_StateMachine_1.Dodge Directional.AnimStateNode_0.Left.AnimGraphNode_BlendSpaceGraph_3.Dodge Left.Left_Left_InAir.AnimGraphNode_StateMachine_0.Left Firing
  ids=id
  nodes [
   AnimStateConduitNode_0 = Is Firing()
   AnimStateNode_0 = Left Fire() {bAlwaysResetOnEntry=true}
   AnimStateNode_1 = Empty()
   AnimStateNode_2 = Right Fire() {bAlwaysResetOnEntry=true}
   AnimStateTransitionNode_0 = AnimStateTransitionNode()
   AnimStateTransitionNode_1 = AnimStateTransitionNode() {crossfadeDuration=0.05}
   AnimStateTransitionNode_10 = AnimStateTransitionNode() {bAutomaticRuleBasedOnSequencePlayerInState=true}
   AnimStateTransitionNode_11 = AnimStateTransitionNode() {bAutomaticRuleBasedOnSequencePlayerInState=true}
   AnimStateTransitionNode_2 = AnimStateTransitionNode() {crossfadeDuration=0.05}
  ]
  links [
   {
    dst=Is Firing
    rule=ItemFullBodyOverride.AnimGraphNode_StateMachine_0.Dodge SM.AnimStateNode_5.In Air.AnimGraphNode_StateMachine_1.Dodge Directional.AnimStateNode_0.Left.AnimGraphNode_BlendSpaceGraph_3.Dodge Left.Left_Left_InAir.AnimGraphNode_StateMachine_0.Left Firing.AnimStateTransitionNode_0.Transition
    src=Empty
    via=AnimStateTransitionNode_0
   }
   {
    dst=Left Fire
    rule=ItemFullBodyOverride.AnimGraphNode_StateMachine_0.Dodge SM.AnimStateNode_5.In Air.AnimGraphNode_StateMachine_1.Dodge Directional.AnimStateNode_0.Left.AnimGraphNode_BlendSpaceGraph_3.Dodge Left.Left_Left_InAir.AnimGraphNode_StateMachine_0.Left Firing.AnimStateTransitionNode_1.Transition
    src=Is Firing
    via=AnimStateTransitionNode_1
   }
   {
    dst=Right Fire
    rule=ItemFullBodyOverride.AnimGraphNode_StateMachine_0.Dodge SM.AnimStateNode_5.In Air.AnimGraphNode_StateMachine_1.Dodge Directional.AnimStateNode_0.Left.AnimGraphNode_BlendSpaceGraph_3.Dodge Left.Left_Left_InAir.AnimGraphNode_StateMachine_0.Left Firing.AnimStateTransitionNode_2.Transition
    src=Is Firing
    via=AnimStateTransitionNode_2
   }
   {
    dst=Is Firing
    rule=ItemFullBodyOverride.AnimGraphNode_StateMachine_0.Dodge SM.AnimStateNode_5.In Air.AnimGraphNode_StateMachine_1.Dodge Directional.AnimStateNode_0.Left.AnimGraphNode_BlendSpaceGraph_3.Dodge Left.Left_Left_InAir.AnimGraphNode_StateMachine_0.Left Firing.AnimStateTransitionNode_10.Transition
    src=Left Fire
    via=AnimStateTransitionNode_10
   }
   {
    dst=Is Firing
    rule=ItemFullBodyOverride.AnimGraphNode_StateMachine_0.Dodge SM.AnimStateNode_5.In Air.AnimGraphNode_StateMachine_1.Dodge Directional.AnimStateNode_0.Left.AnimGraphNode_BlendSpaceGraph_3.Dodge Left.Left_Left_InAir.AnimGraphNode_StateMachine_0.Left Firing.AnimStateTransitionNode_11.Transition
    src=Right Fire
    via=AnimStateTransitionNode_11
   }
  ]
 }
 Is Firing {
  path=ItemFullBodyOverride.AnimGraphNode_StateMachine_0.Dodge SM.AnimStateNode_5.In Air.AnimGraphNode_StateMachine_1.Dodge Directional.AnimStateNode_0.Left.AnimGraphNode_BlendSpaceGraph_3.Dodge Left.Left_Left_InAir.AnimGraphNode_StateMachine_0.Left Firing.AnimStateConduitNode_0.Is Firing
  nodes [Result(bCanEnterTransition=Property Access() -> (Value=~))]
 }
 Left Fire {
  path=ItemFullBodyOverride.AnimGraphNode_StateMachine_0.Dodge SM.AnimStateNode_5.In Air.AnimGraphNode_StateMachine_1.Dodge Directional.AnimStateNode_0.Left.AnimGraphNode_BlendSpaceGraph_3.Dodge Left.Left_Left_InAir.AnimGraphNode_StateMachine_0.Left Firing.AnimStateNode_0.Left Fire
  nodes [
   n1 = "Sequence Player (additive) 'CherrySmoke_DualRevolvers_left_dodge_fire_L'"(PlayRate=Get AdditiveLoopPlayRate()) {bLoopAnimation=false, sequence=&CherrySmoke_DualRevolvers_left_dodge_fire_L}
   Output Animation Pose(Result=@n1) {is=title AnimGraphNode_StateResult}
  ]
 }
 Empty {
  path=ItemFullBodyOverride.AnimGraphNode_StateMachine_0.Dodge SM.AnimStateNode_5.In Air.AnimGraphNode_StateMachine_1.Dodge Directional.AnimStateNode_0.Left.AnimGraphNode_BlendSpaceGraph_3.Dodge Left.Left_Left_InAir.AnimGraphNode_StateMachine_0.Left Firing.AnimStateNode_1.Empty
  nodes [Output Animation Pose(Result=Additive Identity Pose())]
 }
 Right Fire {
  path=ItemFullBodyOverride.AnimGraphNode_StateMachine_0.Dodge SM.AnimStateNode_5.In Air.AnimGraphNode_StateMachine_1.Dodge Directional.AnimStateNode_0.Left.AnimGraphNode_BlendSpaceGraph_3.Dodge Left.Left_Left_InAir.AnimGraphNode_StateMachine_0.Left Firing.AnimStateNode_2.Right Fire
  nodes [
   n1 = "Sequence Player (additive) 'CherrySmoke_DualRevolvers_left_dodge_fire_R'"(PlayRate=Get AdditiveLoopPlayRate()) {bLoopAnimation=false, sequence=&CherrySmoke_DualRevolvers_left_dodge_fire_R}
   Output Animation Pose(Result=@n1)
  ]
 }
 Transition {
  path=ItemFullBodyOverride.AnimGraphNode_StateMachine_0.Dodge SM.AnimStateNode_5.In Air.AnimGraphNode_StateMachine_1.Dodge Directional.AnimStateNode_0.Left.AnimGraphNode_BlendSpaceGraph_3.Dodge Left.Left_Left_InAir.AnimGraphNode_StateMachine_0.Left Firing.AnimStateTransitionNode_0.Transition
  nodes [Result(bCanEnterTransition=true)]
 }
 Transition {
  path=ItemFullBodyOverride.AnimGraphNode_StateMachine_0.Dodge SM.AnimStateNode_5.In Air.AnimGraphNode_StateMachine_1.Dodge Directional.AnimStateNode_0.Left.AnimGraphNode_BlendSpaceGraph_3.Dodge Left.Left_Left_InAir.AnimGraphNode_StateMachine_0.Left Firing.AnimStateTransitionNode_1.Transition
  nodes [Result(bCanEnterTransition=NOT Boolean(A=Get bIsLastHandFireLeft()))]
 }
 Transition {
  path=ItemFullBodyOverride.AnimGraphNode_StateMachine_0.Dodge SM.AnimStateNode_5.In Air.AnimGraphNode_StateMachine_1.Dodge Directional.AnimStateNode_0.Left.AnimGraphNode_BlendSpaceGraph_3.Dodge Left.Left_Left_InAir.AnimGraphNode_StateMachine_0.Left Firing.AnimStateTransitionNode_10.Transition
  nodes [Result()]
 }
 Transition {
  path=ItemFullBodyOverride.AnimGraphNode_StateMachine_0.Dodge SM.AnimStateNode_5.In Air.AnimGraphNode_StateMachine_1.Dodge Directional.AnimStateNode_0.Left.AnimGraphNode_BlendSpaceGraph_3.Dodge Left.Left_Left_InAir.AnimGraphNode_StateMachine_0.Left Firing.AnimStateTransitionNode_11.Transition
  nodes [Result()]
 }
 Transition {
  path=ItemFullBodyOverride.AnimGraphNode_StateMachine_0.Dodge SM.AnimStateNode_5.In Air.AnimGraphNode_StateMachine_1.Dodge Directional.AnimStateNode_0.Left.AnimGraphNode_BlendSpaceGraph_3.Dodge Left.Left_Left_InAir.AnimGraphNode_StateMachine_0.Left Firing.AnimStateTransitionNode_2.Transition
  nodes [Result(bCanEnterTransition=Get bIsLastHandFireLeft())]
 }
 Left_Right_InAir {
  path=ItemFullBodyOverride.AnimGraphNode_StateMachine_0.Dodge SM.AnimStateNode_5.In Air.AnimGraphNode_StateMachine_1.Dodge Directional.AnimStateNode_0.Left.AnimGraphNode_BlendSpaceGraph_3.Dodge Left.Left_Right_InAir
  nodes [
   n1 = Sequence Player 'CherrySmoke_DualRevolver_Dodge_LeftSide_R_Diving'() {bLoopAnimation=false, sequence=/CherrySmokeDualRevolversGameplay/Assets/Anim/CharacterAnim/CherrySmoke_DualRevolver_Dodge_LeftSide_R_Diving}
   AnimGraphNode_RotationOffsetBlendSpaceGraph_0 = AimOffset 'Left_Side_Right_AimOffset'!(X:float=Get AimPitch(), BasePose:Pose Link Structure=@n1, Alpha:float=1) -> (Pose:Pose Link Structure) {is=title AnimGraphNode_RotationOffsetBlendSpaceGraph}
   AnimGraphNode_StateMachine_0 = Right Firing()
   Output Animation Pose(Result=Apply Additive(Base=@AnimGraphNode_RotationOffsetBlendSpaceGraph_0, Additive=@AnimGraphNode_StateMachine_0) {alphaBoolBlend {blendOption=Linear}}) {is=title AnimGraphNode_BlendSpaceSampleResult}
  ]
 }
 Left_Side_Right_AimOffset {
  path=ItemFullBodyOverride.AnimGraphNode_StateMachine_0.Dodge SM.AnimStateNode_5.In Air.AnimGraphNode_StateMachine_1.Dodge Directional.AnimStateNode_0.Left.AnimGraphNode_BlendSpaceGraph_3.Dodge Left.Left_Right_InAir.AnimGraphNode_RotationOffsetBlendSpaceGraph_0.Left_Side_Right_AimOffset
  links []
  nodes []
 }
 L_R_Center {
  path=ItemFullBodyOverride.AnimGraphNode_StateMachine_0.Dodge SM.AnimStateNode_5.In Air.AnimGraphNode_StateMachine_1.Dodge Directional.AnimStateNode_0.Left.AnimGraphNode_BlendSpaceGraph_3.Dodge Left.Left_Right_InAir.AnimGraphNode_RotationOffsetBlendSpaceGraph_0.Left_Side_Right_AimOffset.L_R_Center
  nodes [Output Animation Pose(Result=Additive Identity Pose())]
 }
 L_R_Down {
  path=ItemFullBodyOverride.AnimGraphNode_StateMachine_0.Dodge SM.AnimStateNode_5.In Air.AnimGraphNode_StateMachine_1.Dodge Directional.AnimStateNode_0.Left.AnimGraphNode_BlendSpaceGraph_3.Dodge Left.Left_Right_InAir.AnimGraphNode_RotationOffsetBlendSpaceGraph_0.Left_Side_Right_AimOffset.L_R_Down
  nodes [
   n1 = "Sequence Player (additive) 'CherrySmoke_DualRevolvers_Dodge_LeftSide_R_AO_D50'"() {
    bLoopAnimation=false
    sequence=/CherrySmokeDualRevolversGameplay/Assets/Anim/Poses/Dodge_AO/CherrySmoke_DualRevolvers_Dodge_LeftSide_R_AO_D50
   }
   Output Animation Pose(Result=@n1)
  ]
 }
 L_R_UP {
  path=ItemFullBodyOverride.AnimGraphNode_StateMachine_0.Dodge SM.AnimStateNode_5.In Air.AnimGraphNode_StateMachine_1.Dodge Directional.AnimStateNode_0.Left.AnimGraphNode_BlendSpaceGraph_3.Dodge Left.Left_Right_InAir.AnimGraphNode_RotationOffsetBlendSpaceGraph_0.Left_Side_Right_AimOffset.L_R_UP
  nodes [
   n1 = "Sequence Player (additive) 'CherrySmoke_DualRevolvers_Dodge_LeftSide_R_AO_U50'"() {
    bLoopAnimation=false
    sequence=/CherrySmokeDualRevolversGameplay/Assets/Anim/Poses/Dodge_AO/CherrySmoke_DualRevolvers_Dodge_LeftSide_R_AO_U50
   }
   Output Animation Pose(Result=@n1)
  ]
 }
 Right Firing {
  entry=Empty
  path=ItemFullBodyOverride.AnimGraphNode_StateMachine_0.Dodge SM.AnimStateNode_5.In Air.AnimGraphNode_StateMachine_1.Dodge Directional.AnimStateNode_0.Left.AnimGraphNode_BlendSpaceGraph_3.Dodge Left.Left_Right_InAir.AnimGraphNode_StateMachine_0.Right Firing
  ids=id
  nodes [
   AnimStateConduitNode_0 = Is Firing()
   AnimStateNode_0 = Left Fire() {bAlwaysResetOnEntry=true}
   AnimStateNode_1 = Empty()
   AnimStateNode_2 = Right Fire() {bAlwaysResetOnEntry=true}
   AnimStateTransitionNode_3 = AnimStateTransitionNode()
   AnimStateTransitionNode_4 = AnimStateTransitionNode() {crossfadeDuration=0.05}
   AnimStateTransitionNode_5 = AnimStateTransitionNode() {crossfadeDuration=0.05}
   AnimStateTransitionNode_8 = AnimStateTransitionNode() {bAutomaticRuleBasedOnSequencePlayerInState=true}
   AnimStateTransitionNode_9 = AnimStateTransitionNode() {bAutomaticRuleBasedOnSequencePlayerInState=true}
  ]
  links [
   {
    dst=Is Firing
    rule=ItemFullBodyOverride.AnimGraphNode_StateMachine_0.Dodge SM.AnimStateNode_5.In Air.AnimGraphNode_StateMachine_1.Dodge Directional.AnimStateNode_0.Left.AnimGraphNode_BlendSpaceGraph_3.Dodge Left.Left_Right_InAir.AnimGraphNode_StateMachine_0.Right Firing.AnimStateTransitionNode_3.Transition
    src=Empty
    via=AnimStateTransitionNode_3
   }
   {
    dst=Left Fire
    rule=ItemFullBodyOverride.AnimGraphNode_StateMachine_0.Dodge SM.AnimStateNode_5.In Air.AnimGraphNode_StateMachine_1.Dodge Directional.AnimStateNode_0.Left.AnimGraphNode_BlendSpaceGraph_3.Dodge Left.Left_Right_InAir.AnimGraphNode_StateMachine_0.Right Firing.AnimStateTransitionNode_4.Transition
    src=Is Firing
    via=AnimStateTransitionNode_4
   }
   {
    dst=Right Fire
    rule=ItemFullBodyOverride.AnimGraphNode_StateMachine_0.Dodge SM.AnimStateNode_5.In Air.AnimGraphNode_StateMachine_1.Dodge Directional.AnimStateNode_0.Left.AnimGraphNode_BlendSpaceGraph_3.Dodge Left.Left_Right_InAir.AnimGraphNode_StateMachine_0.Right Firing.AnimStateTransitionNode_5.Transition
    src=Is Firing
    via=AnimStateTransitionNode_5
   }
   {
    dst=Is Firing
    rule=ItemFullBodyOverride.AnimGraphNode_StateMachine_0.Dodge SM.AnimStateNode_5.In Air.AnimGraphNode_StateMachine_1.Dodge Directional.AnimStateNode_0.Left.AnimGraphNode_BlendSpaceGraph_3.Dodge Left.Left_Right_InAir.AnimGraphNode_StateMachine_0.Right Firing.AnimStateTransitionNode_8.Transition
    src=Left Fire
    via=AnimStateTransitionNode_8
   }
   {
    dst=Is Firing
    rule=ItemFullBodyOverride.AnimGraphNode_StateMachine_0.Dodge SM.AnimStateNode_5.In Air.AnimGraphNode_StateMachine_1.Dodge Directional.AnimStateNode_0.Left.AnimGraphNode_BlendSpaceGraph_3.Dodge Left.Left_Right_InAir.AnimGraphNode_StateMachine_0.Right Firing.AnimStateTransitionNode_9.Transition
    src=Right Fire
    via=AnimStateTransitionNode_9
   }
  ]
 }
 Is Firing {
  path=ItemFullBodyOverride.AnimGraphNode_StateMachine_0.Dodge SM.AnimStateNode_5.In Air.AnimGraphNode_StateMachine_1.Dodge Directional.AnimStateNode_0.Left.AnimGraphNode_BlendSpaceGraph_3.Dodge Left.Left_Right_InAir.AnimGraphNode_StateMachine_0.Right Firing.AnimStateConduitNode_0.Is Firing
  nodes [Result(bCanEnterTransition=Property Access() -> (Value=~))]
 }
 Left Fire {
  path=ItemFullBodyOverride.AnimGraphNode_StateMachine_0.Dodge SM.AnimStateNode_5.In Air.AnimGraphNode_StateMachine_1.Dodge Directional.AnimStateNode_0.Left.AnimGraphNode_BlendSpaceGraph_3.Dodge Left.Left_Right_InAir.AnimGraphNode_StateMachine_0.Right Firing.AnimStateNode_0.Left Fire
  nodes [
   n1 = "Sequence Player (additive) 'CherrySmoke_DualRevolvers_right_dodge_fire_L'"(PlayRate=Get AdditiveLoopPlayRate()) {bLoopAnimation=false, sequence=&CherrySmoke_DualRevolvers_right_dodge_fire_L}
   Output Animation Pose(Result=@n1) {is=title AnimGraphNode_StateResult}
  ]
 }
 Empty {
  path=ItemFullBodyOverride.AnimGraphNode_StateMachine_0.Dodge SM.AnimStateNode_5.In Air.AnimGraphNode_StateMachine_1.Dodge Directional.AnimStateNode_0.Left.AnimGraphNode_BlendSpaceGraph_3.Dodge Left.Left_Right_InAir.AnimGraphNode_StateMachine_0.Right Firing.AnimStateNode_1.Empty
  nodes [Output Animation Pose(Result=Additive Identity Pose())]
 }
 Right Fire {
  path=ItemFullBodyOverride.AnimGraphNode_StateMachine_0.Dodge SM.AnimStateNode_5.In Air.AnimGraphNode_StateMachine_1.Dodge Directional.AnimStateNode_0.Left.AnimGraphNode_BlendSpaceGraph_3.Dodge Left.Left_Right_InAir.AnimGraphNode_StateMachine_0.Right Firing.AnimStateNode_2.Right Fire
  nodes [
   n1 = "Sequence Player (additive) 'CherrySmoke_DualRevolvers_right_dodge_fire_R'"(PlayRate=Get AdditiveLoopPlayRate()) {bLoopAnimation=false, sequence=&CherrySmoke_DualRevolvers_right_dodge_fire_R}
   Output Animation Pose(Result=@n1)
  ]
 }
 Transition {
  path=ItemFullBodyOverride.AnimGraphNode_StateMachine_0.Dodge SM.AnimStateNode_5.In Air.AnimGraphNode_StateMachine_1.Dodge Directional.AnimStateNode_0.Left.AnimGraphNode_BlendSpaceGraph_3.Dodge Left.Left_Right_InAir.AnimGraphNode_StateMachine_0.Right Firing.AnimStateTransitionNode_3.Transition
  nodes [Result(bCanEnterTransition=true)]
 }
 Transition {
  path=ItemFullBodyOverride.AnimGraphNode_StateMachine_0.Dodge SM.AnimStateNode_5.In Air.AnimGraphNode_StateMachine_1.Dodge Directional.AnimStateNode_0.Left.AnimGraphNode_BlendSpaceGraph_3.Dodge Left.Left_Right_InAir.AnimGraphNode_StateMachine_0.Right Firing.AnimStateTransitionNode_4.Transition
  nodes [Result(bCanEnterTransition=NOT Boolean(A=Get bIsLastHandFireLeft()))]
 }
 Transition {
  path=ItemFullBodyOverride.AnimGraphNode_StateMachine_0.Dodge SM.AnimStateNode_5.In Air.AnimGraphNode_StateMachine_1.Dodge Directional.AnimStateNode_0.Left.AnimGraphNode_BlendSpaceGraph_3.Dodge Left.Left_Right_InAir.AnimGraphNode_StateMachine_0.Right Firing.AnimStateTransitionNode_5.Transition
  nodes [Result(bCanEnterTransition=Get bIsLastHandFireLeft())]
 }
 Transition {
  path=ItemFullBodyOverride.AnimGraphNode_StateMachine_0.Dodge SM.AnimStateNode_5.In Air.AnimGraphNode_StateMachine_1.Dodge Directional.AnimStateNode_0.Left.AnimGraphNode_BlendSpaceGraph_3.Dodge Left.Left_Right_InAir.AnimGraphNode_StateMachine_0.Right Firing.AnimStateTransitionNode_8.Transition
  nodes [Result()]
 }
 Transition {
  path=ItemFullBodyOverride.AnimGraphNode_StateMachine_0.Dodge SM.AnimStateNode_5.In Air.AnimGraphNode_StateMachine_1.Dodge Directional.AnimStateNode_0.Left.AnimGraphNode_BlendSpaceGraph_3.Dodge Left.Left_Right_InAir.AnimGraphNode_StateMachine_0.Right Firing.AnimStateTransitionNode_9.Transition
  nodes [Result()]
 }
 Right {
  path=ItemFullBodyOverride.AnimGraphNode_StateMachine_0.Dodge SM.AnimStateNode_5.In Air.AnimGraphNode_StateMachine_1.Dodge Directional.AnimStateNode_1.Right
  nodes [
   AnimGraphNode_BlendSpaceGraph_0 = Blendspace 'Dodge Right'!(X:float=Get DodgeAngleRadians()) -> (Pose:Pose Link Structure) {is=title AnimGraphNode_BlendSpaceGraph, tag=DodgeLoopRight, groupName=DodgeLoop}
   Output Animation Pose(Result=@AnimGraphNode_BlendSpaceGraph_0)
  ]
 }
 Dodge Right {
  path=ItemFullBodyOverride.AnimGraphNode_StateMachine_0.Dodge SM.AnimStateNode_5.In Air.AnimGraphNode_StateMachine_1.Dodge Directional.AnimStateNode_1.Right.AnimGraphNode_BlendSpaceGraph_0.Dodge Right
  links []
  nodes []
 }
 Right_BackLeft_InAir {
  path=ItemFullBodyOverride.AnimGraphNode_StateMachine_0.Dodge SM.AnimStateNode_5.In Air.AnimGraphNode_StateMachine_1.Dodge Directional.AnimStateNode_1.Right.AnimGraphNode_BlendSpaceGraph_0.Dodge Right.Right_BackLeft_InAir
  nodes [
   n1 = Sequence Player 'CherrySmoke_DualRevolvers_Dodge_RightSide_BL_Pose'() {bLoopAnimation=false, sequence=/CherrySmokeDualRevolversGameplay/Assets/Anim/Poses/CherrySmoke_DualRevolvers_Dodge_RightSide_BL_Pose}
   AnimGraphNode_RotationOffsetBlendSpaceGraph_0 = AimOffset 'RightSide_BackLeft_AimOffset'!(X:float=Get AimPitch(), BasePose:Pose Link Structure=@n1, Alpha:float=1) -> (Pose:Pose Link Structure) {is=title AnimGraphNode_RotationOffsetBlendSpaceGraph}
   AnimGraphNode_StateMachine_0 = Left Firing()
   Output Animation Pose(Result=Apply Additive(Base=@AnimGraphNode_RotationOffsetBlendSpaceGraph_0, Additive=@AnimGraphNode_StateMachine_0) {alphaBoolBlend {blendOption=Linear}}) {is=title AnimGraphNode_BlendSpaceSampleResult}
  ]
 }
 RightSide_BackLeft_AimOffset {
  path=ItemFullBodyOverride.AnimGraphNode_StateMachine_0.Dodge SM.AnimStateNode_5.In Air.AnimGraphNode_StateMachine_1.Dodge Directional.AnimStateNode_1.Right.AnimGraphNode_BlendSpaceGraph_0.Dodge Right.Right_BackLeft_InAir.AnimGraphNode_RotationOffsetBlendSpaceGraph_0.RightSide_BackLeft_AimOffset
  links []
  nodes []
 }
 R_BL_Center {
  path=ItemFullBodyOverride.AnimGraphNode_StateMachine_0.Dodge SM.AnimStateNode_5.In Air.AnimGraphNode_StateMachine_1.Dodge Directional.AnimStateNode_1.Right.AnimGraphNode_BlendSpaceGraph_0.Dodge Right.Right_BackLeft_InAir.AnimGraphNode_RotationOffsetBlendSpaceGraph_0.RightSide_BackLeft_AimOffset.R_BL_Center
  nodes [Output Animation Pose(Result=Additive Identity Pose())]
 }
 R_BL_Down {
  path=ItemFullBodyOverride.AnimGraphNode_StateMachine_0.Dodge SM.AnimStateNode_5.In Air.AnimGraphNode_StateMachine_1.Dodge Directional.AnimStateNode_1.Right.AnimGraphNode_BlendSpaceGraph_0.Dodge Right.Right_BackLeft_InAir.AnimGraphNode_RotationOffsetBlendSpaceGraph_0.RightSide_BackLeft_AimOffset.R_BL_Down
  nodes [
   n1 = "Sequence Player (additive) 'CherrySmoke_DualRevolvers_Dodge_RightSide_BL_AO_D50'"() {
    bLoopAnimation=false
    sequence=/CherrySmokeDualRevolversGameplay/Assets/Anim/Poses/Dodge_AO/CherrySmoke_DualRevolvers_Dodge_RightSide_BL_AO_D50
   }
   Output Animation Pose(Result=@n1)
  ]
 }
 R_BL_Up {
  path=ItemFullBodyOverride.AnimGraphNode_StateMachine_0.Dodge SM.AnimStateNode_5.In Air.AnimGraphNode_StateMachine_1.Dodge Directional.AnimStateNode_1.Right.AnimGraphNode_BlendSpaceGraph_0.Dodge Right.Right_BackLeft_InAir.AnimGraphNode_RotationOffsetBlendSpaceGraph_0.RightSide_BackLeft_AimOffset.R_BL_Up
  nodes [
   n1 = "Sequence Player (additive) 'CherrySmoke_DualRevolvers_Dodge_RightSide_BL_AO_U50'"() {
    bLoopAnimation=false
    sequence=/CherrySmokeDualRevolversGameplay/Assets/Anim/Poses/Dodge_AO/CherrySmoke_DualRevolvers_Dodge_RightSide_BL_AO_U50
   }
   Output Animation Pose(Result=@n1)
  ]
 }
 Left Firing {
  entry=Empty
  path=ItemFullBodyOverride.AnimGraphNode_StateMachine_0.Dodge SM.AnimStateNode_5.In Air.AnimGraphNode_StateMachine_1.Dodge Directional.AnimStateNode_1.Right.AnimGraphNode_BlendSpaceGraph_0.Dodge Right.Right_BackLeft_InAir.AnimGraphNode_StateMachine_0.Left Firing
  ids=id
  nodes [
   AnimStateConduitNode_0 = Is Firing()
   AnimStateNode_0 = Left Fire() {bAlwaysResetOnEntry=true}
   AnimStateNode_1 = Empty()
   AnimStateNode_2 = Right Fire() {bAlwaysResetOnEntry=true}
   AnimStateTransitionNode_0 = AnimStateTransitionNode()
   AnimStateTransitionNode_1 = AnimStateTransitionNode() {crossfadeDuration=0.05}
   AnimStateTransitionNode_2 = AnimStateTransitionNode() {crossfadeDuration=0.05}
   AnimStateTransitionNode_30 = AnimStateTransitionNode() {bAutomaticRuleBasedOnSequencePlayerInState=true}
   AnimStateTransitionNode_31 = AnimStateTransitionNode() {bAutomaticRuleBasedOnSequencePlayerInState=true}
  ]
  links [
   {
    dst=Is Firing
    rule=ItemFullBodyOverride.AnimGraphNode_StateMachine_0.Dodge SM.AnimStateNode_5.In Air.AnimGraphNode_StateMachine_1.Dodge Directional.AnimStateNode_1.Right.AnimGraphNode_BlendSpaceGraph_0.Dodge Right.Right_BackLeft_InAir.AnimGraphNode_StateMachine_0.Left Firing.AnimStateTransitionNode_0.Transition
    src=Empty
    via=AnimStateTransitionNode_0
   }
   {
    dst=Left Fire
    rule=ItemFullBodyOverride.AnimGraphNode_StateMachine_0.Dodge SM.AnimStateNode_5.In Air.AnimGraphNode_StateMachine_1.Dodge Directional.AnimStateNode_1.Right.AnimGraphNode_BlendSpaceGraph_0.Dodge Right.Right_BackLeft_InAir.AnimGraphNode_StateMachine_0.Left Firing.AnimStateTransitionNode_1.Transition
    src=Is Firing
    via=AnimStateTransitionNode_1
   }
   {
    dst=Right Fire
    rule=ItemFullBodyOverride.AnimGraphNode_StateMachine_0.Dodge SM.AnimStateNode_5.In Air.AnimGraphNode_StateMachine_1.Dodge Directional.AnimStateNode_1.Right.AnimGraphNode_BlendSpaceGraph_0.Dodge Right.Right_BackLeft_InAir.AnimGraphNode_StateMachine_0.Left Firing.AnimStateTransitionNode_2.Transition
    src=Is Firing
    via=AnimStateTransitionNode_2
   }
   {
    dst=Is Firing
    rule=ItemFullBodyOverride.AnimGraphNode_StateMachine_0.Dodge SM.AnimStateNode_5.In Air.AnimGraphNode_StateMachine_1.Dodge Directional.AnimStateNode_1.Right.AnimGraphNode_BlendSpaceGraph_0.Dodge Right.Right_BackLeft_InAir.AnimGraphNode_StateMachine_0.Left Firing.AnimStateTransitionNode_30.Transition
    src=Left Fire
    via=AnimStateTransitionNode_30
   }
   {
    dst=Is Firing
    rule=ItemFullBodyOverride.AnimGraphNode_StateMachine_0.Dodge SM.AnimStateNode_5.In Air.AnimGraphNode_StateMachine_1.Dodge Directional.AnimStateNode_1.Right.AnimGraphNode_BlendSpaceGraph_0.Dodge Right.Right_BackLeft_InAir.AnimGraphNode_StateMachine_0.Left Firing.AnimStateTransitionNode_31.Transition
    src=Right Fire
    via=AnimStateTransitionNode_31
   }
  ]
 }
 Is Firing {
  path=ItemFullBodyOverride.AnimGraphNode_StateMachine_0.Dodge SM.AnimStateNode_5.In Air.AnimGraphNode_StateMachine_1.Dodge Directional.AnimStateNode_1.Right.AnimGraphNode_BlendSpaceGraph_0.Dodge Right.Right_BackLeft_InAir.AnimGraphNode_StateMachine_0.Left Firing.AnimStateConduitNode_0.Is Firing
  nodes [Result(bCanEnterTransition=Property Access() -> (Value=~))]
 }
 Left Fire {
  path=ItemFullBodyOverride.AnimGraphNode_StateMachine_0.Dodge SM.AnimStateNode_5.In Air.AnimGraphNode_StateMachine_1.Dodge Directional.AnimStateNode_1.Right.AnimGraphNode_BlendSpaceGraph_0.Dodge Right.Right_BackLeft_InAir.AnimGraphNode_StateMachine_0.Left Firing.AnimStateNode_0.Left Fire
  nodes [
   n1 = "Sequence Player (additive) 'CherrySmoke_DualRevolvers_left_dodge_fire_L'"(PlayRate=Get AdditiveLoopPlayRate()) {bLoopAnimation=false, sequence=&CherrySmoke_DualRevolvers_left_dodge_fire_L}
   Output Animation Pose(Result=@n1) {is=title AnimGraphNode_StateResult}
  ]
 }
 Empty {
  path=ItemFullBodyOverride.AnimGraphNode_StateMachine_0.Dodge SM.AnimStateNode_5.In Air.AnimGraphNode_StateMachine_1.Dodge Directional.AnimStateNode_1.Right.AnimGraphNode_BlendSpaceGraph_0.Dodge Right.Right_BackLeft_InAir.AnimGraphNode_StateMachine_0.Left Firing.AnimStateNode_1.Empty
  nodes [Output Animation Pose(Result=Additive Identity Pose())]
 }
 Right Fire {
  path=ItemFullBodyOverride.AnimGraphNode_StateMachine_0.Dodge SM.AnimStateNode_5.In Air.AnimGraphNode_StateMachine_1.Dodge Directional.AnimStateNode_1.Right.AnimGraphNode_BlendSpaceGraph_0.Dodge Right.Right_BackLeft_InAir.AnimGraphNode_StateMachine_0.Left Firing.AnimStateNode_2.Right Fire
  nodes [
   n1 = "Sequence Player (additive) 'CherrySmoke_DualRevolvers_left_dodge_fire_R'"(PlayRate=Get AdditiveLoopPlayRate()) {bLoopAnimation=false, sequence=&CherrySmoke_DualRevolvers_left_dodge_fire_R}
   Output Animation Pose(Result=@n1)
  ]
 }
 Transition {
  path=ItemFullBodyOverride.AnimGraphNode_StateMachine_0.Dodge SM.AnimStateNode_5.In Air.AnimGraphNode_StateMachine_1.Dodge Directional.AnimStateNode_1.Right.AnimGraphNode_BlendSpaceGraph_0.Dodge Right.Right_BackLeft_InAir.AnimGraphNode_StateMachine_0.Left Firing.AnimStateTransitionNode_0.Transition
  nodes [Result(bCanEnterTransition=true)]
 }
 Transition {
  path=ItemFullBodyOverride.AnimGraphNode_StateMachine_0.Dodge SM.AnimStateNode_5.In Air.AnimGraphNode_StateMachine_1.Dodge Directional.AnimStateNode_1.Right.AnimGraphNode_BlendSpaceGraph_0.Dodge Right.Right_BackLeft_InAir.AnimGraphNode_StateMachine_0.Left Firing.AnimStateTransitionNode_1.Transition
  nodes [Result(bCanEnterTransition=NOT Boolean(A=Get bIsLastHandFireLeft()))]
 }
 Transition {
  path=ItemFullBodyOverride.AnimGraphNode_StateMachine_0.Dodge SM.AnimStateNode_5.In Air.AnimGraphNode_StateMachine_1.Dodge Directional.AnimStateNode_1.Right.AnimGraphNode_BlendSpaceGraph_0.Dodge Right.Right_BackLeft_InAir.AnimGraphNode_StateMachine_0.Left Firing.AnimStateTransitionNode_2.Transition
  nodes [Result(bCanEnterTransition=Get bIsLastHandFireLeft())]
 }
 Transition {
  path=ItemFullBodyOverride.AnimGraphNode_StateMachine_0.Dodge SM.AnimStateNode_5.In Air.AnimGraphNode_StateMachine_1.Dodge Directional.AnimStateNode_1.Right.AnimGraphNode_BlendSpaceGraph_0.Dodge Right.Right_BackLeft_InAir.AnimGraphNode_StateMachine_0.Left Firing.AnimStateTransitionNode_30.Transition
  nodes [Result()]
 }
 Transition {
  path=ItemFullBodyOverride.AnimGraphNode_StateMachine_0.Dodge SM.AnimStateNode_5.In Air.AnimGraphNode_StateMachine_1.Dodge Directional.AnimStateNode_1.Right.AnimGraphNode_BlendSpaceGraph_0.Dodge Right.Right_BackLeft_InAir.AnimGraphNode_StateMachine_0.Left Firing.AnimStateTransitionNode_31.Transition
  nodes [Result()]
 }
 Right_BackRight_InAir {
  path=ItemFullBodyOverride.AnimGraphNode_StateMachine_0.Dodge SM.AnimStateNode_5.In Air.AnimGraphNode_StateMachine_1.Dodge Directional.AnimStateNode_1.Right.AnimGraphNode_BlendSpaceGraph_0.Dodge Right.Right_BackRight_InAir
  nodes [
   n1 = Sequence Player 'CherrySmoke_DualRevolver_Dodge_RightSide_BR_Diving'() {
    bLoopAnimation=false
    sequence=/CherrySmokeDualRevolversGameplay/Assets/Anim/CharacterAnim/CherrySmoke_DualRevolver_Dodge_RightSide_BR_Diving
   }
   AnimGraphNode_RotationOffsetBlendSpaceGraph_0 = AimOffset 'RightSide_BackRight_AimOffset'!(X:float=Get AimPitch(), BasePose:Pose Link Structure=@n1, Alpha:float=1) -> (Pose:Pose Link Structure) {is=title AnimGraphNode_RotationOffsetBlendSpaceGraph}
   AnimGraphNode_StateMachine_0 = Right Firing()
   Output Animation Pose(Result=Apply Additive(Base=@AnimGraphNode_RotationOffsetBlendSpaceGraph_0, Additive=@AnimGraphNode_StateMachine_0) {alphaBoolBlend {blendOption=Linear}}) {is=title AnimGraphNode_BlendSpaceSampleResult}
  ]
 }
 RightSide_BackRight_AimOffset {
  path=ItemFullBodyOverride.AnimGraphNode_StateMachine_0.Dodge SM.AnimStateNode_5.In Air.AnimGraphNode_StateMachine_1.Dodge Directional.AnimStateNode_1.Right.AnimGraphNode_BlendSpaceGraph_0.Dodge Right.Right_BackRight_InAir.AnimGraphNode_RotationOffsetBlendSpaceGraph_0.RightSide_BackRight_AimOffset
  links []
  nodes []
 }
 R_BR_Center {
  path=ItemFullBodyOverride.AnimGraphNode_StateMachine_0.Dodge SM.AnimStateNode_5.In Air.AnimGraphNode_StateMachine_1.Dodge Directional.AnimStateNode_1.Right.AnimGraphNode_BlendSpaceGraph_0.Dodge Right.Right_BackRight_InAir.AnimGraphNode_RotationOffsetBlendSpaceGraph_0.RightSide_BackRight_AimOffset.R_BR_Center
  nodes [Output Animation Pose(Result=Additive Identity Pose())]
 }
 R_BR_Down {
  path=ItemFullBodyOverride.AnimGraphNode_StateMachine_0.Dodge SM.AnimStateNode_5.In Air.AnimGraphNode_StateMachine_1.Dodge Directional.AnimStateNode_1.Right.AnimGraphNode_BlendSpaceGraph_0.Dodge Right.Right_BackRight_InAir.AnimGraphNode_RotationOffsetBlendSpaceGraph_0.RightSide_BackRight_AimOffset.R_BR_Down
  nodes [
   n1 = "Sequence Player (additive) 'CherrySmoke_DualRevolvers_Dodge_RightSide_BR_AO_D50'"() {
    bLoopAnimation=false
    sequence=/CherrySmokeDualRevolversGameplay/Assets/Anim/Poses/Dodge_AO/CherrySmoke_DualRevolvers_Dodge_RightSide_BR_AO_D50
   }
   Output Animation Pose(Result=@n1)
  ]
 }
 R_BR_Up {
  path=ItemFullBodyOverride.AnimGraphNode_StateMachine_0.Dodge SM.AnimStateNode_5.In Air.AnimGraphNode_StateMachine_1.Dodge Directional.AnimStateNode_1.Right.AnimGraphNode_BlendSpaceGraph_0.Dodge Right.Right_BackRight_InAir.AnimGraphNode_RotationOffsetBlendSpaceGraph_0.RightSide_BackRight_AimOffset.R_BR_Up
  nodes [
   n1 = "Sequence Player (additive) 'CherrySmoke_DualRevolvers_Dodge_RightSide_BR_AO_U50'"() {
    bLoopAnimation=false
    sequence=/CherrySmokeDualRevolversGameplay/Assets/Anim/Poses/Dodge_AO/CherrySmoke_DualRevolvers_Dodge_RightSide_BR_AO_U50
   }
   Output Animation Pose(Result=@n1)
  ]
 }
 Right Firing {
  entry=Empty
  path=ItemFullBodyOverride.AnimGraphNode_StateMachine_0.Dodge SM.AnimStateNode_5.In Air.AnimGraphNode_StateMachine_1.Dodge Directional.AnimStateNode_1.Right.AnimGraphNode_BlendSpaceGraph_0.Dodge Right.Right_BackRight_InAir.AnimGraphNode_StateMachine_0.Right Firing
  ids=id
  nodes [
   AnimStateConduitNode_0 = Is Firing()
   AnimStateNode_0 = Left Fire() {bAlwaysResetOnEntry=true}
   AnimStateNode_1 = Empty()
   AnimStateNode_2 = Right Fire() {bAlwaysResetOnEntry=true}
   AnimStateTransitionNode_20 = AnimStateTransitionNode() {bAutomaticRuleBasedOnSequencePlayerInState=true}
   AnimStateTransitionNode_21 = AnimStateTransitionNode() {bAutomaticRuleBasedOnSequencePlayerInState=true}
   AnimStateTransitionNode_3 = AnimStateTransitionNode()
   AnimStateTransitionNode_4 = AnimStateTransitionNode() {crossfadeDuration=0.05}
   AnimStateTransitionNode_5 = AnimStateTransitionNode() {crossfadeDuration=0.05}
  ]
  links [
   {
    dst=Is Firing
    rule=ItemFullBodyOverride.AnimGraphNode_StateMachine_0.Dodge SM.AnimStateNode_5.In Air.AnimGraphNode_StateMachine_1.Dodge Directional.AnimStateNode_1.Right.AnimGraphNode_BlendSpaceGraph_0.Dodge Right.Right_BackRight_InAir.AnimGraphNode_StateMachine_0.Right Firing.AnimStateTransitionNode_3.Transition
    src=Empty
    via=AnimStateTransitionNode_3
   }
   {
    dst=Left Fire
    rule=ItemFullBodyOverride.AnimGraphNode_StateMachine_0.Dodge SM.AnimStateNode_5.In Air.AnimGraphNode_StateMachine_1.Dodge Directional.AnimStateNode_1.Right.AnimGraphNode_BlendSpaceGraph_0.Dodge Right.Right_BackRight_InAir.AnimGraphNode_StateMachine_0.Right Firing.AnimStateTransitionNode_4.Transition
    src=Is Firing
    via=AnimStateTransitionNode_4
   }
   {
    dst=Right Fire
    rule=ItemFullBodyOverride.AnimGraphNode_StateMachine_0.Dodge SM.AnimStateNode_5.In Air.AnimGraphNode_StateMachine_1.Dodge Directional.AnimStateNode_1.Right.AnimGraphNode_BlendSpaceGraph_0.Dodge Right.Right_BackRight_InAir.AnimGraphNode_StateMachine_0.Right Firing.AnimStateTransitionNode_5.Transition
    src=Is Firing
    via=AnimStateTransitionNode_5
   }
   {
    dst=Is Firing
    rule=ItemFullBodyOverride.AnimGraphNode_StateMachine_0.Dodge SM.AnimStateNode_5.In Air.AnimGraphNode_StateMachine_1.Dodge Directional.AnimStateNode_1.Right.AnimGraphNode_BlendSpaceGraph_0.Dodge Right.Right_BackRight_InAir.AnimGraphNode_StateMachine_0.Right Firing.AnimStateTransitionNode_20.Transition
    src=Left Fire
    via=AnimStateTransitionNode_20
   }
   {
    dst=Is Firing
    rule=ItemFullBodyOverride.AnimGraphNode_StateMachine_0.Dodge SM.AnimStateNode_5.In Air.AnimGraphNode_StateMachine_1.Dodge Directional.AnimStateNode_1.Right.AnimGraphNode_BlendSpaceGraph_0.Dodge Right.Right_BackRight_InAir.AnimGraphNode_StateMachine_0.Right Firing.AnimStateTransitionNode_21.Transition
    src=Right Fire
    via=AnimStateTransitionNode_21
   }
  ]
 }
 Is Firing {
  path=ItemFullBodyOverride.AnimGraphNode_StateMachine_0.Dodge SM.AnimStateNode_5.In Air.AnimGraphNode_StateMachine_1.Dodge Directional.AnimStateNode_1.Right.AnimGraphNode_BlendSpaceGraph_0.Dodge Right.Right_BackRight_InAir.AnimGraphNode_StateMachine_0.Right Firing.AnimStateConduitNode_0.Is Firing
  nodes [Result(bCanEnterTransition=Property Access() -> (Value=~))]
 }
 Left Fire {
  path=ItemFullBodyOverride.AnimGraphNode_StateMachine_0.Dodge SM.AnimStateNode_5.In Air.AnimGraphNode_StateMachine_1.Dodge Directional.AnimStateNode_1.Right.AnimGraphNode_BlendSpaceGraph_0.Dodge Right.Right_BackRight_InAir.AnimGraphNode_StateMachine_0.Right Firing.AnimStateNode_0.Left Fire
  nodes [
   n1 = "Sequence Player (additive) 'CherrySmoke_DualRevolvers_right_dodge_fire_L'"(PlayRate=Get AdditiveLoopPlayRate()) {bLoopAnimation=false, sequence=&CherrySmoke_DualRevolvers_right_dodge_fire_L}
   Output Animation Pose(Result=@n1) {is=title AnimGraphNode_StateResult}
  ]
 }
 Empty {
  path=ItemFullBodyOverride.AnimGraphNode_StateMachine_0.Dodge SM.AnimStateNode_5.In Air.AnimGraphNode_StateMachine_1.Dodge Directional.AnimStateNode_1.Right.AnimGraphNode_BlendSpaceGraph_0.Dodge Right.Right_BackRight_InAir.AnimGraphNode_StateMachine_0.Right Firing.AnimStateNode_1.Empty
  nodes [Output Animation Pose(Result=Additive Identity Pose())]
 }
 Right Fire {
  path=ItemFullBodyOverride.AnimGraphNode_StateMachine_0.Dodge SM.AnimStateNode_5.In Air.AnimGraphNode_StateMachine_1.Dodge Directional.AnimStateNode_1.Right.AnimGraphNode_BlendSpaceGraph_0.Dodge Right.Right_BackRight_InAir.AnimGraphNode_StateMachine_0.Right Firing.AnimStateNode_2.Right Fire
  nodes [
   n1 = "Sequence Player (additive) 'CherrySmoke_DualRevolvers_right_dodge_fire_R'"(PlayRate=Get AdditiveLoopPlayRate()) {bLoopAnimation=false, sequence=&CherrySmoke_DualRevolvers_right_dodge_fire_R}
   Output Animation Pose(Result=@n1)
  ]
 }
 Transition {
  path=ItemFullBodyOverride.AnimGraphNode_StateMachine_0.Dodge SM.AnimStateNode_5.In Air.AnimGraphNode_StateMachine_1.Dodge Directional.AnimStateNode_1.Right.AnimGraphNode_BlendSpaceGraph_0.Dodge Right.Right_BackRight_InAir.AnimGraphNode_StateMachine_0.Right Firing.AnimStateTransitionNode_20.Transition
  nodes [Result()]
 }
 Transition {
  path=ItemFullBodyOverride.AnimGraphNode_StateMachine_0.Dodge SM.AnimStateNode_5.In Air.AnimGraphNode_StateMachine_1.Dodge Directional.AnimStateNode_1.Right.AnimGraphNode_BlendSpaceGraph_0.Dodge Right.Right_BackRight_InAir.AnimGraphNode_StateMachine_0.Right Firing.AnimStateTransitionNode_21.Transition
  nodes [Result()]
 }
 Transition {
  path=ItemFullBodyOverride.AnimGraphNode_StateMachine_0.Dodge SM.AnimStateNode_5.In Air.AnimGraphNode_StateMachine_1.Dodge Directional.AnimStateNode_1.Right.AnimGraphNode_BlendSpaceGraph_0.Dodge Right.Right_BackRight_InAir.AnimGraphNode_StateMachine_0.Right Firing.AnimStateTransitionNode_3.Transition
  nodes [Result(bCanEnterTransition=true)]
 }
 Transition {
  path=ItemFullBodyOverride.AnimGraphNode_StateMachine_0.Dodge SM.AnimStateNode_5.In Air.AnimGraphNode_StateMachine_1.Dodge Directional.AnimStateNode_1.Right.AnimGraphNode_BlendSpaceGraph_0.Dodge Right.Right_BackRight_InAir.AnimGraphNode_StateMachine_0.Right Firing.AnimStateTransitionNode_4.Transition
  nodes [Result(bCanEnterTransition=NOT Boolean(A=Get bIsLastHandFireLeft()))]
 }
 Transition {
  path=ItemFullBodyOverride.AnimGraphNode_StateMachine_0.Dodge SM.AnimStateNode_5.In Air.AnimGraphNode_StateMachine_1.Dodge Directional.AnimStateNode_1.Right.AnimGraphNode_BlendSpaceGraph_0.Dodge Right.Right_BackRight_InAir.AnimGraphNode_StateMachine_0.Right Firing.AnimStateTransitionNode_5.Transition
  nodes [Result(bCanEnterTransition=Get bIsLastHandFireLeft())]
 }
 Right_Back_LeftSide_InAir {
  path=ItemFullBodyOverride.AnimGraphNode_StateMachine_0.Dodge SM.AnimStateNode_5.In Air.AnimGraphNode_StateMachine_1.Dodge Directional.AnimStateNode_1.Right.AnimGraphNode_BlendSpaceGraph_0.Dodge Right.Right_Back_LeftSide_InAir
  nodes [
   AnimGraphNode_RotationOffsetBlendSpaceGraph_0 = AimOffset 'RightSide_Back_Left_AimOffset'!(
    X:float=Get AimPitch()
    BasePose:Pose Link Structure="Sequence Evaluator (by time) 'CherrySmoke_DualRevolvers_Dodge_RightSide_B_Pose'"!(ExplicitTime:float=Get InAirAccumulatedTime()) -> (Pose:Pose Link Structure) {
     is=title AnimGraphNode_SequenceEvaluator
     sequence=/CherrySmokeDualRevolversGameplay/Assets/Anim/Poses/CherrySmoke_DualRevolvers_Dodge_RightSide_B_Pose
    }
    Alpha:float=1
   ) -> (Pose:Pose Link Structure) {is=title AnimGraphNode_RotationOffsetBlendSpaceGraph}
   AnimGraphNode_StateMachine_0 = Backwards Firing()
   Output Animation Pose(Result=Apply Additive(Base=@AnimGraphNode_RotationOffsetBlendSpaceGraph_0, Additive=@AnimGraphNode_StateMachine_0) {alphaBoolBlend {blendOption=Linear}}) {is=title AnimGraphNode_BlendSpaceSampleResult}
  ]
 }
 RightSide_Back_Left_AimOffset {
  path=ItemFullBodyOverride.AnimGraphNode_StateMachine_0.Dodge SM.AnimStateNode_5.In Air.AnimGraphNode_StateMachine_1.Dodge Directional.AnimStateNode_1.Right.AnimGraphNode_BlendSpaceGraph_0.Dodge Right.Right_Back_LeftSide_InAir.AnimGraphNode_RotationOffsetBlendSpaceGraph_0.RightSide_Back_Left_AimOffset
  links []
  nodes []
 }
 R_B_L_Center {
  path=ItemFullBodyOverride.AnimGraphNode_StateMachine_0.Dodge SM.AnimStateNode_5.In Air.AnimGraphNode_StateMachine_1.Dodge Directional.AnimStateNode_1.Right.AnimGraphNode_BlendSpaceGraph_0.Dodge Right.Right_Back_LeftSide_InAir.AnimGraphNode_RotationOffsetBlendSpaceGraph_0.RightSide_Back_Left_AimOffset.R_B_L_Center
  nodes [Output Animation Pose(Result=Additive Identity Pose())]
 }
 R_B_L_Down {
  path=ItemFullBodyOverride.AnimGraphNode_StateMachine_0.Dodge SM.AnimStateNode_5.In Air.AnimGraphNode_StateMachine_1.Dodge Directional.AnimStateNode_1.Right.AnimGraphNode_BlendSpaceGraph_0.Dodge Right.Right_Back_LeftSide_InAir.AnimGraphNode_RotationOffsetBlendSpaceGraph_0.RightSide_Back_Left_AimOffset.R_B_L_Down
  nodes [
   n1 = "Sequence Player (additive) 'CherrySmoke_DualRevolvers_Dodge_RightSide_B_AO_D20'"() {
    bLoopAnimation=false
    sequence=/CherrySmokeDualRevolversGameplay/Assets/Anim/Poses/Dodge_AO/CherrySmoke_DualRevolvers_Dodge_RightSide_B_AO_D20
   }
   Output Animation Pose(Result=@n1)
  ]
 }
 R_B_L_Up {
  path=ItemFullBodyOverride.AnimGraphNode_StateMachine_0.Dodge SM.AnimStateNode_5.In Air.AnimGraphNode_StateMachine_1.Dodge Directional.AnimStateNode_1.Right.AnimGraphNode_BlendSpaceGraph_0.Dodge Right.Right_Back_LeftSide_InAir.AnimGraphNode_RotationOffsetBlendSpaceGraph_0.RightSide_Back_Left_AimOffset.R_B_L_Up
  nodes [
   n1 = "Sequence Player (additive) 'CherrySmoke_DualRevolvers_Dodge_RightSide_B_AO_U50'"() {
    bLoopAnimation=false
    sequence=/CherrySmokeDualRevolversGameplay/Assets/Anim/Poses/Dodge_AO/CherrySmoke_DualRevolvers_Dodge_RightSide_B_AO_U50
   }
   Output Animation Pose(Result=@n1)
  ]
 }
 Backwards Firing {
  entry=Empty
  path=ItemFullBodyOverride.AnimGraphNode_StateMachine_0.Dodge SM.AnimStateNode_5.In Air.AnimGraphNode_StateMachine_1.Dodge Directional.AnimStateNode_1.Right.AnimGraphNode_BlendSpaceGraph_0.Dodge Right.Right_Back_LeftSide_InAir.AnimGraphNode_StateMachine_0.Backwards Firing
  ids=id
  nodes [
   AnimStateConduitNode_0 = Is Firing()
   AnimStateNode_0 = Left Fire() {bAlwaysResetOnEntry=true}
   AnimStateNode_1 = Empty()
   AnimStateNode_2 = Right Fire() {bAlwaysResetOnEntry=true}
   AnimStateTransitionNode_18 = AnimStateTransitionNode() {bAutomaticRuleBasedOnSequencePlayerInState=true}
   AnimStateTransitionNode_19 = AnimStateTransitionNode() {bAutomaticRuleBasedOnSequencePlayerInState=true}
   AnimStateTransitionNode_3 = AnimStateTransitionNode()
   AnimStateTransitionNode_4 = AnimStateTransitionNode() {crossfadeDuration=0.05}
   AnimStateTransitionNode_5 = AnimStateTransitionNode() {crossfadeDuration=0.05}
  ]
  links [
   {
    dst=Is Firing
    rule=ItemFullBodyOverride.AnimGraphNode_StateMachine_0.Dodge SM.AnimStateNode_5.In Air.AnimGraphNode_StateMachine_1.Dodge Directional.AnimStateNode_1.Right.AnimGraphNode_BlendSpaceGraph_0.Dodge Right.Right_Back_LeftSide_InAir.AnimGraphNode_StateMachine_0.Backwards Firing.AnimStateTransitionNode_3.Transition
    src=Empty
    via=AnimStateTransitionNode_3
   }
   {
    dst=Left Fire
    rule=ItemFullBodyOverride.AnimGraphNode_StateMachine_0.Dodge SM.AnimStateNode_5.In Air.AnimGraphNode_StateMachine_1.Dodge Directional.AnimStateNode_1.Right.AnimGraphNode_BlendSpaceGraph_0.Dodge Right.Right_Back_LeftSide_InAir.AnimGraphNode_StateMachine_0.Backwards Firing.AnimStateTransitionNode_4.Transition
    src=Is Firing
    via=AnimStateTransitionNode_4
   }
   {
    dst=Right Fire
    rule=ItemFullBodyOverride.AnimGraphNode_StateMachine_0.Dodge SM.AnimStateNode_5.In Air.AnimGraphNode_StateMachine_1.Dodge Directional.AnimStateNode_1.Right.AnimGraphNode_BlendSpaceGraph_0.Dodge Right.Right_Back_LeftSide_InAir.AnimGraphNode_StateMachine_0.Backwards Firing.AnimStateTransitionNode_5.Transition
    src=Is Firing
    via=AnimStateTransitionNode_5
   }
   {
    dst=Is Firing
    rule=ItemFullBodyOverride.AnimGraphNode_StateMachine_0.Dodge SM.AnimStateNode_5.In Air.AnimGraphNode_StateMachine_1.Dodge Directional.AnimStateNode_1.Right.AnimGraphNode_BlendSpaceGraph_0.Dodge Right.Right_Back_LeftSide_InAir.AnimGraphNode_StateMachine_0.Backwards Firing.AnimStateTransitionNode_18.Transition
    src=Left Fire
    via=AnimStateTransitionNode_18
   }
   {
    dst=Is Firing
    rule=ItemFullBodyOverride.AnimGraphNode_StateMachine_0.Dodge SM.AnimStateNode_5.In Air.AnimGraphNode_StateMachine_1.Dodge Directional.AnimStateNode_1.Right.AnimGraphNode_BlendSpaceGraph_0.Dodge Right.Right_Back_LeftSide_InAir.AnimGraphNode_StateMachine_0.Backwards Firing.AnimStateTransitionNode_19.Transition
    src=Right Fire
    via=AnimStateTransitionNode_19
   }
  ]
 }
 Is Firing {
  path=ItemFullBodyOverride.AnimGraphNode_StateMachine_0.Dodge SM.AnimStateNode_5.In Air.AnimGraphNode_StateMachine_1.Dodge Directional.AnimStateNode_1.Right.AnimGraphNode_BlendSpaceGraph_0.Dodge Right.Right_Back_LeftSide_InAir.AnimGraphNode_StateMachine_0.Backwards Firing.AnimStateConduitNode_0.Is Firing
  nodes [Result(bCanEnterTransition=Property Access() -> (Value=~))]
 }
 Left Fire {
  path=ItemFullBodyOverride.AnimGraphNode_StateMachine_0.Dodge SM.AnimStateNode_5.In Air.AnimGraphNode_StateMachine_1.Dodge Directional.AnimStateNode_1.Right.AnimGraphNode_BlendSpaceGraph_0.Dodge Right.Right_Back_LeftSide_InAir.AnimGraphNode_StateMachine_0.Backwards Firing.AnimStateNode_0.Left Fire
  nodes [
   n1 = "Sequence Player (additive) 'CherrySmoke_DualRevolvers_back_dodge_fire_L'"(PlayRate=Get AdditiveLoopPlayRate()) {bLoopAnimation=false, sequence=&CherrySmoke_DualRevolvers_back_dodge_fire_L}
   Output Animation Pose(Result=@n1) {is=title AnimGraphNode_StateResult}
  ]
 }
 Empty {
  path=ItemFullBodyOverride.AnimGraphNode_StateMachine_0.Dodge SM.AnimStateNode_5.In Air.AnimGraphNode_StateMachine_1.Dodge Directional.AnimStateNode_1.Right.AnimGraphNode_BlendSpaceGraph_0.Dodge Right.Right_Back_LeftSide_InAir.AnimGraphNode_StateMachine_0.Backwards Firing.AnimStateNode_1.Empty
  nodes [Output Animation Pose(Result=Additive Identity Pose())]
 }
 Right Fire {
  path=ItemFullBodyOverride.AnimGraphNode_StateMachine_0.Dodge SM.AnimStateNode_5.In Air.AnimGraphNode_StateMachine_1.Dodge Directional.AnimStateNode_1.Right.AnimGraphNode_BlendSpaceGraph_0.Dodge Right.Right_Back_LeftSide_InAir.AnimGraphNode_StateMachine_0.Backwards Firing.AnimStateNode_2.Right Fire
  nodes [
   n1 = "Sequence Player (additive) 'CherrySmoke_DualRevolvers_back_dodge_fire_R'"(PlayRate=Get AdditiveLoopPlayRate()) {bLoopAnimation=false, sequence=&CherrySmoke_DualRevolvers_back_dodge_fire_R}
   Output Animation Pose(Result=@n1)
  ]
 }
 Transition {
  path=ItemFullBodyOverride.AnimGraphNode_StateMachine_0.Dodge SM.AnimStateNode_5.In Air.AnimGraphNode_StateMachine_1.Dodge Directional.AnimStateNode_1.Right.AnimGraphNode_BlendSpaceGraph_0.Dodge Right.Right_Back_LeftSide_InAir.AnimGraphNode_StateMachine_0.Backwards Firing.AnimStateTransitionNode_18.Transition
  nodes [Result()]
 }
 Transition {
  path=ItemFullBodyOverride.AnimGraphNode_StateMachine_0.Dodge SM.AnimStateNode_5.In Air.AnimGraphNode_StateMachine_1.Dodge Directional.AnimStateNode_1.Right.AnimGraphNode_BlendSpaceGraph_0.Dodge Right.Right_Back_LeftSide_InAir.AnimGraphNode_StateMachine_0.Backwards Firing.AnimStateTransitionNode_19.Transition
  nodes [Result()]
 }
 Transition {
  path=ItemFullBodyOverride.AnimGraphNode_StateMachine_0.Dodge SM.AnimStateNode_5.In Air.AnimGraphNode_StateMachine_1.Dodge Directional.AnimStateNode_1.Right.AnimGraphNode_BlendSpaceGraph_0.Dodge Right.Right_Back_LeftSide_InAir.AnimGraphNode_StateMachine_0.Backwards Firing.AnimStateTransitionNode_3.Transition
  nodes [Result(bCanEnterTransition=true)]
 }
 Transition {
  path=ItemFullBodyOverride.AnimGraphNode_StateMachine_0.Dodge SM.AnimStateNode_5.In Air.AnimGraphNode_StateMachine_1.Dodge Directional.AnimStateNode_1.Right.AnimGraphNode_BlendSpaceGraph_0.Dodge Right.Right_Back_LeftSide_InAir.AnimGraphNode_StateMachine_0.Backwards Firing.AnimStateTransitionNode_4.Transition
  nodes [Result(bCanEnterTransition=NOT Boolean(A=Get bIsLastHandFireLeft()))]
 }
 Transition {
  path=ItemFullBodyOverride.AnimGraphNode_StateMachine_0.Dodge SM.AnimStateNode_5.In Air.AnimGraphNode_StateMachine_1.Dodge Directional.AnimStateNode_1.Right.AnimGraphNode_BlendSpaceGraph_0.Dodge Right.Right_Back_LeftSide_InAir.AnimGraphNode_StateMachine_0.Backwards Firing.AnimStateTransitionNode_5.Transition
  nodes [Result(bCanEnterTransition=Get bIsLastHandFireLeft())]
 }
 Right_Back_RightSide_InAir {
  path=ItemFullBodyOverride.AnimGraphNode_StateMachine_0.Dodge SM.AnimStateNode_5.In Air.AnimGraphNode_StateMachine_1.Dodge Directional.AnimStateNode_1.Right.AnimGraphNode_BlendSpaceGraph_0.Dodge Right.Right_Back_RightSide_InAir
  nodes [
   AnimGraphNode_RotationOffsetBlendSpaceGraph_0 = AimOffset 'RightSide_Back_Right_AimOffset'!(
    X:float=Get AimPitch()
    BasePose:Pose Link Structure="Sequence Evaluator (by time) 'CherrySmoke_DualRevolvers_Dodge_RightSide_B_Pose'"(ExplicitTime=Get InAirAccumulatedTime()) {sequence=&CherrySmoke_DualRevolvers_Dodge_RightSide_B_Pose}
    Alpha:float=1
   ) -> (Pose:Pose Link Structure) {is=title AnimGraphNode_RotationOffsetBlendSpaceGraph}
   AnimGraphNode_StateMachine_0 = Backwards Firing()
   Output Animation Pose(Result=Apply Additive(Base=@AnimGraphNode_RotationOffsetBlendSpaceGraph_0, Additive=@AnimGraphNode_StateMachine_0) {alphaBoolBlend {blendOption=Linear}}) {is=title AnimGraphNode_BlendSpaceSampleResult}
  ]
 }
 RightSide_Back_Right_AimOffset {
  path=ItemFullBodyOverride.AnimGraphNode_StateMachine_0.Dodge SM.AnimStateNode_5.In Air.AnimGraphNode_StateMachine_1.Dodge Directional.AnimStateNode_1.Right.AnimGraphNode_BlendSpaceGraph_0.Dodge Right.Right_Back_RightSide_InAir.AnimGraphNode_RotationOffsetBlendSpaceGraph_0.RightSide_Back_Right_AimOffset
  links []
  nodes []
 }
 R_B_R_Center {
  path=ItemFullBodyOverride.AnimGraphNode_StateMachine_0.Dodge SM.AnimStateNode_5.In Air.AnimGraphNode_StateMachine_1.Dodge Directional.AnimStateNode_1.Right.AnimGraphNode_BlendSpaceGraph_0.Dodge Right.Right_Back_RightSide_InAir.AnimGraphNode_RotationOffsetBlendSpaceGraph_0.RightSide_Back_Right_AimOffset.R_B_R_Center
  nodes [Output Animation Pose(Result=Additive Identity Pose())]
 }
 R_B_R_Down {
  path=ItemFullBodyOverride.AnimGraphNode_StateMachine_0.Dodge SM.AnimStateNode_5.In Air.AnimGraphNode_StateMachine_1.Dodge Directional.AnimStateNode_1.Right.AnimGraphNode_BlendSpaceGraph_0.Dodge Right.Right_Back_RightSide_InAir.AnimGraphNode_RotationOffsetBlendSpaceGraph_0.RightSide_Back_Right_AimOffset.R_B_R_Down
  nodes [
   n1 = "Sequence Player (additive) 'CherrySmoke_DualRevolvers_Dodge_RightSide_B_AO_D20'"() {
    bLoopAnimation=false
    sequence=&CherrySmoke_DualRevolvers_Dodge_RightSide_B_AO_D20
   }
   Output Animation Pose(Result=@n1)
  ]
 }
 R_B_R_Up {
  path=ItemFullBodyOverride.AnimGraphNode_StateMachine_0.Dodge SM.AnimStateNode_5.In Air.AnimGraphNode_StateMachine_1.Dodge Directional.AnimStateNode_1.Right.AnimGraphNode_BlendSpaceGraph_0.Dodge Right.Right_Back_RightSide_InAir.AnimGraphNode_RotationOffsetBlendSpaceGraph_0.RightSide_Back_Right_AimOffset.R_B_R_Up
  nodes [
   n1 = "Sequence Player (additive) 'CherrySmoke_DualRevolvers_Dodge_RightSide_B_AO_U50'"() {
    bLoopAnimation=false
    sequence=&CherrySmoke_DualRevolvers_Dodge_RightSide_B_AO_U50
   }
   Output Animation Pose(Result=@n1)
  ]
 }
 Backwards Firing {
  entry=Empty
  path=ItemFullBodyOverride.AnimGraphNode_StateMachine_0.Dodge SM.AnimStateNode_5.In Air.AnimGraphNode_StateMachine_1.Dodge Directional.AnimStateNode_1.Right.AnimGraphNode_BlendSpaceGraph_0.Dodge Right.Right_Back_RightSide_InAir.AnimGraphNode_StateMachine_0.Backwards Firing
  ids=id
  nodes [
   AnimStateConduitNode_0 = Is Firing()
   AnimStateNode_0 = Left Fire() {bAlwaysResetOnEntry=true}
   AnimStateNode_1 = Empty()
   AnimStateNode_2 = Right Fire() {bAlwaysResetOnEntry=true}
   AnimStateTransitionNode_3 = AnimStateTransitionNode()
   AnimStateTransitionNode_32 = AnimStateTransitionNode() {bAutomaticRuleBasedOnSequencePlayerInState=true}
   AnimStateTransitionNode_33 = AnimStateTransitionNode() {bAutomaticRuleBasedOnSequencePlayerInState=true}
   AnimStateTransitionNode_4 = AnimStateTransitionNode() {crossfadeDuration=0.05}
   AnimStateTransitionNode_5 = AnimStateTransitionNode() {crossfadeDuration=0.05}
  ]
  links [
   {
    dst=Is Firing
    rule=ItemFullBodyOverride.AnimGraphNode_StateMachine_0.Dodge SM.AnimStateNode_5.In Air.AnimGraphNode_StateMachine_1.Dodge Directional.AnimStateNode_1.Right.AnimGraphNode_BlendSpaceGraph_0.Dodge Right.Right_Back_RightSide_InAir.AnimGraphNode_StateMachine_0.Backwards Firing.AnimStateTransitionNode_3.Transition
    src=Empty
    via=AnimStateTransitionNode_3
   }
   {
    dst=Left Fire
    rule=ItemFullBodyOverride.AnimGraphNode_StateMachine_0.Dodge SM.AnimStateNode_5.In Air.AnimGraphNode_StateMachine_1.Dodge Directional.AnimStateNode_1.Right.AnimGraphNode_BlendSpaceGraph_0.Dodge Right.Right_Back_RightSide_InAir.AnimGraphNode_StateMachine_0.Backwards Firing.AnimStateTransitionNode_4.Transition
    src=Is Firing
    via=AnimStateTransitionNode_4
   }
   {
    dst=Right Fire
    rule=ItemFullBodyOverride.AnimGraphNode_StateMachine_0.Dodge SM.AnimStateNode_5.In Air.AnimGraphNode_StateMachine_1.Dodge Directional.AnimStateNode_1.Right.AnimGraphNode_BlendSpaceGraph_0.Dodge Right.Right_Back_RightSide_InAir.AnimGraphNode_StateMachine_0.Backwards Firing.AnimStateTransitionNode_5.Transition
    src=Is Firing
    via=AnimStateTransitionNode_5
   }
   {
    dst=Is Firing
    rule=ItemFullBodyOverride.AnimGraphNode_StateMachine_0.Dodge SM.AnimStateNode_5.In Air.AnimGraphNode_StateMachine_1.Dodge Directional.AnimStateNode_1.Right.AnimGraphNode_BlendSpaceGraph_0.Dodge Right.Right_Back_RightSide_InAir.AnimGraphNode_StateMachine_0.Backwards Firing.AnimStateTransitionNode_32.Transition
    src=Left Fire
    via=AnimStateTransitionNode_32
   }
   {
    dst=Is Firing
    rule=ItemFullBodyOverride.AnimGraphNode_StateMachine_0.Dodge SM.AnimStateNode_5.In Air.AnimGraphNode_StateMachine_1.Dodge Directional.AnimStateNode_1.Right.AnimGraphNode_BlendSpaceGraph_0.Dodge Right.Right_Back_RightSide_InAir.AnimGraphNode_StateMachine_0.Backwards Firing.AnimStateTransitionNode_33.Transition
    src=Right Fire
    via=AnimStateTransitionNode_33
   }
  ]
 }
 Is Firing {
  path=ItemFullBodyOverride.AnimGraphNode_StateMachine_0.Dodge SM.AnimStateNode_5.In Air.AnimGraphNode_StateMachine_1.Dodge Directional.AnimStateNode_1.Right.AnimGraphNode_BlendSpaceGraph_0.Dodge Right.Right_Back_RightSide_InAir.AnimGraphNode_StateMachine_0.Backwards Firing.AnimStateConduitNode_0.Is Firing
  nodes [Result(bCanEnterTransition=Property Access() -> (Value=~))]
 }
 Left Fire {
  path=ItemFullBodyOverride.AnimGraphNode_StateMachine_0.Dodge SM.AnimStateNode_5.In Air.AnimGraphNode_StateMachine_1.Dodge Directional.AnimStateNode_1.Right.AnimGraphNode_BlendSpaceGraph_0.Dodge Right.Right_Back_RightSide_InAir.AnimGraphNode_StateMachine_0.Backwards Firing.AnimStateNode_0.Left Fire
  nodes [
   n1 = "Sequence Player (additive) 'CherrySmoke_DualRevolvers_back_dodge_fire_L'"(PlayRate=Get AdditiveLoopPlayRate()) {bLoopAnimation=false, sequence=&CherrySmoke_DualRevolvers_back_dodge_fire_L}
   Output Animation Pose(Result=@n1) {is=title AnimGraphNode_StateResult}
  ]
 }
 Empty {
  path=ItemFullBodyOverride.AnimGraphNode_StateMachine_0.Dodge SM.AnimStateNode_5.In Air.AnimGraphNode_StateMachine_1.Dodge Directional.AnimStateNode_1.Right.AnimGraphNode_BlendSpaceGraph_0.Dodge Right.Right_Back_RightSide_InAir.AnimGraphNode_StateMachine_0.Backwards Firing.AnimStateNode_1.Empty
  nodes [Output Animation Pose(Result=Additive Identity Pose())]
 }
 Right Fire {
  path=ItemFullBodyOverride.AnimGraphNode_StateMachine_0.Dodge SM.AnimStateNode_5.In Air.AnimGraphNode_StateMachine_1.Dodge Directional.AnimStateNode_1.Right.AnimGraphNode_BlendSpaceGraph_0.Dodge Right.Right_Back_RightSide_InAir.AnimGraphNode_StateMachine_0.Backwards Firing.AnimStateNode_2.Right Fire
  nodes [
   n1 = "Sequence Player (additive) 'CherrySmoke_DualRevolvers_back_dodge_fire_R'"(PlayRate=Get AdditiveLoopPlayRate()) {bLoopAnimation=false, sequence=&CherrySmoke_DualRevolvers_back_dodge_fire_R}
   Output Animation Pose(Result=@n1)
  ]
 }
 Transition {
  path=ItemFullBodyOverride.AnimGraphNode_StateMachine_0.Dodge SM.AnimStateNode_5.In Air.AnimGraphNode_StateMachine_1.Dodge Directional.AnimStateNode_1.Right.AnimGraphNode_BlendSpaceGraph_0.Dodge Right.Right_Back_RightSide_InAir.AnimGraphNode_StateMachine_0.Backwards Firing.AnimStateTransitionNode_3.Transition
  nodes [Result(bCanEnterTransition=true)]
 }
 Transition {
  path=ItemFullBodyOverride.AnimGraphNode_StateMachine_0.Dodge SM.AnimStateNode_5.In Air.AnimGraphNode_StateMachine_1.Dodge Directional.AnimStateNode_1.Right.AnimGraphNode_BlendSpaceGraph_0.Dodge Right.Right_Back_RightSide_InAir.AnimGraphNode_StateMachine_0.Backwards Firing.AnimStateTransitionNode_32.Transition
  nodes [Result()]
 }
 Transition {
  path=ItemFullBodyOverride.AnimGraphNode_StateMachine_0.Dodge SM.AnimStateNode_5.In Air.AnimGraphNode_StateMachine_1.Dodge Directional.AnimStateNode_1.Right.AnimGraphNode_BlendSpaceGraph_0.Dodge Right.Right_Back_RightSide_InAir.AnimGraphNode_StateMachine_0.Backwards Firing.AnimStateTransitionNode_33.Transition
  nodes [Result()]
 }
 Transition {
  path=ItemFullBodyOverride.AnimGraphNode_StateMachine_0.Dodge SM.AnimStateNode_5.In Air.AnimGraphNode_StateMachine_1.Dodge Directional.AnimStateNode_1.Right.AnimGraphNode_BlendSpaceGraph_0.Dodge Right.Right_Back_RightSide_InAir.AnimGraphNode_StateMachine_0.Backwards Firing.AnimStateTransitionNode_4.Transition
  nodes [Result(bCanEnterTransition=NOT Boolean(A=Get bIsLastHandFireLeft()))]
 }
 Transition {
  path=ItemFullBodyOverride.AnimGraphNode_StateMachine_0.Dodge SM.AnimStateNode_5.In Air.AnimGraphNode_StateMachine_1.Dodge Directional.AnimStateNode_1.Right.AnimGraphNode_BlendSpaceGraph_0.Dodge Right.Right_Back_RightSide_InAir.AnimGraphNode_StateMachine_0.Backwards Firing.AnimStateTransitionNode_5.Transition
  nodes [Result(bCanEnterTransition=Get bIsLastHandFireLeft())]
 }
 Right_ForwardLeft_InAir {
  path=ItemFullBodyOverride.AnimGraphNode_StateMachine_0.Dodge SM.AnimStateNode_5.In Air.AnimGraphNode_StateMachine_1.Dodge Directional.AnimStateNode_1.Right.AnimGraphNode_BlendSpaceGraph_0.Dodge Right.Right_ForwardLeft_InAir
  nodes [
   n1 = Sequence Player 'CherrySmoke_DualRevolvers_Dodge_RightSide_FL_Pose'() {bLoopAnimation=false, sequence=/CherrySmokeDualRevolversGameplay/Assets/Anim/Poses/CherrySmoke_DualRevolvers_Dodge_RightSide_FL_Pose}
   AnimGraphNode_RotationOffsetBlendSpaceGraph_0 = AimOffset 'RightSide_ForwardLeft_AimOffset'!(X:float=Get AimPitch(), BasePose:Pose Link Structure=@n1, Alpha:float=1) -> (Pose:Pose Link Structure) {is=title AnimGraphNode_RotationOffsetBlendSpaceGraph}
   AnimGraphNode_StateMachine_0 = Left Firing()
   Output Animation Pose(Result=Apply Additive(Base=@AnimGraphNode_RotationOffsetBlendSpaceGraph_0, Additive=@AnimGraphNode_StateMachine_0) {alphaBoolBlend {blendOption=Linear}}) {is=title AnimGraphNode_BlendSpaceSampleResult}
  ]
 }
 RightSide_ForwardLeft_AimOffset {
  path=ItemFullBodyOverride.AnimGraphNode_StateMachine_0.Dodge SM.AnimStateNode_5.In Air.AnimGraphNode_StateMachine_1.Dodge Directional.AnimStateNode_1.Right.AnimGraphNode_BlendSpaceGraph_0.Dodge Right.Right_ForwardLeft_InAir.AnimGraphNode_RotationOffsetBlendSpaceGraph_0.RightSide_ForwardLeft_AimOffset
  links []
  nodes []
 }
 R_FL_Center {
  path=ItemFullBodyOverride.AnimGraphNode_StateMachine_0.Dodge SM.AnimStateNode_5.In Air.AnimGraphNode_StateMachine_1.Dodge Directional.AnimStateNode_1.Right.AnimGraphNode_BlendSpaceGraph_0.Dodge Right.Right_ForwardLeft_InAir.AnimGraphNode_RotationOffsetBlendSpaceGraph_0.RightSide_ForwardLeft_AimOffset.R_FL_Center
  nodes [Output Animation Pose(Result=Additive Identity Pose())]
 }
 R_FL_Down {
  path=ItemFullBodyOverride.AnimGraphNode_StateMachine_0.Dodge SM.AnimStateNode_5.In Air.AnimGraphNode_StateMachine_1.Dodge Directional.AnimStateNode_1.Right.AnimGraphNode_BlendSpaceGraph_0.Dodge Right.Right_ForwardLeft_InAir.AnimGraphNode_RotationOffsetBlendSpaceGraph_0.RightSide_ForwardLeft_AimOffset.R_FL_Down
  nodes [
   n1 = "Sequence Player (additive) 'CherrySmoke_DualRevolvers_Dodge_RightSide_FL_AO_D40'"() {
    bLoopAnimation=false
    sequence=/CherrySmokeDualRevolversGameplay/Assets/Anim/Poses/Dodge_AO/CherrySmoke_DualRevolvers_Dodge_RightSide_FL_AO_D40
   }
   Output Animation Pose(Result=@n1)
  ]
 }
 R_FL_Up {
  path=ItemFullBodyOverride.AnimGraphNode_StateMachine_0.Dodge SM.AnimStateNode_5.In Air.AnimGraphNode_StateMachine_1.Dodge Directional.AnimStateNode_1.Right.AnimGraphNode_BlendSpaceGraph_0.Dodge Right.Right_ForwardLeft_InAir.AnimGraphNode_RotationOffsetBlendSpaceGraph_0.RightSide_ForwardLeft_AimOffset.R_FL_Up
  nodes [
   n1 = "Sequence Player (additive) 'CherrySmoke_DualRevolvers_Dodge_RightSide_FL_AO_U40'"() {
    bLoopAnimation=false
    sequence=/CherrySmokeDualRevolversGameplay/Assets/Anim/Poses/Dodge_AO/CherrySmoke_DualRevolvers_Dodge_RightSide_FL_AO_U40
   }
   Output Animation Pose(Result=@n1)
  ]
 }
 Left Firing {
  entry=Empty
  path=ItemFullBodyOverride.AnimGraphNode_StateMachine_0.Dodge SM.AnimStateNode_5.In Air.AnimGraphNode_StateMachine_1.Dodge Directional.AnimStateNode_1.Right.AnimGraphNode_BlendSpaceGraph_0.Dodge Right.Right_ForwardLeft_InAir.AnimGraphNode_StateMachine_0.Left Firing
  ids=id
  nodes [
   AnimStateConduitNode_0 = Is Firing()
   AnimStateNode_0 = Left Fire() {bAlwaysResetOnEntry=true}
   AnimStateNode_1 = Empty()
   AnimStateNode_2 = Right Fire() {bAlwaysResetOnEntry=true}
   AnimStateTransitionNode_0 = AnimStateTransitionNode()
   AnimStateTransitionNode_1 = AnimStateTransitionNode() {crossfadeDuration=0.05}
   AnimStateTransitionNode_2 = AnimStateTransitionNode() {crossfadeDuration=0.05}
   AnimStateTransitionNode_26 = AnimStateTransitionNode() {bAutomaticRuleBasedOnSequencePlayerInState=true}
   AnimStateTransitionNode_27 = AnimStateTransitionNode() {bAutomaticRuleBasedOnSequencePlayerInState=true}
  ]
  links [
   {
    dst=Is Firing
    rule=ItemFullBodyOverride.AnimGraphNode_StateMachine_0.Dodge SM.AnimStateNode_5.In Air.AnimGraphNode_StateMachine_1.Dodge Directional.AnimStateNode_1.Right.AnimGraphNode_BlendSpaceGraph_0.Dodge Right.Right_ForwardLeft_InAir.AnimGraphNode_StateMachine_0.Left Firing.AnimStateTransitionNode_0.Transition
    src=Empty
    via=AnimStateTransitionNode_0
   }
   {
    dst=Left Fire
    rule=ItemFullBodyOverride.AnimGraphNode_StateMachine_0.Dodge SM.AnimStateNode_5.In Air.AnimGraphNode_StateMachine_1.Dodge Directional.AnimStateNode_1.Right.AnimGraphNode_BlendSpaceGraph_0.Dodge Right.Right_ForwardLeft_InAir.AnimGraphNode_StateMachine_0.Left Firing.AnimStateTransitionNode_1.Transition
    src=Is Firing
    via=AnimStateTransitionNode_1
   }
   {
    dst=Right Fire
    rule=ItemFullBodyOverride.AnimGraphNode_StateMachine_0.Dodge SM.AnimStateNode_5.In Air.AnimGraphNode_StateMachine_1.Dodge Directional.AnimStateNode_1.Right.AnimGraphNode_BlendSpaceGraph_0.Dodge Right.Right_ForwardLeft_InAir.AnimGraphNode_StateMachine_0.Left Firing.AnimStateTransitionNode_2.Transition
    src=Is Firing
    via=AnimStateTransitionNode_2
   }
   {
    dst=Is Firing
    rule=ItemFullBodyOverride.AnimGraphNode_StateMachine_0.Dodge SM.AnimStateNode_5.In Air.AnimGraphNode_StateMachine_1.Dodge Directional.AnimStateNode_1.Right.AnimGraphNode_BlendSpaceGraph_0.Dodge Right.Right_ForwardLeft_InAir.AnimGraphNode_StateMachine_0.Left Firing.AnimStateTransitionNode_26.Transition
    src=Left Fire
    via=AnimStateTransitionNode_26
   }
   {
    dst=Is Firing
    rule=ItemFullBodyOverride.AnimGraphNode_StateMachine_0.Dodge SM.AnimStateNode_5.In Air.AnimGraphNode_StateMachine_1.Dodge Directional.AnimStateNode_1.Right.AnimGraphNode_BlendSpaceGraph_0.Dodge Right.Right_ForwardLeft_InAir.AnimGraphNode_StateMachine_0.Left Firing.AnimStateTransitionNode_27.Transition
    src=Right Fire
    via=AnimStateTransitionNode_27
   }
  ]
 }
 Is Firing {
  path=ItemFullBodyOverride.AnimGraphNode_StateMachine_0.Dodge SM.AnimStateNode_5.In Air.AnimGraphNode_StateMachine_1.Dodge Directional.AnimStateNode_1.Right.AnimGraphNode_BlendSpaceGraph_0.Dodge Right.Right_ForwardLeft_InAir.AnimGraphNode_StateMachine_0.Left Firing.AnimStateConduitNode_0.Is Firing
  nodes [Result(bCanEnterTransition=Property Access() -> (Value=~))]
 }
 Left Fire {
  path=ItemFullBodyOverride.AnimGraphNode_StateMachine_0.Dodge SM.AnimStateNode_5.In Air.AnimGraphNode_StateMachine_1.Dodge Directional.AnimStateNode_1.Right.AnimGraphNode_BlendSpaceGraph_0.Dodge Right.Right_ForwardLeft_InAir.AnimGraphNode_StateMachine_0.Left Firing.AnimStateNode_0.Left Fire
  nodes [
   n1 = "Sequence Player (additive) 'CherrySmoke_DualRevolvers_left_dodge_fire_L'"(PlayRate=Get AdditiveLoopPlayRate()) {bLoopAnimation=false, sequence=&CherrySmoke_DualRevolvers_left_dodge_fire_L}
   Output Animation Pose(Result=@n1) {is=title AnimGraphNode_StateResult}
  ]
 }
 Empty {
  path=ItemFullBodyOverride.AnimGraphNode_StateMachine_0.Dodge SM.AnimStateNode_5.In Air.AnimGraphNode_StateMachine_1.Dodge Directional.AnimStateNode_1.Right.AnimGraphNode_BlendSpaceGraph_0.Dodge Right.Right_ForwardLeft_InAir.AnimGraphNode_StateMachine_0.Left Firing.AnimStateNode_1.Empty
  nodes [Output Animation Pose(Result=Additive Identity Pose())]
 }
 Right Fire {
  path=ItemFullBodyOverride.AnimGraphNode_StateMachine_0.Dodge SM.AnimStateNode_5.In Air.AnimGraphNode_StateMachine_1.Dodge Directional.AnimStateNode_1.Right.AnimGraphNode_BlendSpaceGraph_0.Dodge Right.Right_ForwardLeft_InAir.AnimGraphNode_StateMachine_0.Left Firing.AnimStateNode_2.Right Fire
  nodes [
   n1 = "Sequence Player (additive) 'CherrySmoke_DualRevolvers_left_dodge_fire_R'"(PlayRate=Get AdditiveLoopPlayRate()) {bLoopAnimation=false, sequence=&CherrySmoke_DualRevolvers_left_dodge_fire_R}
   Output Animation Pose(Result=@n1)
  ]
 }
 Transition {
  path=ItemFullBodyOverride.AnimGraphNode_StateMachine_0.Dodge SM.AnimStateNode_5.In Air.AnimGraphNode_StateMachine_1.Dodge Directional.AnimStateNode_1.Right.AnimGraphNode_BlendSpaceGraph_0.Dodge Right.Right_ForwardLeft_InAir.AnimGraphNode_StateMachine_0.Left Firing.AnimStateTransitionNode_0.Transition
  nodes [Result(bCanEnterTransition=true)]
 }
 Transition {
  path=ItemFullBodyOverride.AnimGraphNode_StateMachine_0.Dodge SM.AnimStateNode_5.In Air.AnimGraphNode_StateMachine_1.Dodge Directional.AnimStateNode_1.Right.AnimGraphNode_BlendSpaceGraph_0.Dodge Right.Right_ForwardLeft_InAir.AnimGraphNode_StateMachine_0.Left Firing.AnimStateTransitionNode_1.Transition
  nodes [Result(bCanEnterTransition=NOT Boolean(A=Get bIsLastHandFireLeft()))]
 }
 Transition {
  path=ItemFullBodyOverride.AnimGraphNode_StateMachine_0.Dodge SM.AnimStateNode_5.In Air.AnimGraphNode_StateMachine_1.Dodge Directional.AnimStateNode_1.Right.AnimGraphNode_BlendSpaceGraph_0.Dodge Right.Right_ForwardLeft_InAir.AnimGraphNode_StateMachine_0.Left Firing.AnimStateTransitionNode_2.Transition
  nodes [Result(bCanEnterTransition=Get bIsLastHandFireLeft())]
 }
 Transition {
  path=ItemFullBodyOverride.AnimGraphNode_StateMachine_0.Dodge SM.AnimStateNode_5.In Air.AnimGraphNode_StateMachine_1.Dodge Directional.AnimStateNode_1.Right.AnimGraphNode_BlendSpaceGraph_0.Dodge Right.Right_ForwardLeft_InAir.AnimGraphNode_StateMachine_0.Left Firing.AnimStateTransitionNode_26.Transition
  nodes [Result()]
 }
 Transition {
  path=ItemFullBodyOverride.AnimGraphNode_StateMachine_0.Dodge SM.AnimStateNode_5.In Air.AnimGraphNode_StateMachine_1.Dodge Directional.AnimStateNode_1.Right.AnimGraphNode_BlendSpaceGraph_0.Dodge Right.Right_ForwardLeft_InAir.AnimGraphNode_StateMachine_0.Left Firing.AnimStateTransitionNode_27.Transition
  nodes [Result()]
 }
 Right_ForwardRight_InAir {
  path=ItemFullBodyOverride.AnimGraphNode_StateMachine_0.Dodge SM.AnimStateNode_5.In Air.AnimGraphNode_StateMachine_1.Dodge Directional.AnimStateNode_1.Right.AnimGraphNode_BlendSpaceGraph_0.Dodge Right.Right_ForwardRight_InAir
  nodes [
   n1 = Sequence Player 'CherrySmoke_DualRevolver_Dodge_RightSide_FR_Diving'() {
    bLoopAnimation=false
    sequence=/CherrySmokeDualRevolversGameplay/Assets/Anim/CharacterAnim/CherrySmoke_DualRevolver_Dodge_RightSide_FR_Diving
   }
   AnimGraphNode_RotationOffsetBlendSpaceGraph_0 = AimOffset 'RightSide_ForwardRight_AimOffset'!(X:float=Get AimPitch(), BasePose:Pose Link Structure=@n1, Alpha:float=1) -> (Pose:Pose Link Structure) {is=title AnimGraphNode_RotationOffsetBlendSpaceGraph}
   AnimGraphNode_StateMachine_0 = Right Firing()
   Output Animation Pose(Result=Apply Additive(Base=@AnimGraphNode_RotationOffsetBlendSpaceGraph_0, Additive=@AnimGraphNode_StateMachine_0) {alphaBoolBlend {blendOption=Linear}}) {is=title AnimGraphNode_BlendSpaceSampleResult}
  ]
 }
 RightSide_ForwardRight_AimOffset {
  path=ItemFullBodyOverride.AnimGraphNode_StateMachine_0.Dodge SM.AnimStateNode_5.In Air.AnimGraphNode_StateMachine_1.Dodge Directional.AnimStateNode_1.Right.AnimGraphNode_BlendSpaceGraph_0.Dodge Right.Right_ForwardRight_InAir.AnimGraphNode_RotationOffsetBlendSpaceGraph_0.RightSide_ForwardRight_AimOffset
  links []
  nodes []
 }
 R_F_Center {
  path=ItemFullBodyOverride.AnimGraphNode_StateMachine_0.Dodge SM.AnimStateNode_5.In Air.AnimGraphNode_StateMachine_1.Dodge Directional.AnimStateNode_1.Right.AnimGraphNode_BlendSpaceGraph_0.Dodge Right.Right_ForwardRight_InAir.AnimGraphNode_RotationOffsetBlendSpaceGraph_0.RightSide_ForwardRight_AimOffset.R_F_Center
  nodes [Output Animation Pose(Result=Additive Identity Pose())]
 }
 R_F_Down {
  path=ItemFullBodyOverride.AnimGraphNode_StateMachine_0.Dodge SM.AnimStateNode_5.In Air.AnimGraphNode_StateMachine_1.Dodge Directional.AnimStateNode_1.Right.AnimGraphNode_BlendSpaceGraph_0.Dodge Right.Right_ForwardRight_InAir.AnimGraphNode_RotationOffsetBlendSpaceGraph_0.RightSide_ForwardRight_AimOffset.R_F_Down
  nodes [
   n1 = "Sequence Player (additive) 'CherrySmoke_DualRevolvers_Dodge_RightSide_FR_AO_D50'"() {
    bLoopAnimation=false
    sequence=/CherrySmokeDualRevolversGameplay/Assets/Anim/Poses/Dodge_AO/CherrySmoke_DualRevolvers_Dodge_RightSide_FR_AO_D50
   }
   Output Animation Pose(Result=@n1)
   "Sequence Player (additive) 'CherrySmoke_DualRevolvers_Dodge_RightSide_F_AO_D50'"() {
    bLoopAnimation=false
    sequence=/CherrySmokeDualRevolversGameplay/Assets/Anim/Poses/Dodge_AO/CherrySmoke_DualRevolvers_Dodge_RightSide_F_AO_D50
   }
  ]
 }
 R_F_Up {
  path=ItemFullBodyOverride.AnimGraphNode_StateMachine_0.Dodge SM.AnimStateNode_5.In Air.AnimGraphNode_StateMachine_1.Dodge Directional.AnimStateNode_1.Right.AnimGraphNode_BlendSpaceGraph_0.Dodge Right.Right_ForwardRight_InAir.AnimGraphNode_RotationOffsetBlendSpaceGraph_0.RightSide_ForwardRight_AimOffset.R_F_Up
  nodes [
   n1 = "Sequence Player (additive) 'CherrySmoke_DualRevolvers_Dodge_RightSide_FR_AO_U40'"() {
    bLoopAnimation=false
    sequence=/CherrySmokeDualRevolversGameplay/Assets/Anim/Poses/Dodge_AO/CherrySmoke_DualRevolvers_Dodge_RightSide_FR_AO_U40
   }
   Output Animation Pose(Result=@n1)
   "Sequence Player (additive) 'CherrySmoke_DualRevolvers_Dodge_RightSide_F_AO_U20'"() {
    bLoopAnimation=false
    sequence=/CherrySmokeDualRevolversGameplay/Assets/Anim/Poses/Dodge_AO/CherrySmoke_DualRevolvers_Dodge_RightSide_F_AO_U20
   }
  ]
 }
 Right Firing {
  entry=Empty
  path=ItemFullBodyOverride.AnimGraphNode_StateMachine_0.Dodge SM.AnimStateNode_5.In Air.AnimGraphNode_StateMachine_1.Dodge Directional.AnimStateNode_1.Right.AnimGraphNode_BlendSpaceGraph_0.Dodge Right.Right_ForwardRight_InAir.AnimGraphNode_StateMachine_0.Right Firing
  ids=id
  nodes [
   AnimStateConduitNode_0 = Is Firing()
   AnimStateNode_0 = Left Fire() {bAlwaysResetOnEntry=true}
   AnimStateNode_1 = Empty()
   AnimStateNode_2 = Right Fire() {bAlwaysResetOnEntry=true}
   AnimStateTransitionNode_24 = AnimStateTransitionNode() {bAutomaticRuleBasedOnSequencePlayerInState=true}
   AnimStateTransitionNode_25 = AnimStateTransitionNode() {bAutomaticRuleBasedOnSequencePlayerInState=true}
   AnimStateTransitionNode_3 = AnimStateTransitionNode()
   AnimStateTransitionNode_4 = AnimStateTransitionNode()
   AnimStateTransitionNode_5 = AnimStateTransitionNode()
  ]
  links [
   {
    dst=Is Firing
    rule=ItemFullBodyOverride.AnimGraphNode_StateMachine_0.Dodge SM.AnimStateNode_5.In Air.AnimGraphNode_StateMachine_1.Dodge Directional.AnimStateNode_1.Right.AnimGraphNode_BlendSpaceGraph_0.Dodge Right.Right_ForwardRight_InAir.AnimGraphNode_StateMachine_0.Right Firing.AnimStateTransitionNode_3.Transition
    src=Empty
    via=AnimStateTransitionNode_3
   }
   {
    dst=Left Fire
    rule=ItemFullBodyOverride.AnimGraphNode_StateMachine_0.Dodge SM.AnimStateNode_5.In Air.AnimGraphNode_StateMachine_1.Dodge Directional.AnimStateNode_1.Right.AnimGraphNode_BlendSpaceGraph_0.Dodge Right.Right_ForwardRight_InAir.AnimGraphNode_StateMachine_0.Right Firing.AnimStateTransitionNode_4.Transition
    src=Is Firing
    via=AnimStateTransitionNode_4
   }
   {
    dst=Right Fire
    rule=ItemFullBodyOverride.AnimGraphNode_StateMachine_0.Dodge SM.AnimStateNode_5.In Air.AnimGraphNode_StateMachine_1.Dodge Directional.AnimStateNode_1.Right.AnimGraphNode_BlendSpaceGraph_0.Dodge Right.Right_ForwardRight_InAir.AnimGraphNode_StateMachine_0.Right Firing.AnimStateTransitionNode_5.Transition
    src=Is Firing
    via=AnimStateTransitionNode_5
   }
   {
    dst=Is Firing
    rule=ItemFullBodyOverride.AnimGraphNode_StateMachine_0.Dodge SM.AnimStateNode_5.In Air.AnimGraphNode_StateMachine_1.Dodge Directional.AnimStateNode_1.Right.AnimGraphNode_BlendSpaceGraph_0.Dodge Right.Right_ForwardRight_InAir.AnimGraphNode_StateMachine_0.Right Firing.AnimStateTransitionNode_24.Transition
    src=Left Fire
    via=AnimStateTransitionNode_24
   }
   {
    dst=Is Firing
    rule=ItemFullBodyOverride.AnimGraphNode_StateMachine_0.Dodge SM.AnimStateNode_5.In Air.AnimGraphNode_StateMachine_1.Dodge Directional.AnimStateNode_1.Right.AnimGraphNode_BlendSpaceGraph_0.Dodge Right.Right_ForwardRight_InAir.AnimGraphNode_StateMachine_0.Right Firing.AnimStateTransitionNode_25.Transition
    src=Right Fire
    via=AnimStateTransitionNode_25
   }
  ]
 }
 Is Firing {
  path=ItemFullBodyOverride.AnimGraphNode_StateMachine_0.Dodge SM.AnimStateNode_5.In Air.AnimGraphNode_StateMachine_1.Dodge Directional.AnimStateNode_1.Right.AnimGraphNode_BlendSpaceGraph_0.Dodge Right.Right_ForwardRight_InAir.AnimGraphNode_StateMachine_0.Right Firing.AnimStateConduitNode_0.Is Firing
  nodes [Result(bCanEnterTransition=Property Access() -> (Value=~))]
 }
 Left Fire {
  path=ItemFullBodyOverride.AnimGraphNode_StateMachine_0.Dodge SM.AnimStateNode_5.In Air.AnimGraphNode_StateMachine_1.Dodge Directional.AnimStateNode_1.Right.AnimGraphNode_BlendSpaceGraph_0.Dodge Right.Right_ForwardRight_InAir.AnimGraphNode_StateMachine_0.Right Firing.AnimStateNode_0.Left Fire
  nodes [
   n1 = "Sequence Player (additive) 'CherrySmoke_DualRevolvers_right_dodge_fire_L'"(PlayRate=Get AdditiveLoopPlayRate()) {bLoopAnimation=false, sequence=&CherrySmoke_DualRevolvers_right_dodge_fire_L}
   Output Animation Pose(Result=@n1) {is=title AnimGraphNode_StateResult}
  ]
 }
 Empty {
  path=ItemFullBodyOverride.AnimGraphNode_StateMachine_0.Dodge SM.AnimStateNode_5.In Air.AnimGraphNode_StateMachine_1.Dodge Directional.AnimStateNode_1.Right.AnimGraphNode_BlendSpaceGraph_0.Dodge Right.Right_ForwardRight_InAir.AnimGraphNode_StateMachine_0.Right Firing.AnimStateNode_1.Empty
  nodes [Output Animation Pose(Result=Additive Identity Pose())]
 }
 Right Fire {
  path=ItemFullBodyOverride.AnimGraphNode_StateMachine_0.Dodge SM.AnimStateNode_5.In Air.AnimGraphNode_StateMachine_1.Dodge Directional.AnimStateNode_1.Right.AnimGraphNode_BlendSpaceGraph_0.Dodge Right.Right_ForwardRight_InAir.AnimGraphNode_StateMachine_0.Right Firing.AnimStateNode_2.Right Fire
  nodes [
   n1 = "Sequence Player (additive) 'CherrySmoke_DualRevolvers_right_dodge_fire_R'"(PlayRate=Get AdditiveLoopPlayRate()) {bLoopAnimation=false, sequence=&CherrySmoke_DualRevolvers_right_dodge_fire_R}
   Output Animation Pose(Result=@n1)
  ]
 }
 Transition {
  path=ItemFullBodyOverride.AnimGraphNode_StateMachine_0.Dodge SM.AnimStateNode_5.In Air.AnimGraphNode_StateMachine_1.Dodge Directional.AnimStateNode_1.Right.AnimGraphNode_BlendSpaceGraph_0.Dodge Right.Right_ForwardRight_InAir.AnimGraphNode_StateMachine_0.Right Firing.AnimStateTransitionNode_24.Transition
  nodes [Result()]
 }
 Transition {
  path=ItemFullBodyOverride.AnimGraphNode_StateMachine_0.Dodge SM.AnimStateNode_5.In Air.AnimGraphNode_StateMachine_1.Dodge Directional.AnimStateNode_1.Right.AnimGraphNode_BlendSpaceGraph_0.Dodge Right.Right_ForwardRight_InAir.AnimGraphNode_StateMachine_0.Right Firing.AnimStateTransitionNode_25.Transition
  nodes [Result()]
 }
 Transition {
  path=ItemFullBodyOverride.AnimGraphNode_StateMachine_0.Dodge SM.AnimStateNode_5.In Air.AnimGraphNode_StateMachine_1.Dodge Directional.AnimStateNode_1.Right.AnimGraphNode_BlendSpaceGraph_0.Dodge Right.Right_ForwardRight_InAir.AnimGraphNode_StateMachine_0.Right Firing.AnimStateTransitionNode_3.Transition
  nodes [Result(bCanEnterTransition=true)]
 }
 Transition {
  path=ItemFullBodyOverride.AnimGraphNode_StateMachine_0.Dodge SM.AnimStateNode_5.In Air.AnimGraphNode_StateMachine_1.Dodge Directional.AnimStateNode_1.Right.AnimGraphNode_BlendSpaceGraph_0.Dodge Right.Right_ForwardRight_InAir.AnimGraphNode_StateMachine_0.Right Firing.AnimStateTransitionNode_4.Transition
  nodes [Result(bCanEnterTransition=NOT Boolean(A=Get bIsLastHandFireLeft()))]
 }
 Transition {
  path=ItemFullBodyOverride.AnimGraphNode_StateMachine_0.Dodge SM.AnimStateNode_5.In Air.AnimGraphNode_StateMachine_1.Dodge Directional.AnimStateNode_1.Right.AnimGraphNode_BlendSpaceGraph_0.Dodge Right.Right_ForwardRight_InAir.AnimGraphNode_StateMachine_0.Right Firing.AnimStateTransitionNode_5.Transition
  nodes [Result(bCanEnterTransition=Get bIsLastHandFireLeft())]
 }
 Right_Forward_InAir {
  path=ItemFullBodyOverride.AnimGraphNode_StateMachine_0.Dodge SM.AnimStateNode_5.In Air.AnimGraphNode_StateMachine_1.Dodge Directional.AnimStateNode_1.Right.AnimGraphNode_BlendSpaceGraph_0.Dodge Right.Right_Forward_InAir
  nodes [
   AnimGraphNode_RotationOffsetBlendSpaceGraph_0 = AimOffset 'RightSide_Forward_AimOffset'!(
    X:float=Get AimPitch()
    BasePose:Pose Link Structure=Sequence Player 'CherrySmoke_DualRevolvers_Dodge_RightSide_F_Pose'() {bLoopAnimation=false, sequence=/CherrySmokeDualRevolversGameplay/Assets/Anim/Poses/CherrySmoke_DualRevolvers_Dodge_RightSide_F_Pose}
    Alpha:float=1
   ) -> (Pose:Pose Link Structure) {is=title AnimGraphNode_RotationOffsetBlendSpaceGraph}
   AnimGraphNode_StateMachine_0 = Forward Firing()
   Output Animation Pose(Result=Apply Additive(Base=@AnimGraphNode_RotationOffsetBlendSpaceGraph_0, Additive=@AnimGraphNode_StateMachine_0) {alphaBoolBlend {blendOption=Linear}}) {is=title AnimGraphNode_BlendSpaceSampleResult}
  ]
 }
 RightSide_Forward_AimOffset {
  path=ItemFullBodyOverride.AnimGraphNode_StateMachine_0.Dodge SM.AnimStateNode_5.In Air.AnimGraphNode_StateMachine_1.Dodge Directional.AnimStateNode_1.Right.AnimGraphNode_BlendSpaceGraph_0.Dodge Right.Right_Forward_InAir.AnimGraphNode_RotationOffsetBlendSpaceGraph_0.RightSide_Forward_AimOffset
  links []
  nodes []
 }
 R_FR_Center {
  path=ItemFullBodyOverride.AnimGraphNode_StateMachine_0.Dodge SM.AnimStateNode_5.In Air.AnimGraphNode_StateMachine_1.Dodge Directional.AnimStateNode_1.Right.AnimGraphNode_BlendSpaceGraph_0.Dodge Right.Right_Forward_InAir.AnimGraphNode_RotationOffsetBlendSpaceGraph_0.RightSide_Forward_AimOffset.R_FR_Center
  nodes [Output Animation Pose(Result=Additive Identity Pose())]
 }
 R_FR_Down {
  path=ItemFullBodyOverride.AnimGraphNode_StateMachine_0.Dodge SM.AnimStateNode_5.In Air.AnimGraphNode_StateMachine_1.Dodge Directional.AnimStateNode_1.Right.AnimGraphNode_BlendSpaceGraph_0.Dodge Right.Right_Forward_InAir.AnimGraphNode_RotationOffsetBlendSpaceGraph_0.RightSide_Forward_AimOffset.R_FR_Down
  nodes [
   n1 = "Sequence Player (additive) 'CherrySmoke_DualRevolvers_Dodge_RightSide_F_AO_D50'"() {
    bLoopAnimation=false
    sequence=&CherrySmoke_DualRevolvers_Dodge_RightSide_F_AO_D50
   }
   Output Animation Pose(Result=@n1)
   "Sequence Player (additive) 'CherrySmoke_DualRevolvers_Dodge_RightSide_FR_AO_D50'"() {
    bLoopAnimation=false
    sequence=&CherrySmoke_DualRevolvers_Dodge_RightSide_FR_AO_D50
   }
  ]
 }
 R_FR_Up {
  path=ItemFullBodyOverride.AnimGraphNode_StateMachine_0.Dodge SM.AnimStateNode_5.In Air.AnimGraphNode_StateMachine_1.Dodge Directional.AnimStateNode_1.Right.AnimGraphNode_BlendSpaceGraph_0.Dodge Right.Right_Forward_InAir.AnimGraphNode_RotationOffsetBlendSpaceGraph_0.RightSide_Forward_AimOffset.R_FR_Up
  nodes [
   n1 = "Sequence Player (additive) 'CherrySmoke_DualRevolvers_Dodge_RightSide_F_AO_U20'"() {
    bLoopAnimation=false
    sequence=&CherrySmoke_DualRevolvers_Dodge_RightSide_F_AO_U20
   }
   Output Animation Pose(Result=@n1)
   "Sequence Player (additive) 'CherrySmoke_DualRevolvers_Dodge_RightSide_FR_AO_U40'"() {
    bLoopAnimation=false
    sequence=&CherrySmoke_DualRevolvers_Dodge_RightSide_FR_AO_U40
   }
  ]
 }
 Forward Firing {
  entry=Empty
  path=ItemFullBodyOverride.AnimGraphNode_StateMachine_0.Dodge SM.AnimStateNode_5.In Air.AnimGraphNode_StateMachine_1.Dodge Directional.AnimStateNode_1.Right.AnimGraphNode_BlendSpaceGraph_0.Dodge Right.Right_Forward_InAir.AnimGraphNode_StateMachine_0.Forward Firing
  ids=id
  nodes [
   AnimStateConduitNode_0 = Is Firing()
   AnimStateNode_0 = Left Fire() {bAlwaysResetOnEntry=true}
   AnimStateNode_1 = Empty()
   AnimStateNode_2 = Right Fire() {bAlwaysResetOnEntry=true}
   AnimStateTransitionNode_0 = AnimStateTransitionNode()
   AnimStateTransitionNode_1 = AnimStateTransitionNode() {crossfadeDuration=0}
   AnimStateTransitionNode_2 = AnimStateTransitionNode() {crossfadeDuration=0}
   AnimStateTransitionNode_34 = AnimStateTransitionNode() {bAutomaticRuleBasedOnSequencePlayerInState=true}
   AnimStateTransitionNode_35 = AnimStateTransitionNode() {bAutomaticRuleBasedOnSequencePlayerInState=true}
  ]
  links [
   {
    dst=Is Firing
    rule=ItemFullBodyOverride.AnimGraphNode_StateMachine_0.Dodge SM.AnimStateNode_5.In Air.AnimGraphNode_StateMachine_1.Dodge Directional.AnimStateNode_1.Right.AnimGraphNode_BlendSpaceGraph_0.Dodge Right.Right_Forward_InAir.AnimGraphNode_StateMachine_0.Forward Firing.AnimStateTransitionNode_0.Transition
    src=Empty
    via=AnimStateTransitionNode_0
   }
   {
    dst=Left Fire
    rule=ItemFullBodyOverride.AnimGraphNode_StateMachine_0.Dodge SM.AnimStateNode_5.In Air.AnimGraphNode_StateMachine_1.Dodge Directional.AnimStateNode_1.Right.AnimGraphNode_BlendSpaceGraph_0.Dodge Right.Right_Forward_InAir.AnimGraphNode_StateMachine_0.Forward Firing.AnimStateTransitionNode_1.Transition
    src=Is Firing
    via=AnimStateTransitionNode_1
   }
   {
    dst=Right Fire
    rule=ItemFullBodyOverride.AnimGraphNode_StateMachine_0.Dodge SM.AnimStateNode_5.In Air.AnimGraphNode_StateMachine_1.Dodge Directional.AnimStateNode_1.Right.AnimGraphNode_BlendSpaceGraph_0.Dodge Right.Right_Forward_InAir.AnimGraphNode_StateMachine_0.Forward Firing.AnimStateTransitionNode_2.Transition
    src=Is Firing
    via=AnimStateTransitionNode_2
   }
   {
    dst=Is Firing
    rule=ItemFullBodyOverride.AnimGraphNode_StateMachine_0.Dodge SM.AnimStateNode_5.In Air.AnimGraphNode_StateMachine_1.Dodge Directional.AnimStateNode_1.Right.AnimGraphNode_BlendSpaceGraph_0.Dodge Right.Right_Forward_InAir.AnimGraphNode_StateMachine_0.Forward Firing.AnimStateTransitionNode_34.Transition
    src=Left Fire
    via=AnimStateTransitionNode_34
   }
   {
    dst=Is Firing
    rule=ItemFullBodyOverride.AnimGraphNode_StateMachine_0.Dodge SM.AnimStateNode_5.In Air.AnimGraphNode_StateMachine_1.Dodge Directional.AnimStateNode_1.Right.AnimGraphNode_BlendSpaceGraph_0.Dodge Right.Right_Forward_InAir.AnimGraphNode_StateMachine_0.Forward Firing.AnimStateTransitionNode_35.Transition
    src=Right Fire
    via=AnimStateTransitionNode_35
   }
  ]
 }
 Is Firing {
  path=ItemFullBodyOverride.AnimGraphNode_StateMachine_0.Dodge SM.AnimStateNode_5.In Air.AnimGraphNode_StateMachine_1.Dodge Directional.AnimStateNode_1.Right.AnimGraphNode_BlendSpaceGraph_0.Dodge Right.Right_Forward_InAir.AnimGraphNode_StateMachine_0.Forward Firing.AnimStateConduitNode_0.Is Firing
  nodes [Result(bCanEnterTransition=Property Access() -> (Value=~))]
 }
 Left Fire {
  path=ItemFullBodyOverride.AnimGraphNode_StateMachine_0.Dodge SM.AnimStateNode_5.In Air.AnimGraphNode_StateMachine_1.Dodge Directional.AnimStateNode_1.Right.AnimGraphNode_BlendSpaceGraph_0.Dodge Right.Right_Forward_InAir.AnimGraphNode_StateMachine_0.Forward Firing.AnimStateNode_0.Left Fire
  nodes [
   n1 = "Sequence Player (additive) 'CherrySmoke_DualRevolvers_front_dodge_fire_L'"(PlayRate=Get AdditiveLoopPlayRate()) {bLoopAnimation=false, sequence=&CherrySmoke_DualRevolvers_front_dodge_fire_L}
   Output Animation Pose(Result=@n1) {is=title AnimGraphNode_StateResult}
  ]
 }
 Empty {
  path=ItemFullBodyOverride.AnimGraphNode_StateMachine_0.Dodge SM.AnimStateNode_5.In Air.AnimGraphNode_StateMachine_1.Dodge Directional.AnimStateNode_1.Right.AnimGraphNode_BlendSpaceGraph_0.Dodge Right.Right_Forward_InAir.AnimGraphNode_StateMachine_0.Forward Firing.AnimStateNode_1.Empty
  nodes [Output Animation Pose(Result=Additive Identity Pose())]
 }
 Right Fire {
  path=ItemFullBodyOverride.AnimGraphNode_StateMachine_0.Dodge SM.AnimStateNode_5.In Air.AnimGraphNode_StateMachine_1.Dodge Directional.AnimStateNode_1.Right.AnimGraphNode_BlendSpaceGraph_0.Dodge Right.Right_Forward_InAir.AnimGraphNode_StateMachine_0.Forward Firing.AnimStateNode_2.Right Fire
  nodes [
   n1 = "Sequence Player (additive) 'CherrySmoke_DualRevolvers_front_dodge_fire_R'"(PlayRate=Get AdditiveLoopPlayRate()) {bLoopAnimation=false, sequence=&CherrySmoke_DualRevolvers_front_dodge_fire_R}
   Output Animation Pose(Result=@n1)
  ]
 }
 Transition {
  path=ItemFullBodyOverride.AnimGraphNode_StateMachine_0.Dodge SM.AnimStateNode_5.In Air.AnimGraphNode_StateMachine_1.Dodge Directional.AnimStateNode_1.Right.AnimGraphNode_BlendSpaceGraph_0.Dodge Right.Right_Forward_InAir.AnimGraphNode_StateMachine_0.Forward Firing.AnimStateTransitionNode_0.Transition
  nodes [Result(bCanEnterTransition=true)]
 }
 Transition {
  path=ItemFullBodyOverride.AnimGraphNode_StateMachine_0.Dodge SM.AnimStateNode_5.In Air.AnimGraphNode_StateMachine_1.Dodge Directional.AnimStateNode_1.Right.AnimGraphNode_BlendSpaceGraph_0.Dodge Right.Right_Forward_InAir.AnimGraphNode_StateMachine_0.Forward Firing.AnimStateTransitionNode_1.Transition
  nodes [Result(bCanEnterTransition=NOT Boolean(A=Get bIsLastHandFireLeft()))]
 }
 Transition {
  path=ItemFullBodyOverride.AnimGraphNode_StateMachine_0.Dodge SM.AnimStateNode_5.In Air.AnimGraphNode_StateMachine_1.Dodge Directional.AnimStateNode_1.Right.AnimGraphNode_BlendSpaceGraph_0.Dodge Right.Right_Forward_InAir.AnimGraphNode_StateMachine_0.Forward Firing.AnimStateTransitionNode_2.Transition
  nodes [Result(bCanEnterTransition=Get bIsLastHandFireLeft())]
 }
 Transition {
  path=ItemFullBodyOverride.AnimGraphNode_StateMachine_0.Dodge SM.AnimStateNode_5.In Air.AnimGraphNode_StateMachine_1.Dodge Directional.AnimStateNode_1.Right.AnimGraphNode_BlendSpaceGraph_0.Dodge Right.Right_Forward_InAir.AnimGraphNode_StateMachine_0.Forward Firing.AnimStateTransitionNode_34.Transition
  nodes [Result()]
 }
 Transition {
  path=ItemFullBodyOverride.AnimGraphNode_StateMachine_0.Dodge SM.AnimStateNode_5.In Air.AnimGraphNode_StateMachine_1.Dodge Directional.AnimStateNode_1.Right.AnimGraphNode_BlendSpaceGraph_0.Dodge Right.Right_Forward_InAir.AnimGraphNode_StateMachine_0.Forward Firing.AnimStateTransitionNode_35.Transition
  nodes [Result()]
 }
 Right_Left_InAir {
  path=ItemFullBodyOverride.AnimGraphNode_StateMachine_0.Dodge SM.AnimStateNode_5.In Air.AnimGraphNode_StateMachine_1.Dodge Directional.AnimStateNode_1.Right.AnimGraphNode_BlendSpaceGraph_0.Dodge Right.Right_Left_InAir
  nodes [
   AnimGraphNode_RotationOffsetBlendSpaceGraph_0 = AimOffset 'RightSide_Left_AimOffset'!(
    X:float=Get AimPitch()
    BasePose:Pose Link Structure=Sequence Player 'CherrySmoke_DualRevolvers_Dodge_RightSide_L_Pose'() {bLoopAnimation=false, sequence=/CherrySmokeDualRevolversGameplay/Assets/Anim/Poses/CherrySmoke_DualRevolvers_Dodge_RightSide_L_Pose}
    Alpha:float=1
   ) -> (Pose:Pose Link Structure) {is=title AnimGraphNode_RotationOffsetBlendSpaceGraph}
   AnimGraphNode_StateMachine_0 = Left Firing()
   Output Animation Pose(Result=Apply Additive(Base=@AnimGraphNode_RotationOffsetBlendSpaceGraph_0, Additive=@AnimGraphNode_StateMachine_0) {alphaBoolBlend {blendOption=Linear}}) {is=title AnimGraphNode_BlendSpaceSampleResult}
  ]
 }
 RightSide_Left_AimOffset {
  path=ItemFullBodyOverride.AnimGraphNode_StateMachine_0.Dodge SM.AnimStateNode_5.In Air.AnimGraphNode_StateMachine_1.Dodge Directional.AnimStateNode_1.Right.AnimGraphNode_BlendSpaceGraph_0.Dodge Right.Right_Left_InAir.AnimGraphNode_RotationOffsetBlendSpaceGraph_0.RightSide_Left_AimOffset
  links []
  nodes []
 }
 R_FL_Center {
  path=ItemFullBodyOverride.AnimGraphNode_StateMachine_0.Dodge SM.AnimStateNode_5.In Air.AnimGraphNode_StateMachine_1.Dodge Directional.AnimStateNode_1.Right.AnimGraphNode_BlendSpaceGraph_0.Dodge Right.Right_Left_InAir.AnimGraphNode_RotationOffsetBlendSpaceGraph_0.RightSide_Left_AimOffset.R_FL_Center
  nodes [Output Animation Pose(Result=Additive Identity Pose())]
 }
 R_FL_Down {
  path=ItemFullBodyOverride.AnimGraphNode_StateMachine_0.Dodge SM.AnimStateNode_5.In Air.AnimGraphNode_StateMachine_1.Dodge Directional.AnimStateNode_1.Right.AnimGraphNode_BlendSpaceGraph_0.Dodge Right.Right_Left_InAir.AnimGraphNode_RotationOffsetBlendSpaceGraph_0.RightSide_Left_AimOffset.R_FL_Down
  nodes [
   n1 = "Sequence Player (additive) 'CherrySmoke_DualRevolvers_Dodge_RightSide_FL_AO_D40'"() {
    bLoopAnimation=false
    sequence=&CherrySmoke_DualRevolvers_Dodge_RightSide_FL_AO_D40
   }
   Output Animation Pose(Result=@n1)
   Sequence Player 'CherrySmoke_DualRevolvers_Dodge_RightSide_L_AO_D50'() {
    bLoopAnimation=false
    sequence=/CherrySmokeDualRevolversGameplay/Assets/Anim/Poses/Dodge_AO/CherrySmoke_DualRevolvers_Dodge_RightSide_L_AO_D50
   }
  ]
 }
 R_FL_Up {
  path=ItemFullBodyOverride.AnimGraphNode_StateMachine_0.Dodge SM.AnimStateNode_5.In Air.AnimGraphNode_StateMachine_1.Dodge Directional.AnimStateNode_1.Right.AnimGraphNode_BlendSpaceGraph_0.Dodge Right.Right_Left_InAir.AnimGraphNode_RotationOffsetBlendSpaceGraph_0.RightSide_Left_AimOffset.R_FL_Up
  nodes [
   n1 = "Sequence Player (additive) 'CherrySmoke_DualRevolvers_Dodge_RightSide_FL_AO_U40'"() {
    bLoopAnimation=false
    sequence=&CherrySmoke_DualRevolvers_Dodge_RightSide_FL_AO_U40
   }
   Output Animation Pose(Result=@n1)
   Sequence Player 'CherrySmoke_DualRevolvers_Dodge_RightSide_L_AO_U50'() {
    bLoopAnimation=false
    sequence=/CherrySmokeDualRevolversGameplay/Assets/Anim/Poses/Dodge_AO/CherrySmoke_DualRevolvers_Dodge_RightSide_L_AO_U50
   }
  ]
 }
 Left Firing {
  entry=Empty
  path=ItemFullBodyOverride.AnimGraphNode_StateMachine_0.Dodge SM.AnimStateNode_5.In Air.AnimGraphNode_StateMachine_1.Dodge Directional.AnimStateNode_1.Right.AnimGraphNode_BlendSpaceGraph_0.Dodge Right.Right_Left_InAir.AnimGraphNode_StateMachine_0.Left Firing
  ids=id
  nodes [
   AnimStateConduitNode_0 = Is Firing()
   AnimStateNode_0 = Left Fire() {bAlwaysResetOnEntry=true}
   AnimStateNode_1 = Empty()
   AnimStateNode_2 = Right Fire() {bAlwaysResetOnEntry=true}
   AnimStateTransitionNode_0 = AnimStateTransitionNode()
   AnimStateTransitionNode_1 = AnimStateTransitionNode() {crossfadeDuration=0.05}
   AnimStateTransitionNode_2 = AnimStateTransitionNode() {crossfadeDuration=0.05}
   AnimStateTransitionNode_28 = AnimStateTransitionNode() {bAutomaticRuleBasedOnSequencePlayerInState=true}
   AnimStateTransitionNode_29 = AnimStateTransitionNode() {bAutomaticRuleBasedOnSequencePlayerInState=true}
  ]
  links [
   {
    dst=Is Firing
    rule=ItemFullBodyOverride.AnimGraphNode_StateMachine_0.Dodge SM.AnimStateNode_5.In Air.AnimGraphNode_StateMachine_1.Dodge Directional.AnimStateNode_1.Right.AnimGraphNode_BlendSpaceGraph_0.Dodge Right.Right_Left_InAir.AnimGraphNode_StateMachine_0.Left Firing.AnimStateTransitionNode_0.Transition
    src=Empty
    via=AnimStateTransitionNode_0
   }
   {
    dst=Left Fire
    rule=ItemFullBodyOverride.AnimGraphNode_StateMachine_0.Dodge SM.AnimStateNode_5.In Air.AnimGraphNode_StateMachine_1.Dodge Directional.AnimStateNode_1.Right.AnimGraphNode_BlendSpaceGraph_0.Dodge Right.Right_Left_InAir.AnimGraphNode_StateMachine_0.Left Firing.AnimStateTransitionNode_1.Transition
    src=Is Firing
    via=AnimStateTransitionNode_1
   }
   {
    dst=Right Fire
    rule=ItemFullBodyOverride.AnimGraphNode_StateMachine_0.Dodge SM.AnimStateNode_5.In Air.AnimGraphNode_StateMachine_1.Dodge Directional.AnimStateNode_1.Right.AnimGraphNode_BlendSpaceGraph_0.Dodge Right.Right_Left_InAir.AnimGraphNode_StateMachine_0.Left Firing.AnimStateTransitionNode_2.Transition
    src=Is Firing
    via=AnimStateTransitionNode_2
   }
   {
    dst=Is Firing
    rule=ItemFullBodyOverride.AnimGraphNode_StateMachine_0.Dodge SM.AnimStateNode_5.In Air.AnimGraphNode_StateMachine_1.Dodge Directional.AnimStateNode_1.Right.AnimGraphNode_BlendSpaceGraph_0.Dodge Right.Right_Left_InAir.AnimGraphNode_StateMachine_0.Left Firing.AnimStateTransitionNode_28.Transition
    src=Left Fire
    via=AnimStateTransitionNode_28
   }
   {
    dst=Is Firing
    rule=ItemFullBodyOverride.AnimGraphNode_StateMachine_0.Dodge SM.AnimStateNode_5.In Air.AnimGraphNode_StateMachine_1.Dodge Directional.AnimStateNode_1.Right.AnimGraphNode_BlendSpaceGraph_0.Dodge Right.Right_Left_InAir.AnimGraphNode_StateMachine_0.Left Firing.AnimStateTransitionNode_29.Transition
    src=Right Fire
    via=AnimStateTransitionNode_29
   }
  ]
 }
 Is Firing {
  path=ItemFullBodyOverride.AnimGraphNode_StateMachine_0.Dodge SM.AnimStateNode_5.In Air.AnimGraphNode_StateMachine_1.Dodge Directional.AnimStateNode_1.Right.AnimGraphNode_BlendSpaceGraph_0.Dodge Right.Right_Left_InAir.AnimGraphNode_StateMachine_0.Left Firing.AnimStateConduitNode_0.Is Firing
  nodes [Result(bCanEnterTransition=Property Access() -> (Value=~))]
 }
 Left Fire {
  path=ItemFullBodyOverride.AnimGraphNode_StateMachine_0.Dodge SM.AnimStateNode_5.In Air.AnimGraphNode_StateMachine_1.Dodge Directional.AnimStateNode_1.Right.AnimGraphNode_BlendSpaceGraph_0.Dodge Right.Right_Left_InAir.AnimGraphNode_StateMachine_0.Left Firing.AnimStateNode_0.Left Fire
  nodes [
   n1 = "Sequence Player (additive) 'CherrySmoke_DualRevolvers_left_dodge_fire_L'"(PlayRate=Get AdditiveLoopPlayRate()) {bLoopAnimation=false, sequence=&CherrySmoke_DualRevolvers_left_dodge_fire_L}
   Output Animation Pose(Result=@n1) {is=title AnimGraphNode_StateResult}
  ]
 }
 Empty {
  path=ItemFullBodyOverride.AnimGraphNode_StateMachine_0.Dodge SM.AnimStateNode_5.In Air.AnimGraphNode_StateMachine_1.Dodge Directional.AnimStateNode_1.Right.AnimGraphNode_BlendSpaceGraph_0.Dodge Right.Right_Left_InAir.AnimGraphNode_StateMachine_0.Left Firing.AnimStateNode_1.Empty
  nodes [Output Animation Pose(Result=Additive Identity Pose())]
 }
 Right Fire {
  path=ItemFullBodyOverride.AnimGraphNode_StateMachine_0.Dodge SM.AnimStateNode_5.In Air.AnimGraphNode_StateMachine_1.Dodge Directional.AnimStateNode_1.Right.AnimGraphNode_BlendSpaceGraph_0.Dodge Right.Right_Left_InAir.AnimGraphNode_StateMachine_0.Left Firing.AnimStateNode_2.Right Fire
  nodes [
   n1 = "Sequence Player (additive) 'CherrySmoke_DualRevolvers_left_dodge_fire_R'"(PlayRate=Get AdditiveLoopPlayRate()) {bLoopAnimation=false, sequence=&CherrySmoke_DualRevolvers_left_dodge_fire_R}
   Output Animation Pose(Result=@n1)
  ]
 }
 Transition {
  path=ItemFullBodyOverride.AnimGraphNode_StateMachine_0.Dodge SM.AnimStateNode_5.In Air.AnimGraphNode_StateMachine_1.Dodge Directional.AnimStateNode_1.Right.AnimGraphNode_BlendSpaceGraph_0.Dodge Right.Right_Left_InAir.AnimGraphNode_StateMachine_0.Left Firing.AnimStateTransitionNode_0.Transition
  nodes [Result(bCanEnterTransition=true)]
 }
 Transition {
  path=ItemFullBodyOverride.AnimGraphNode_StateMachine_0.Dodge SM.AnimStateNode_5.In Air.AnimGraphNode_StateMachine_1.Dodge Directional.AnimStateNode_1.Right.AnimGraphNode_BlendSpaceGraph_0.Dodge Right.Right_Left_InAir.AnimGraphNode_StateMachine_0.Left Firing.AnimStateTransitionNode_1.Transition
  nodes [Result(bCanEnterTransition=NOT Boolean(A=Get bIsLastHandFireLeft()))]
 }
 Transition {
  path=ItemFullBodyOverride.AnimGraphNode_StateMachine_0.Dodge SM.AnimStateNode_5.In Air.AnimGraphNode_StateMachine_1.Dodge Directional.AnimStateNode_1.Right.AnimGraphNode_BlendSpaceGraph_0.Dodge Right.Right_Left_InAir.AnimGraphNode_StateMachine_0.Left Firing.AnimStateTransitionNode_2.Transition
  nodes [Result(bCanEnterTransition=Get bIsLastHandFireLeft())]
 }
 Transition {
  path=ItemFullBodyOverride.AnimGraphNode_StateMachine_0.Dodge SM.AnimStateNode_5.In Air.AnimGraphNode_StateMachine_1.Dodge Directional.AnimStateNode_1.Right.AnimGraphNode_BlendSpaceGraph_0.Dodge Right.Right_Left_InAir.AnimGraphNode_StateMachine_0.Left Firing.AnimStateTransitionNode_28.Transition
  nodes [Result()]
 }
 Transition {
  path=ItemFullBodyOverride.AnimGraphNode_StateMachine_0.Dodge SM.AnimStateNode_5.In Air.AnimGraphNode_StateMachine_1.Dodge Directional.AnimStateNode_1.Right.AnimGraphNode_BlendSpaceGraph_0.Dodge Right.Right_Left_InAir.AnimGraphNode_StateMachine_0.Left Firing.AnimStateTransitionNode_29.Transition
  nodes [Result()]
 }
 Right_Right_InAir {
  path=ItemFullBodyOverride.AnimGraphNode_StateMachine_0.Dodge SM.AnimStateNode_5.In Air.AnimGraphNode_StateMachine_1.Dodge Directional.AnimStateNode_1.Right.AnimGraphNode_BlendSpaceGraph_0.Dodge Right.Right_Right_InAir
  nodes [
   n1 = Sequence Player 'CherrySmoke_DualRevolver_Dodge_RightSide_R_Diving'() {bLoopAnimation=false, sequence=/CherrySmokeDualRevolversGameplay/Assets/Anim/CharacterAnim/CherrySmoke_DualRevolver_Dodge_RightSide_R_Diving}
   AnimGraphNode_RotationOffsetBlendSpaceGraph_0 = AimOffset 'RightSide_Right_AimOffset'!(X:float=Get AimPitch(), BasePose:Pose Link Structure=@n1, Alpha:float=1) -> (Pose:Pose Link Structure) {is=title AnimGraphNode_RotationOffsetBlendSpaceGraph}
   AnimGraphNode_StateMachine_0 = Right Firing()
   Output Animation Pose(Result=Apply Additive(Base=@AnimGraphNode_RotationOffsetBlendSpaceGraph_0, Additive=@AnimGraphNode_StateMachine_0) {alphaBoolBlend {blendOption=Linear}}) {is=title AnimGraphNode_BlendSpaceSampleResult}
  ]
 }
 RightSide_Right_AimOffset {
  path=ItemFullBodyOverride.AnimGraphNode_StateMachine_0.Dodge SM.AnimStateNode_5.In Air.AnimGraphNode_StateMachine_1.Dodge Directional.AnimStateNode_1.Right.AnimGraphNode_BlendSpaceGraph_0.Dodge Right.Right_Right_InAir.AnimGraphNode_RotationOffsetBlendSpaceGraph_0.RightSide_Right_AimOffset
  links []
  nodes []
 }
 R_R_Center {
  path=ItemFullBodyOverride.AnimGraphNode_StateMachine_0.Dodge SM.AnimStateNode_5.In Air.AnimGraphNode_StateMachine_1.Dodge Directional.AnimStateNode_1.Right.AnimGraphNode_BlendSpaceGraph_0.Dodge Right.Right_Right_InAir.AnimGraphNode_RotationOffsetBlendSpaceGraph_0.RightSide_Right_AimOffset.R_R_Center
  nodes [Output Animation Pose(Result=Additive Identity Pose())]
 }
 R_R_Down {
  path=ItemFullBodyOverride.AnimGraphNode_StateMachine_0.Dodge SM.AnimStateNode_5.In Air.AnimGraphNode_StateMachine_1.Dodge Directional.AnimStateNode_1.Right.AnimGraphNode_BlendSpaceGraph_0.Dodge Right.Right_Right_InAir.AnimGraphNode_RotationOffsetBlendSpaceGraph_0.RightSide_Right_AimOffset.R_R_Down
  nodes [
   n1 = "Sequence Player (additive) 'CherrySmoke_DualRevolvers_Dodge_RightSide_R_AO_D50'"() {
    bLoopAnimation=false
    sequence=/CherrySmokeDualRevolversGameplay/Assets/Anim/Poses/Dodge_AO/CherrySmoke_DualRevolvers_Dodge_RightSide_R_AO_D50
   }
   Output Animation Pose(Result=@n1)
  ]
 }
 R_R_Up {
  path=ItemFullBodyOverride.AnimGraphNode_StateMachine_0.Dodge SM.AnimStateNode_5.In Air.AnimGraphNode_StateMachine_1.Dodge Directional.AnimStateNode_1.Right.AnimGraphNode_BlendSpaceGraph_0.Dodge Right.Right_Right_InAir.AnimGraphNode_RotationOffsetBlendSpaceGraph_0.RightSide_Right_AimOffset.R_R_Up
  nodes [
   n1 = "Sequence Player (additive) 'CherrySmoke_DualRevolvers_Dodge_RightSide_R_AO_U50'"() {
    bLoopAnimation=false
    sequence=/CherrySmokeDualRevolversGameplay/Assets/Anim/Poses/Dodge_AO/CherrySmoke_DualRevolvers_Dodge_RightSide_R_AO_U50
   }
   Output Animation Pose(Result=@n1)
  ]
 }
 Right Firing {
  entry=Empty
  path=ItemFullBodyOverride.AnimGraphNode_StateMachine_0.Dodge SM.AnimStateNode_5.In Air.AnimGraphNode_StateMachine_1.Dodge Directional.AnimStateNode_1.Right.AnimGraphNode_BlendSpaceGraph_0.Dodge Right.Right_Right_InAir.AnimGraphNode_StateMachine_0.Right Firing
  ids=id
  nodes [
   AnimStateConduitNode_0 = Is Firing()
   AnimStateNode_0 = Left Fire() {bAlwaysResetOnEntry=true}
   AnimStateNode_1 = Empty()
   AnimStateNode_2 = Right Fire() {bAlwaysResetOnEntry=true}
   AnimStateTransitionNode_22 = AnimStateTransitionNode() {bAutomaticRuleBasedOnSequencePlayerInState=true}
   AnimStateTransitionNode_23 = AnimStateTransitionNode() {bAutomaticRuleBasedOnSequencePlayerInState=true}
   AnimStateTransitionNode_3 = AnimStateTransitionNode()
   AnimStateTransitionNode_4 = AnimStateTransitionNode() {crossfadeDuration=0.05}
   AnimStateTransitionNode_5 = AnimStateTransitionNode() {crossfadeDuration=0.05}
  ]
  links [
   {
    dst=Is Firing
    rule=ItemFullBodyOverride.AnimGraphNode_StateMachine_0.Dodge SM.AnimStateNode_5.In Air.AnimGraphNode_StateMachine_1.Dodge Directional.AnimStateNode_1.Right.AnimGraphNode_BlendSpaceGraph_0.Dodge Right.Right_Right_InAir.AnimGraphNode_StateMachine_0.Right Firing.AnimStateTransitionNode_3.Transition
    src=Empty
    via=AnimStateTransitionNode_3
   }
   {
    dst=Left Fire
    rule=ItemFullBodyOverride.AnimGraphNode_StateMachine_0.Dodge SM.AnimStateNode_5.In Air.AnimGraphNode_StateMachine_1.Dodge Directional.AnimStateNode_1.Right.AnimGraphNode_BlendSpaceGraph_0.Dodge Right.Right_Right_InAir.AnimGraphNode_StateMachine_0.Right Firing.AnimStateTransitionNode_4.Transition
    src=Is Firing
    via=AnimStateTransitionNode_4
   }
   {
    dst=Right Fire
    rule=ItemFullBodyOverride.AnimGraphNode_StateMachine_0.Dodge SM.AnimStateNode_5.In Air.AnimGraphNode_StateMachine_1.Dodge Directional.AnimStateNode_1.Right.AnimGraphNode_BlendSpaceGraph_0.Dodge Right.Right_Right_InAir.AnimGraphNode_StateMachine_0.Right Firing.AnimStateTransitionNode_5.Transition
    src=Is Firing
    via=AnimStateTransitionNode_5
   }
   {
    dst=Is Firing
    rule=ItemFullBodyOverride.AnimGraphNode_StateMachine_0.Dodge SM.AnimStateNode_5.In Air.AnimGraphNode_StateMachine_1.Dodge Directional.AnimStateNode_1.Right.AnimGraphNode_BlendSpaceGraph_0.Dodge Right.Right_Right_InAir.AnimGraphNode_StateMachine_0.Right Firing.AnimStateTransitionNode_22.Transition
    src=Left Fire
    via=AnimStateTransitionNode_22
   }
   {
    dst=Is Firing
    rule=ItemFullBodyOverride.AnimGraphNode_StateMachine_0.Dodge SM.AnimStateNode_5.In Air.AnimGraphNode_StateMachine_1.Dodge Directional.AnimStateNode_1.Right.AnimGraphNode_BlendSpaceGraph_0.Dodge Right.Right_Right_InAir.AnimGraphNode_StateMachine_0.Right Firing.AnimStateTransitionNode_23.Transition
    src=Right Fire
    via=AnimStateTransitionNode_23
   }
  ]
 }
 Is Firing {
  path=ItemFullBodyOverride.AnimGraphNode_StateMachine_0.Dodge SM.AnimStateNode_5.In Air.AnimGraphNode_StateMachine_1.Dodge Directional.AnimStateNode_1.Right.AnimGraphNode_BlendSpaceGraph_0.Dodge Right.Right_Right_InAir.AnimGraphNode_StateMachine_0.Right Firing.AnimStateConduitNode_0.Is Firing
  nodes [Result(bCanEnterTransition=Property Access() -> (Value=~))]
 }
 Left Fire {
  path=ItemFullBodyOverride.AnimGraphNode_StateMachine_0.Dodge SM.AnimStateNode_5.In Air.AnimGraphNode_StateMachine_1.Dodge Directional.AnimStateNode_1.Right.AnimGraphNode_BlendSpaceGraph_0.Dodge Right.Right_Right_InAir.AnimGraphNode_StateMachine_0.Right Firing.AnimStateNode_0.Left Fire
  nodes [
   n1 = "Sequence Player (additive) 'CherrySmoke_DualRevolvers_right_dodge_fire_L'"(PlayRate=Get AdditiveLoopPlayRate()) {bLoopAnimation=false, sequence=&CherrySmoke_DualRevolvers_right_dodge_fire_L}
   Output Animation Pose(Result=@n1) {is=title AnimGraphNode_StateResult}
  ]
 }
 Empty {
  path=ItemFullBodyOverride.AnimGraphNode_StateMachine_0.Dodge SM.AnimStateNode_5.In Air.AnimGraphNode_StateMachine_1.Dodge Directional.AnimStateNode_1.Right.AnimGraphNode_BlendSpaceGraph_0.Dodge Right.Right_Right_InAir.AnimGraphNode_StateMachine_0.Right Firing.AnimStateNode_1.Empty
  nodes [Output Animation Pose(Result=Additive Identity Pose())]
 }
 Right Fire {
  path=ItemFullBodyOverride.AnimGraphNode_StateMachine_0.Dodge SM.AnimStateNode_5.In Air.AnimGraphNode_StateMachine_1.Dodge Directional.AnimStateNode_1.Right.AnimGraphNode_BlendSpaceGraph_0.Dodge Right.Right_Right_InAir.AnimGraphNode_StateMachine_0.Right Firing.AnimStateNode_2.Right Fire
  nodes [
   n1 = "Sequence Player (additive) 'CherrySmoke_DualRevolvers_right_dodge_fire_R'"(PlayRate=Get AdditiveLoopPlayRate()) {bLoopAnimation=false, sequence=&CherrySmoke_DualRevolvers_right_dodge_fire_R}
   Output Animation Pose(Result=@n1)
  ]
 }
 Transition {
  path=ItemFullBodyOverride.AnimGraphNode_StateMachine_0.Dodge SM.AnimStateNode_5.In Air.AnimGraphNode_StateMachine_1.Dodge Directional.AnimStateNode_1.Right.AnimGraphNode_BlendSpaceGraph_0.Dodge Right.Right_Right_InAir.AnimGraphNode_StateMachine_0.Right Firing.AnimStateTransitionNode_22.Transition
  nodes [Result()]
 }
 Transition {
  path=ItemFullBodyOverride.AnimGraphNode_StateMachine_0.Dodge SM.AnimStateNode_5.In Air.AnimGraphNode_StateMachine_1.Dodge Directional.AnimStateNode_1.Right.AnimGraphNode_BlendSpaceGraph_0.Dodge Right.Right_Right_InAir.AnimGraphNode_StateMachine_0.Right Firing.AnimStateTransitionNode_23.Transition
  nodes [Result()]
 }
 Transition {
  path=ItemFullBodyOverride.AnimGraphNode_StateMachine_0.Dodge SM.AnimStateNode_5.In Air.AnimGraphNode_StateMachine_1.Dodge Directional.AnimStateNode_1.Right.AnimGraphNode_BlendSpaceGraph_0.Dodge Right.Right_Right_InAir.AnimGraphNode_StateMachine_0.Right Firing.AnimStateTransitionNode_3.Transition
  nodes [Result(bCanEnterTransition=true)]
 }
 Transition {
  path=ItemFullBodyOverride.AnimGraphNode_StateMachine_0.Dodge SM.AnimStateNode_5.In Air.AnimGraphNode_StateMachine_1.Dodge Directional.AnimStateNode_1.Right.AnimGraphNode_BlendSpaceGraph_0.Dodge Right.Right_Right_InAir.AnimGraphNode_StateMachine_0.Right Firing.AnimStateTransitionNode_4.Transition
  nodes [Result(bCanEnterTransition=NOT Boolean(A=Get bIsLastHandFireLeft()))]
 }
 Transition {
  path=ItemFullBodyOverride.AnimGraphNode_StateMachine_0.Dodge SM.AnimStateNode_5.In Air.AnimGraphNode_StateMachine_1.Dodge Directional.AnimStateNode_1.Right.AnimGraphNode_BlendSpaceGraph_0.Dodge Right.Right_Right_InAir.AnimGraphNode_StateMachine_0.Right Firing.AnimStateTransitionNode_5.Transition
  nodes [Result(bCanEnterTransition=Get bIsLastHandFireLeft())]
 }
 Transition {
  path=ItemFullBodyOverride.AnimGraphNode_StateMachine_0.Dodge SM.AnimStateNode_5.In Air.AnimGraphNode_StateMachine_1.Dodge Directional.AnimStateTransitionNode_2.Transition
  nodes [Result(bCanEnterTransition=Get bIsLeftDodgeSpace())]
 }
 Transition {
  path=ItemFullBodyOverride.AnimGraphNode_StateMachine_0.Dodge SM.AnimStateNode_5.In Air.AnimGraphNode_StateMachine_1.Dodge Directional.AnimStateTransitionNode_3.Transition
  nodes [Result(bCanEnterTransition=NOT Boolean(A=Get bIsLeftDodgeSpace()))]
 }
 Fall {
  path=ItemFullBodyOverride.AnimGraphNode_StateMachine_0.Dodge SM.AnimStateNode_7.Fall
  nodes [
   AnimGraphNode_BlendSpaceGraph_1 = Blendspace 'Dodge Fall'!(X:float=Get DodgeAngleRadians()) -> (Pose:Pose Link Structure) {is=title AnimGraphNode_BlendSpaceGraph}
   Output Animation Pose(Result=@AnimGraphNode_BlendSpaceGraph_1)
  ]
 }
 Dodge Fall {
  path=ItemFullBodyOverride.AnimGraphNode_StateMachine_0.Dodge SM.AnimStateNode_7.Fall.AnimGraphNode_BlendSpaceGraph_1.Dodge Fall
  links []
  nodes []
 }
 BackwardDodgeLoop_L {
  path=ItemFullBodyOverride.AnimGraphNode_StateMachine_0.Dodge SM.AnimStateNode_7.Fall.AnimGraphNode_BlendSpaceGraph_1.Dodge Fall.BackwardDodgeLoop_L
  nodes [Output Animation Pose(Result=Sequence Player 'CherrySmoke_DualRevolver_Dodge_Antic_Fall_B_anim'() {bLoopAnimation=false, sequence=/CherrySmokeDualRevolversGameplay/Assets/Anim/TEMP/CherrySmoke_DualRevolver_Dodge_Antic_Fall_B_anim}) {is=title AnimGraphNode_BlendSpaceSampleResult}]
 }
 BackwardsDodgeLoop_R {
  path=ItemFullBodyOverride.AnimGraphNode_StateMachine_0.Dodge SM.AnimStateNode_7.Fall.AnimGraphNode_BlendSpaceGraph_1.Dodge Fall.BackwardsDodgeLoop_R
  nodes [Output Animation Pose(Result=Sequence Player 'CherrySmoke_DualRevolver_Dodge_Antic_Fall_B_anim'() {bLoopAnimation=false, sequence=&CherrySmoke_DualRevolver_Dodge_Antic_Fall_B_anim})]
 }
 ForwardDodgeLoop {
  path=ItemFullBodyOverride.AnimGraphNode_StateMachine_0.Dodge SM.AnimStateNode_7.Fall.AnimGraphNode_BlendSpaceGraph_1.Dodge Fall.ForwardDodgeLoop
  nodes [Output Animation Pose(Result=Sequence Player 'CherrySmoke_DualRevolver_Dodge_Antic_Fall_F_anim'() {bLoopAnimation=false, sequence=/CherrySmokeDualRevolversGameplay/Assets/Anim/TEMP/CherrySmoke_DualRevolver_Dodge_Antic_Fall_F_anim})]
 }
 LeftDodgeLoop {
  path=ItemFullBodyOverride.AnimGraphNode_StateMachine_0.Dodge SM.AnimStateNode_7.Fall.AnimGraphNode_BlendSpaceGraph_1.Dodge Fall.LeftDodgeLoop
  nodes [Output Animation Pose(Result=Sequence Player 'CherrySmoke_DualRevolver_Dodge_Antic_Fall_L_anim'() {bLoopAnimation=false, sequence=/CherrySmokeDualRevolversGameplay/Assets/Anim/TEMP/CherrySmoke_DualRevolver_Dodge_Antic_Fall_L_anim})]
 }
 RightDodgeLoop {
  path=ItemFullBodyOverride.AnimGraphNode_StateMachine_0.Dodge SM.AnimStateNode_7.Fall.AnimGraphNode_BlendSpaceGraph_1.Dodge Fall.RightDodgeLoop
  nodes [Output Animation Pose(Result=Sequence Player 'CherrySmoke_DualRevolver_Dodge_Antic_Fall_R_anim'() {bLoopAnimation=false, sequence=/CherrySmokeDualRevolversGameplay/Assets/Anim/TEMP/CherrySmoke_DualRevolver_Dodge_Antic_Fall_R_anim})]
 }
 Transition {
  path=ItemFullBodyOverride.AnimGraphNode_StateMachine_0.Dodge SM.AnimStateTransitionNode_0.Transition
  nodes [Result(bCanEnterTransition=Get bIncompatibleMovement() -> (bIncompatibleMovement:bool))]
 }
 Transition {
  path=ItemFullBodyOverride.AnimGraphNode_StateMachine_0.Dodge SM.AnimStateTransitionNode_1.Transition
  nodes [Result(bCanEnterTransition=true)]
 }
 Transition {
  path=ItemFullBodyOverride.AnimGraphNode_StateMachine_0.Dodge SM.AnimStateTransitionNode_10.Transition
  nodes [Result(bCanEnterTransition=true)]
 }
 Transition {
  path=ItemFullBodyOverride.AnimGraphNode_StateMachine_0.Dodge SM.AnimStateTransitionNode_11.Transition
  nodes [Result(bCanEnterTransition=NOT Boolean(A=Get bAirDodgeActive()))]
 }
 Transition {
  path=ItemFullBodyOverride.AnimGraphNode_StateMachine_0.Dodge SM.AnimStateTransitionNode_13.Transition
  nodes [Result(bCanEnterTransition=Get bCanEnterDoubleJump() -> (bCanEnterDoubleJump:bool))]
 }
 Transition {path=ItemFullBodyOverride.AnimGraphNode_StateMachine_0.Dodge SM.AnimStateTransitionNode_14.Transition, nodes [Result()]}
 Transition {
  path=ItemFullBodyOverride.AnimGraphNode_StateMachine_0.Dodge SM.AnimStateTransitionNode_15.Transition
  nodes [Result(bCanEnterTransition=Get bCanEnterDodgeStart() -> (bCanEnterDodgeStart:bool))]
 }
 Transition {path=ItemFullBodyOverride.AnimGraphNode_StateMachine_0.Dodge SM.AnimStateTransitionNode_16.Transition, nodes [Result()]}
 Transition {path=ItemFullBodyOverride.AnimGraphNode_StateMachine_0.Dodge SM.AnimStateTransitionNode_2.Transition, nodes [Result()]}
 Transition {
  path=ItemFullBodyOverride.AnimGraphNode_StateMachine_0.Dodge SM.AnimStateTransitionNode_4.Transition
  nodes [Result(bCanEnterTransition=Property Access() -> (Value=~))]
 }
 Transition {
  path=ItemFullBodyOverride.AnimGraphNode_StateMachine_0.Dodge SM.AnimStateTransitionNode_5.Transition
  nodes [Result(bCanEnterTransition=NOT Boolean(A=Get bDoubleJumpActive() -> (bDoubleJumpActive:bool)))]
 }
 Transition {path=ItemFullBodyOverride.AnimGraphNode_StateMachine_0.Dodge SM.AnimStateTransitionNode_6.Transition, nodes [Result()]}
 Transition {
  path=ItemFullBodyOverride.AnimGraphNode_StateMachine_0.Dodge SM.AnimStateTransitionNode_7.Transition
  nodes [Result(bCanEnterTransition=Get bCanFall() -> (bCanFall:bool))]
 }
 Transition {
  path=ItemFullBodyOverride.AnimGraphNode_StateMachine_0.Dodge SM.AnimStateTransitionNode_8.Transition
  nodes [Result(bCanEnterTransition=Get bDidRebound() -> (bDidRebound:bool))]
 }
 ItemIdleOverride {nodes [
   Input Pose() {name=InPose_Idle}
   Output Pose() {layerGroup=Item Override}
  ]}
 ItemJetPackJumpAdditive {nodes [Output Pose() {layerGroup=Item Override}]}
 ItemJetPackStartAdditive {nodes [Output Pose() {layerGroup=Item Override}]}
 ItemJumpLoopAdditive {nodes [
   Input Pose() {name=InPoseItemJumpLoopAdditive}
   Output Pose() {layerGroup=Item Override}
  ]}
 ItemJumpOverride {nodes [
   Input Pose() {name=InPoseJump}
   Output Pose() {layerGroup=Item Override}
  ]}
 ItemJumpUpAdditive {nodes [
   Input Pose() {name=InPoseItemJumpUpAdditive}
   Output Pose() {layerGroup=Item Override}
  ]}
 ItemPreAimOffsets {nodes [
   Input Pose() {name=InPosePreAimOffsets}
   Output Pose() {layerGroup=Item Override}
  ]}
 ItemPreIK {nodes [
   Input Pose() {name=InPose_PreIK}
   Output Pose() {layerGroup=Item Override}
  ]}
 ItemRightHandIK {nodes [
   Input Pose() {name=InPosePreRightHandIK}
   Input Pose() {name=InPosePostRightHandIK}
   Output Pose() {layerGroup=Item Override}
  ]}
 ItemSkeletalControl {nodes [
   Input Pose()
   Output Pose() {layerGroup=Item Override}
  ]}
 ItemSwimJumpFallLoop {nodes [Output Pose() {layerGroup=Item Override}]}
 ItemSwimJumpSurfaceEnd {nodes [Output Pose() {layerGroup=Item Override}]}
 ItemSwimJumpSurfaceLoop {nodes [Output Pose() {layerGroup=Item Override}]}
 ItemUpperBody {nodes [
   Input Pose!() -> (Pose, InputParam:Fort Anim Input Adjusted Aim Structure) {name=InPoseUpperBody}
   Output Pose() {layerGroup=Item Override}
  ]}
 ItemZipLineStartAdditive {nodes [Output Pose() {layerGroup=Item Override}]}
 OnEmptyBecomeRelevant {nodes [
   OnEmptyBecomeRelevant!() -> (then, Context:"Anim Update Context Structure (by ref)", Node:"Anim Node Reference Structure (by ref)") {is=title K2Node_FunctionEntry}
   > Set InAirTime(InAirTime:double=0) -> (Output_Get:double)
   > Return Node()
  ]}
 OnInAirRefSequenceUpdate {nodes [
   n1 = OnInAirRefSequenceUpdate!() -> (then, Context:"Anim Update Context Structure (by ref)", Node:"Anim Node Reference Structure (by ref)") {is=title K2Node_FunctionEntry}
   > Set InAirAccumulatedTime(InAirAccumulatedTime:double=GetAccumulatedTime!(SequencePlayer:"Sequence Player Reference Structure (by ref)"=Convert to Sequence Player!(Node:"Anim Node Reference Structure (by ref)"=@n1.Node) -> (SequencePlayer:Sequence Player Reference Structure, Result:bool)) -> (ReturnValue:float)) -> (Output_Get:double)
   > Return Node()
  ]}
 OnInAirStateExit {nodes [
   OnInAirStateExit!() -> (then, Context:"Anim Update Context Structure (by ref)", Node:"Anim Node Reference Structure (by ref)") {is=title K2Node_FunctionEntry}
   > Set LastDodgeDirection(LastDodgeDirection:float=Get DodgeAngleRadians()) -> (Output_Get:float)
   > Return Node()
  ]}
 OnInAirUpdate {nodes [
   OnInAirUpdate!() -> (then, Context:"Anim Update Context Structure (by ref)", Node:"Anim Node Reference Structure (by ref)") {is=title K2Node_FunctionEntry}
   > Set InAirTime(InAirTime=float + float(A=Get InAirTime() -> (InAirTime:double), B=Get DeltaTime() -> (DeltaTime:float)))
   > Return Node()
  ]}
 OnStartRollRelevant {nodes [
   OnStartRollRelevant!() -> (then, Context:"Anim Update Context Structure (by ref)", Node:"Anim Node Reference Structure (by ref)") {is=title K2Node_FunctionEntry}
   > Set InAirAccumulatedTime(InAirAccumulatedTime=0)
   > Set InitialDodgeDirection(InitialDodgeDirection:Vector=Get CurrentWorldSpaceDodgeDirection() -> (CurrentWorldSpaceDodgeDirection:Vector)) -> (Output_Get:Vector)
   > Set InAirTime(InAirTime=0)
   > Return Node()
  ]}
]
variables [InAirAccumulatedTime {type=number}]
