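# Post-process material for the Dual Revolvers gameplay plugin that darkens the scene color with a vignette.
# Viewport UVs are corrected for aspect ratio, and their distance from screen center, divided by Scale, feeds an inverted per-channel cubic polynomial weighted by vector parameters A, B and C.
# VignetteAmount, zero by default, blends from white to that falloff before it multiplies the scene.
# Scalar parameters also named A, B and C are left unconnected.
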Material /CherrySmokeDualRevolversGameplay/Assets/Materials/PP_DualRevolversVignette
graph {
 properties {materialDomain=MD_PostProcess, pixelDepthOffsetMode=PDOM_Legacy, usedShadingModels=Default Lit}
 nodes [
  n1 = ViewProperty() {property=MEVP_ViewSize}
  n2 = ComponentMask(None=@n1.Property) {r=true}
  n3 = ComponentMask(None=@n1.Property) {g=true}
  n4 = AppendVector(A=Min(A=Divide(A=@n2, B=@n3)), B=Min(A=Divide(A=@n3, B=@n2)))
  n5 = Multiply(A=ScreenPosition().ViewportUV, B=@n4)
  n6 = Divide(A=OneMinus(None=@n4))
  n7 = Divide(A=Distance(A=Add(A=@n5, B=@n6), B=Constant2Vector() {g=0.5, r=0.5}) {desc=x}, B=ScalarParameter() {defaultValue=2, parameterName=Scale})
  n8 = Multiply(A=@n7, B=VectorParameter() {parameterName=A, defaultValue=<rgba 0.5 0.5 0.5 1>}.RGB)
  n9 = Power(Base=Multiply(A=@n7, B=VectorParameter() {parameterName=B, defaultValue=<rgba 1.7 1.7 1.7 1>}.RGB))
  n10 = Multiply(A=@n7, B=VectorParameter() {parameterName=C, defaultValue=<rgba 2.574 2.6 2.5844 1>}.RGB)
  n11 = OneMinus(None=Add(A=Subtract(A=@n8, B=@n9), B=Power(Base=@n10) {constExponent=3}))
  n12 = Constant() {r=1}
  Multiply(
   A=SceneTexture() {sceneTextureId=PPI_PostProcessInput0}.Color
   B=LinearInterpolate(B=AppendVector(A=@n11, B=@n12), Alpha=ScalarParameter() {parameterName=VignetteAmount}) {constA=1}
  )
  ScalarParameter() {defaultValue=1, parameterName=A}
  ScalarParameter() {defaultValue=1.4, parameterName=B}
  ScalarParameter() {defaultValue=2.6, parameterName=C}
 ]
}
