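# Translucent, unlit, two-sided Niagara material for sprites, ribbons and meshes, giving the Dual Revolvers gameplay VFX a digital derez look.
# Triangle textures, grid-quantized UVs and hashed, time-driven sines build named derez mask, dissolve and gradient signals, alongside a two-color blend.
# A custom-depth soft-outline function feeds depth fade, combining with tint and particle color for emissive adjustment.
# Another chain multiplies that fade by particle alpha, inverted Fresnel and optional sine flicker.
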Material /CherrySmokeDualRevolversGameplay/Assets/VFX/Materials/M_DualRevolvers_Digital
graph {
 properties {
  bAutomaticallySetUsageInEditor=false
  bUsedWithNiagaraMeshParticles=true
  bUsedWithNiagaraRibbons=true
  bUsedWithNiagaraSprites=true
  blendMode=BLEND_Translucent
  opacityMaskClipValue=0.1
  shadingModel=MSM_Unlit
  twoSided=true
  usedShadingModels=Unlit
 }
 nodes [
  Abs(None=Sine(None=Time()))
  n1 = TextureSample(UVs=TextureCoordinate() {uTiling=8, vTiling=8}) {samplerSource=SSM_Wrap_WorldGroupSettings, texture=&/T_Triangles_TEMP}
  n2 = NamedRerouteUsage()
  Add(A=Add(A=Multiply(A=@n1.R), B=Power(Base=@n2) {constExponent=5}), B=LinearInterpolate(B=@n2, Alpha=NamedRerouteUsage()))
  ComponentMask(None=MaterialFunctionCall!(
    TextureObject=TextureObject() {texture=&T_Triangles_TEMP}
    TextureSize
    RotYAxis
    RotZAxis
    RotXAxis
    PreSkinNormal
    ProjectionTransitionContrast
    WorldPosition
   ) {materialFunction=/Game/Packages/Fortress_SharedMaterials/Base_Material_Functions/BaseCustomizationFunctions/PreSkinnedAlignedNormal}.XYZ Texture) {g=true, r=true}
  n3 = Time()
  n4 = TextureCoordinate() {uTiling=3}
  n5 = Multiply(A=Time(), B=Constant2Vector() {r=0.2})
  n6 = ScalarParameter() {defaultValue=300, parameterName=Tiling}
  n7 = Frac(None=Multiply(A=Add(A=@n4, B=@n5), B=@n6))
  n8 = TextureSample(UVs=Add(A=@n4, B=Divide(A=@n7, B=@n6))) {samplerSource=SSM_Wrap_WorldGroupSettings, texture=&T_Triangles_TEMP}
  ConstantBiasScale(None=Sine(None=Add(A=@n3, B=@n8.G)))
  Constant() {r=4}
  Constant() {r=1}
  n9 = LinearInterpolate(
   A=VectorParameter() {group=Derez, parameterName=Color_01, defaultValue=<rgba 0.0308 0.1138 0 1>}.RGB
   B=VectorParameter() {group=Derez, parameterName=Color_02, defaultValue=<rgba 0 1 0.0295 1>}.RGB
   Alpha=Multiply(A=@n1.R, B=Power(Base=@n1.R))
  )
  LinearInterpolate(A=@n9, B=Multiply(A=@n9) {constB=5})
  n10 = NamedRerouteUsage()
  n11 = DepthFade(
   Opacity=MaterialFunctionCall!(
    GlowMeshStencilIndex=ScalarParameter() {defaultValue=40, parameterName="GlowMeshStencilIndex (S)"}
    UseCustomDepth=StaticBoolParameter() {defaultValue=true, parameterName=UseCustomDepth}
    UseCustomStencil=StaticBoolParameter() {parameterName=UseCustomStencil}
    OutlineThickness=ScalarParameter() {defaultValue=2.7, parameterName="OutlineThickness (S)"}
    OutlineStrength
    PixelDepthOffset
    MaxBlurDepth=ScalarParameter() {defaultValue=150, parameterName="MaxBlurDepth (S)"}
    ------EnableInnerGlow------=ScalarParameter() {defaultValue=1}
    InnerGlowMaxArea=ScalarParameter() {defaultValue=100, parameterName="InnerGlowMaxArea (S)"}
    InnerGlowPower=ScalarParameter() {defaultValue=10, parameterName="InnerGlowPower (S)"}
    InnerGlowStrength=ScalarParameter() {defaultValue=0.5, parameterName="InnerGlowStrength (S)"}
    ------LessUse------
    MeshPosition
    Distance Steps
    Radial Steps
   ) {materialFunction=/Game/Effects/Fort_Effects/Effects/Atlas/Materials/FXCommon_SoftOutline}.Result
   FadeDistance=ScalarParameter() {defaultValue=100, parameterName=DepthFade}
  )
  n12 = VectorParameter() {parameterName=C, defaultValue=<rgba 0.7182 7 0.5126 0.08>}
  n13 = Multiply(A=@n11, B=@n12.A)
  n14 = ParticleColor()
  n15 = LinearInterpolate(A=Multiply(A=@n12.RGB, B=@n14.RGB), B=@n14.RGB, Alpha=ScalarParameter() {parameterName=ParticleColorOnly})
  MaterialFunctionCall(Emissive=Add(A=Multiply(A=@n10, B=@n13), B=Divide(A=@n15))) {materialFunction=&MF_FX_EmissiveAdjustments}
  n16 = ParticleColor()
  Multiply(A=Constant() {r=10}, B=@n16.RGB)
  n17 = Multiply(A=LinearInterpolate(A=@n13, B=@n11, Alpha=ScalarParameter() {parameterName=ParticleColorOnly}), B=@n14.A)
  n18 = Fresnel(ExponentIn=ScalarParameter() {defaultValue=3, parameterName=Fresnel_Exponent}) {baseReflectFraction=0, exponent=1}
  n19 = Saturate(None=Power(Base=OneMinus(None=@n18), Exp=ScalarParameter() {parameterName=Fresnel_Power}))
  n20 = Time()
  n21 = ScalarParameter() {defaultValue=10, parameterName=Flick_Speed}
  n22 = MaterialFunctionCall(
   Input=Sine(None=Multiply(A=@n20, B=@n21))
   Input Low=Constant() {r=-1}
   Input High=Constant() {r=1}
   Target Low=ScalarParameter() {defaultValue=0.7, parameterName=Flick_Low}
   Target High=ScalarParameter() {defaultValue=1.3, parameterName=Flick_High}
  ) {materialFunction=&RemapValueRange}
  n23 = ScalarParameter() {parameterName=EnableFlick}
  Multiply(A=@n11, B=Multiply(A=Multiply(A=@n17, B=@n19), B=LinearInterpolate(B=@n22.Result, Alpha=@n23) {constA=1}))
  n24 = NamedRerouteUsage()
  Multiply(A=Multiply(A=@n24) {constB=0.25}, B=@n11)
  n25 = Multiply(B=@n8.R) {constA=1}
  NamedRerouteDeclaration(None=LinearInterpolate(A=@n25, Alpha=NamedRerouteUsage())) {name=Derez_Mask, nodeColor=<rgba 0.0076 0.75 1 1>}
  NamedRerouteDeclaration(None=@n16.RGB) {name=Emissive, nodeColor=<rgba 0 0.3437 1 1>}
  NamedRerouteDeclaration() {name=Grid, nodeColor=<rgba 1 0 0.5859 1>}
  n26 = TextureCoordinate()
  n27 = ScalarParameter() {defaultValue=300, parameterName=Tiling}
  n28 = Multiply(A=OneMinus(None=ComponentMask(None=@n26) {g=true, r=true}), B=@n27)
  n29 = Saturate(None=Divide(A=Floor(None=@n28), B=@n27))
  NamedRerouteDeclaration(None=@n29) {name=U_Gradient, nodeColor=<rgba 0.1484 1 0 1>}
  n30 = TextureCoordinate()
  n31 = ScalarParameter() {defaultValue=300, parameterName=Tiling}
  n32 = MaterialFunctionCall(Vec2=Ceil(None=Multiply(A=@n30, B=@n31))) {materialFunction=&MF_Hash}
  n33 = TextureCoordinate() {uTiling=0.5, vTiling=0.5}
  n34 = Ceil(None=Abs(None=@n33))
  n35 = Add(A=Time() {period=1}, B=Add(A=ComponentMask(None=@n34) {r=true}, B=ComponentMask(None=@n34) {g=true}))
  n36 = Multiply(A=Sine(None=Add(A=@n32.Vec2, B=@n35)) {period=125}, B=@n29)
  NamedRerouteDeclaration(None=ComponentMask(None=LinearInterpolate(A=@n36, Alpha=@n29) {constB=0.5}) {r=true}) {name=Gradient_Dissolve, nodeColor=<rgba 0 1 0.625 1>}
  NamedRerouteDeclaration(None=SmoothStep(Value=@n1.G) {constMax=0.5}) {name=Derez_Dissolve, nodeColor=<rgba 0.4766 1 0 1>}
  NamedRerouteUsage()
  n37 = NamedRerouteUsage()
  OneMinus(None=@n37)
  n38 = TextureSample(UVs=Multiply(A=TextureCoordinate() {uTiling=3}, B=ScalarParameter() {defaultValue=300, parameterName=Tiling})) {texture=&T_Triangles_TEMP}
  OneMinus(None=Saturate(None=LinearInterpolate(Alpha=@n38.R) {constA=-2, constB=10}))
  Power(Base=Multiply(A=@n24, B=@n37)) {constExponent=4}
  Saturate(None=Multiply(A=@n25) {constB=1.2})
  ScalarParameter() {parameterName=UseGlitch}
  n39 = Multiply(A=ScalarParameter() {defaultValue=0.5, parameterName=Sine Speed}, B=@n35)
  n40 = TextureSample(UVs=@n33) {texture=&T_Triangles_TEMP}
  Sine(None=Frac(None=Add(A=@n39, B=@n40.R)))
  Step() {constX=0.5}
 ]
}
