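# Additive unlit holographic material for Niagara mesh particles in the Dual Revolvers VFX.
# It multiplies particle color RGB by a blend between a flat value of one and a Fresnel term.
# The Fresnel channel of a Dynamic Parameter drives that blend, so each particle can shift from solid color to rim glow.
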Material /CherrySmokeDualRevolversGameplay/Assets/VFX/Materials/M_DualRevolvers_Holo
graph {
 properties {
  bAutomaticallySetUsageInEditor=false
  bUsedWithNiagaraMeshParticles=true
  blendMode=BLEND_Additive
  shadingModel=MSM_Unlit
  usedShadingModels=Unlit
 }
 nodes [Multiply(
   A=ParticleColor().RGB
   B=LinearInterpolate(B=Fresnel(), Alpha=DynamicParameter() {paramNames [Fresnel,Param2,Param3,Param4]}.Fresnel) {constA=1}
  )]
}
