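# Post-process material for Dual Revolvers hangtime that adds a green, pulsing triangle-patterned vignette to scene color after depth of field.
# A tiled placeholder triangle texture supplies a color gradient from red and a time-sine SmoothStep pulse from green, masked toward screen edges by an inverted SphereMask.
# Low quality, forward and mobile paths use a flat-colored vignette instead.
# BlendWeight runs inverted: its default of 1 shows the untouched scene, 0 the full effect.
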Material /CherrySmokeDualRevolversGameplay/Assets/VFX/Materials/M_PostProcess_DualRevolvers_Hangtime
graph {
 properties {
  blendableLocation=BL_SceneColorAfterDOF
  materialDomain=MD_PostProcess
  pixelDepthOffsetMode=PDOM_Legacy
  shadingModel=MSM_Unlit
  usedShadingModels=Unlit
 }
 nodes [
  n1 = Sine(None=Time())
  n2 = TextureSample(UVs=TextureCoordinate() {uTiling=2, vTiling=2}) {samplerSource=SSM_Wrap_WorldGroupSettings, texture=&/T_Triangles_TEMP}
  n3 = Multiply(A=Multiply(A=NamedRerouteUsage()) {constB=5}, B=SmoothStep(Max=Abs(None=@n1), Value=@n2.G))
  n4 = TextureCoordinate()
  n5 = Constant2Vector() {g=0.5, r=0.5}
  n6 = Constant() {r=0.75}
  n7 = Constant()
  n8 = Power(Base=OneMinus(None=SphereMask(A=@n4, B=@n5, Radius=@n6, Hardness=@n7))) {constExponent=4.5}
  n9 = Multiply(A=@n8, B=VectorParameter() {group=Derez, parameterName=Color_03, defaultValue=<rgba 0.1946 0.7204 0 1>}.RGB)
  n10 = ShadingPathSwitch(Default=Multiply(A=Multiply(A=@n3, B=@n8), B=@n8), Forward=@n9, Mobile=@n9)
  n11 = QualitySwitch(Default=@n10, Low=@n9, High=@n10, Medium=@n10, Epic=@n10)
  n12 = ComponentMask(None=SceneTexture() {sceneTextureId=PPI_PostProcessInput0}.Color) {b=true, g=true, r=true}
  LinearInterpolate(A=Add(A=Multiply(A=@n11), B=@n12), B=@n12, Alpha=ScalarParameter() {defaultValue=1, parameterName=BlendWeight})
  n13 = OneMinus(None=@n2.R)
  n14 = LinearInterpolate(
   A=VectorParameter() {group=Derez, parameterName=Color_01, defaultValue=<rgba 0.054 0.2 0 1>}.RGB
   B=VectorParameter() {group=Derez, parameterName=Color_02, defaultValue=<rgba 0 5 0.1475 1>}.RGB
   Alpha=Multiply(A=@n13, B=Power(Base=@n13) {constExponent=0.5})
  )
  NamedRerouteDeclaration(None=LinearInterpolate(A=@n14, B=Multiply(A=@n14) {constB=5}) {constAlpha=0}) {name=Emissive, nodeColor=<rgba 0 0.3438 1 1>}
 ]
}
