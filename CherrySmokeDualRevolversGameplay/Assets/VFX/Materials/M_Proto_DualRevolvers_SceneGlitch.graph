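# Translucent unlit, two-sided Niagara sprite material for a prototype Dual Revolvers scene glitch effect that distorts the screen behind particles.
# Soft noise panned in radial coordinates offsets a SceneColor lookup, masked by a radial gradient, with dynamic parameters for speed, strength and time offset.
# A platform switch zeroes the scene color on PC-Perf, Android and Switch.
# Particle color, depth fade, hue-shift and tiled-grid branches are present, but final output wiring is not shown.
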Material /CherrySmokeDualRevolversGameplay/Assets/VFX/Materials/M_Proto_DualRevolvers_SceneGlitch
graph {
 properties {
  bAutomaticallySetUsageInEditor=false
  bUseTranslucencyVertexFog=false
  bUsedWithNiagaraSprites=true
  blendMode=BLEND_Translucent
  shadingModel=MSM_Unlit
  twoSided=true
  usedShadingModels=Unlit
 }
 nodes [
  Add(A=TextureCoordinate(), B=Constant2Vector() {g=-0.5, r=-0.5})
  n1 = DynamicParameter() {defaultValue=<rgba -0.6 1 1 1>, paramNames [DistortionSpeed, DistortionStrength, TimeOffset, Param4]}
  n2 = Panner(
   Coordinate=MaterialFunctionCall() {materialFunction=&VectorToRadialValue}.Radial Coordinates
   Time=Add(A=Time(), B=@n1.TimeOffset)
   Speed=MaterialFunctionCall(X=Constant(), Y=@n1.DistortionSpeed) {materialFunction=&MakeFloat2}.Result
  ) {speedY=-0.6}
  n3 = TextureSample!(UVs=@n2, Tex, "DDX(UVs)"=DDX(Value=@n2), "DDY(UVs)"=DDY(Value=@n2), Apply View MipBias) {mipValueMode=TMVM_Derivative, texture=/Game/Effects/Fort_Effects/Emotes/UltraLight/Texture/T_SoftNoise_BC7}
  n4 = Constant() {r=1}
  n5 = MaterialFunctionCall(Input=@n3.G, Input Low=Constant(), Input High=@n4, Target Low=Constant() {r=-1}, Target High=@n4) {materialFunction=&RemapValueRange}
  n6 = MaterialFunctionCall(Radius=Constant() {r=0.5}, Density=Constant() {r=1}) {materialFunction=&RadialGradientExponential}
  n7 = SceneColor!(OffsetFraction=Multiply(A=Multiply(A=@n5.Result, B=@n1.DistortionStrength) {constB=0.05}, B=@n6.RadialGradientExponential)) {inputMode=OffsetFraction, constInput=<0.02 0.02>}
  n8 = Constant()
  n9 = MaterialFunctionCall!("Xb1,PS4"=@n7.RGB, "PC-High&&Epic,XSX,PS5"=@n7.RGB, "PC-Perf,Android"=@n8, Switch=@n8, "PC-Low(NotPerf)&&Medium"=@n7.RGB) {materialFunction=/Game/Effects/Niagara/MasterMaterials/Functions/Common/PlatformSwitch_3D}
  Desaturation(None=@n9.Result, Fraction=Constant() {r=0.5})
  MaterialFunctionCall(Hue Shift Percentage=Constant() {r=0.7}, Texture=@n9.Result) {materialFunction=&HueShift}
  MaterialFunctionCall("Xb1,PS4"=@n7.A, "PC-High&&Epic,XSX,PS5"=@n7.A, "PC-Perf,Android"=@n8, Switch=@n8, "PC-Low(NotPerf)&&Medium"=@n7.A) {materialFunction=&PlatformSwitch_1D}
  MaterialFunctionCall(Input=@n3.G) {materialFunction=&RemapValue_0-1}
  n10 = ParticleColor()
  n11 = Multiply(A=@n10.A, B=SmoothStep(Value=@n6.RadialGradientExponential) {constMax=0.2})
  Multiply(A=@n11, B=DepthFade() {fadeDistanceDefault=250})
  Multiply(A=@n10.RGB, B=@n9.Result)
  n12 = Frac(None=TextureCoordinate() {uTiling=100, vTiling=100})
  n13 = ComponentMask(None=@n12) {r=true}
  Multiply(A=@n13, B=OneMinus(None=@n13))
  n14 = ComponentMask(None=@n12) {g=true}
  n15 = Multiply(A=@n14, B=OneMinus(None=@n14))
  Power(Base=Multiply(A=@n11, B=Multiply(A=@n15) {constB=2}))
 ]
}
